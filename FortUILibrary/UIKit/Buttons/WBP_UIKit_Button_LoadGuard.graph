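# UIKit wrapper widget that shows a loading button with a progress bar in place of slotted content until loading completes.
# SetLoadingProgress sets the bar, optionally animated, and swaps visibility at progress 1, keeping unfinished content Hidden.
# GetDesiredFocusTarget follows that loading state, while hover and click dispatchers relay only the loading button's events.
# PreConstruct applies default progress and, at design time, fills an empty slot with a placeholder regular button.
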WidgetBlueprint /FortUILibrary/UIKit/Buttons/WBP_UIKit_Button_LoadGuard
graphs [
 EventGraph {nodes [
   n1 = "On Hovered (Button_Loading)"!() -> (OutputDelegate, then, Button:Common Button Base Object Reference) {is=title K2Node_ComponentBoundEvent}
   > Call On Hovered!(execute:Exec, self:WBP UIKit Button Load Guard Object Reference, Button:Common Button Base Object Reference=@n1.Button) -> (then:Exec) {is=title K2Node_CallDelegate}
   n2 = "On Clicked (Button_Loading)"!() -> (OutputDelegate, then, Button:Common Button Base Object Reference) {is=title K2Node_ComponentBoundEvent}
   > Call On Clicked!(execute:Exec, self:WBP UIKit Button Load Guard Object Reference, Button:Common Button Base Object Reference=@n2.Button) -> (then:Exec) {is=title K2Node_CallDelegate}
   n3 = Event PreConstruct()
   > n4 = Branch(Condition=@n3.IsDesignTime)
   > n5 = Branch(Condition=HasAnyChildren!(self:Panel Widget Object Reference=Get Slot_Content() -> (Slot_Content:Named Slot Object Reference)) -> (ReturnValue:bool))
   SetLoadingProgress!(
    execute=[@n7.then, @n5.then]
    self
    InLoadProgress:double=Get DefaultLoadingProgress() -> (DefaultLoadingProgress:double)
    ShowTransition:bool=false
   ) -> (then)
   > SetLoadingText!(execute, self, InLoadingText:Text="LOCTEXT(\"Loading...\")") -> (then)
   n6 = Create Widget!(
    execute:Exec=@n5.else
    Class:User Widget Class Reference=&/WBP_UIKit_Button_Regular.WBP_UIKit_Button_Regular_C
    OwningPlayer:Player Controller Object Reference
    Style:Common Button Style Class Reference=/FortUILibrary/UIKit/Buttons/Blocks/ButtonStyle_UIKit_Empty.ButtonStyle_UIKit_Empty_C
    bApplyAlphaOnDisable:bool=False
    bLocked:bool=False
    bSelectable:bool=False
    bShouldSelectUponReceivingFocus:bool=False
    bInteractableWhenSelected:bool=False
    bToggleable:bool=False
    bTriggerClickedAfterSelection:bool=False
    bDisplayInputActionWhenNotInteractable:bool=True
    bHideInputActionWithKeyboard:bool=False
    bShouldUseFallbackDefaultInputAction:bool=True
    bRequiresHold:bool=False
    HoldData:Common UIHold Data Class Reference=&CommonUI_HoldData_Base_C
    ClickMethod:EButtonClickMethod Enum=DownAndUp
    InputPriority:int=0
    TriggeringInputAction:Data Table Row Handle Structure
   ) -> (then:Exec, ReturnValue:WBP UIKit Button Regular Object Reference)
   > n7 = AddChild(self=Get Slot_Content(), Content=@n6)
   SetLoadingProgress(execute=@n4.else, InLoadProgress=Get DefaultLoadingProgress(), ShowTransition=false)
  ]}
 GetDesiredFocusTarget {nodes [
   GetDesiredFocusTarget() {is=title K2Node_FunctionEntry}
   n1 = Get Slot_Content()
   n2 = GetChildAt!(self:Panel Widget Object Reference=@n1, Index:int=0) -> (ReturnValue:Widget Object Reference)
   n3 = HasAnyChildren(self=@n1)
   > Return Node!(
    execute
    Target:Widget Object Reference=Select(
     Option 0:Widget Object Reference=Select(Option 0=@n1, Option 1=@n2, Index=@n3)
     Option 1:Widget Object Reference=Get Button_Loading() -> (Button_Loading:WBP UIKit Button Generic Base Object Reference)
     Index=GetIsLoading!(self) -> (IsLoading:bool)
    ) -> (ReturnValue:Widget Object Reference)
   )
  ]}
 GetIsLoading {nodes [
   GetIsLoading() {is=title K2Node_FunctionEntry}
   n1 = Get ProgressBar() -> (ProgressBar:WBP UIKit Progress Bar Object Reference)
   > Return Node!(
    execute
    IsLoading:bool="float < float"(A=Get ProgressValue!(self:WBP UIKit Progress Bar Base Object Reference=@n1) -> (ProgressValue:double), B=1)
   )
  ]}
 OnClicked {nodes [OnClicked!() -> (then, Button:Common Button Base Object Reference) {is=title K2Node_FunctionEntry}]}
 OnHovered {nodes [OnHovered!() -> (then, Button:Common Button Base Object Reference) {is=title K2Node_FunctionEntry}]}
 SetLoadingProgress {nodes [
   n1 = SetLoadingProgress!() -> (then, InLoadProgress:double=~, ShowTransition:bool=~) {is=title K2Node_FunctionEntry}
   > n2 = Branch(Condition=@n1.ShowTransition)
   n3 = Get ProgressBar()
   n4 = Get InLoadProgress() -> (InLoadProgress:double)
   > n5 = SetProgressTransition!(execute, self:WBP UIKit Progress Bar Base Object Reference=@n3, Progress:double=@n4) -> (then)
   n6 = "float >= float"(A=Get InLoadProgress(), B=1)
   Set Visibility(
    execute=[@n7.then, @n5.then]
    Visibility=Select(Option 0:ESlateVisibility Enum=SelfHitTestInvisible, Option 1:ESlateVisibility Enum=Collapsed, Index=@n6) -> (ReturnValue:ESlateVisibility Enum)
    self=Get Button_Loading()
   )
   > Set Visibility(Visibility=Select(Option 0=Hidden, Option 1=SelfHitTestInvisible, Index=@n6), self=Get Slot_Content())
   n7 = SetProgress!(execute=@n2.else, self:WBP UIKit Progress Bar Base Object Reference=@n3, Progress:double=@n4) -> (then)
  ]}
 SetLoadingText {nodes [
   n1 = SetLoadingText!() -> (then, InLoadingText:Text="LOCTEXT(\"Loading...\")") {is=title K2Node_FunctionEntry}
   > SetText!(execute, self:WBP UIKit Button Generic Base Object Reference=Get Button_Loading(), Text:Text=@n1) -> (then)
  ]}
]
variables [
 DefaultLoadingProgress {type=number, default=0.5}
 OnClicked {}
 OnHovered {}
]
widgets {
 info {inheritedWidgetCount=0, namedSlotCount=1, parentClass=&UserWidget, rootWidgetClass=&Overlay, widgetCount=4}
 widgets [
  Overlay_10 {widgetClassPath=&Overlay, widget=Overlay_10}
  Slot_Content {
   widgetClassPath=&NamedSlot
   bIsVariable=true
   parent=Overlay_10
   widget=Slot_Content
   slot {class=OverlaySlot, properties {horizontalAlignment=HAlign_Fill, verticalAlignment=VAlign_Fill}}
  }
  Button_Loading {
   widgetClassPath=/FortUILibrary/Systems/ModularButton/WBP_UIKit_ButtonGeneric_Base
   bIsVariable=true
   parent=Overlay_10
   widget=Button_Loading
   properties {
    bHideInputAction=true
    backgroundClass=/FortUILibrary/UIKit/Buttons/Blocks/WBP_UIKit_Block_BackgroundWithOutline.WBP_UIKit_Block_BackgroundWithOutline_C
    backgroundMaterial=/FortUILibrary/UIKit/Buttons/MaterialInstances/CTAButtons/MI_UIKit_EmptyProgress
    contentHorizontalAlignment=HAlign_Fill
    contentVerticalAlignment=VAlign_Fill
    iconHorizontalAlignment=HAlign_Left
    iconMaterial=/FortUILibrary/UIKit/Throbber/MI_UIKit_Color_Throbber_Light
    buttonPadding=<ltrb 12 0 12 0>
    font {
     bForceMonospaced=false
     bMaterialIsStencil=false
     fontMaterial=/FortUILibrary/UIKit/Buttons/MaterialInstances/CTAButtons/MI_UIKit_TextQuiet_Prompt
     fontObject=&HeadingNow
     letterSpacing=10
     monospacedWidth=1
     size=18
     skewAmount=0
     typefaceFontName=76 Bold
     outlineSettings {
      bApplyOutlineToDropShadows=false
      bMiteredCorners=false
      bSeparateFillAlpha=false
      outlineMaterial=null
      outlineSize=0
      outlineColor=<rgba 0 0 0 1>
     }
    }
    iconSize=<36 36>
   }
   slot {class=OverlaySlot, properties {horizontalAlignment=HAlign_Fill, verticalAlignment=VAlign_Fill}}
  }
  ProgressBar {
   widgetClassPath=/FortUILibrary/UIKit/ProgressBar/WBP_UIKit_ProgressBar
   bIsVariable=true
   namedSlotHost=Button_Loading
   widget=ProgressBar
   properties {
    cornerRoundness=0.6
    durationTime=0.25
    progressValue=0.5
    useCircular=false
    useRoundedProgress=false
    visibility=Visible
    colorAndOpacity=<rgba 1 1 1 0.4>
   }
  }
 ]
}
