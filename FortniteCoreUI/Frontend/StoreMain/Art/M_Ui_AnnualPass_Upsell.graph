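# Translucent UI material that draws a tinted, fading Battle Pass icon, apparently for an Annual Pass upsell in the store front end.
# A horizontal UV gradient blends two color parameters, RGB2 to RGB1 (red-orange to yellow by default), and is added to the icon texture's RGB.
# The icon's alpha is multiplied by inverted vertical UV, so it fades out from top to bottom.
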Material /FortniteCoreUI/Frontend/StoreMain/Art/M_Ui_AnnualPass_Upsell
graph {
 properties {
  bUsedWithStaticMesh=false
  blendMode=BLEND_Translucent
  materialDomain=MD_UI
  pixelDepthOffsetMode=PDOM_Legacy
  usedShadingModels=Default Lit
 }
 nodes [
  n1 = TextureCoordinate()
  n2 = TextureSample() {texture=/Game/UI/Foundation/Textures/Icons/Items/T-FNBR-BattlePass-L}
  Add(
   A=LinearInterpolate(
    A=VectorParameter() {parameterName=RGB2, defaultValue=<rgba 0.6795 0.0844 0.0103 1>}.RGB
    B=VectorParameter() {parameterName=RGB1, defaultValue=<rgba 1 0.5615 0 1>}.RGB
    Alpha=ComponentMask(None=@n1) {r=true}
   )
   B=@n2.RGB
  )
  Multiply(A=@n2.A, B=OneMinus(None=ComponentMask(None=@n1) {g=true}))
 ]
}
