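# Sound cue for the item shop offer details UI sound in the front end, routed to the Fort_UI sound class.
# A single wave player feeds a modulator that applies a fixed upward pitch shift.
# The cue trims its volume slightly and overrides concurrency to prevent new instances once the limit is reached.
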SoundCue /Game/Sounds/Fort_UI_Sounds/FrontEnd/ItemShop/Fort_UI_Shop_OfferDetails_Cue
root=Modulator_0
nodes [
 {id=Modulator_0, children [WavePlayer_0], properties {pitchMax=1.1, pitchMin=1.1}}
 {id=WavePlayer_0, properties {soundWaveAssetPtr=/Game/Sounds/Fort_UI_Sounds/FrontEnd/ItemShop/wav/Fort_UI_Shop_OfferDetails}}
]
properties {
 bOverrideConcurrency=true
 maxDistance=2097152
 soundClassObject=&Fort_UI
 volumeMultiplier=0.8141
 concurrencyOverrides {resolutionRule=PreventNew}
}
