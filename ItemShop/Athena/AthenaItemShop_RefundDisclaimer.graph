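# Item shop widget that shows a small no-refund disclaimer, reading "This purchase is not eligible for refund." by default.
# A translucent dark CommonBorder wraps a horizontal box holding the centered, wrapping DisclaimerText in an extra-small base text style.
# On PreConstruct, a non-empty OverrideText variable replaces the default wording; an empty value leaves it unchanged.
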WidgetBlueprint /ItemShop/Athena/AthenaItemShop_RefundDisclaimer
graphs [EventGraph {nodes [
   Event PreConstruct()
   n1 = Get OverrideText() -> (OverrideText:Text)
   > n2 = Branch(Condition=TextIsEmpty(InText=@n1))
   "SetText (Text)"(execute=@n2.else, self=Get DisclaimerText() -> (DisclaimerText:Common Text Object Reference), InText=@n1)
  ]}]
variables [OverrideText {type=string, default=""}]
widgets {
 info {inheritedWidgetCount=0, parentClass=&UserWidget, rootWidgetClass=&CommonBorder, widgetCount=3}
 widgets [
  CommonBorder_0 {
   widgetClassPath=&CommonBorder
   widget=CommonBorder_0
   properties {
    background {
     bIsDynamicallyLoaded=false
     drawAs=Image
     imageType=NoImage
     mirroring=NoMirror
     resourceName=None
     resourceObject=null
     tiling=NoTile
     imageSize=<32 32>
     margin=<ltrb 0 0 0 0>
     outlineSettings {
      bUseBrushTransparency=false
      roundingType=HalfHeightRadius
      width=0
      color {colorUseRule=UseColor_Specified, specifiedColor=<rgba 0 0 0 0>}
      cornerRadii=<0 0 0 0>
     }
     tintColor {colorUseRule=UseColor_Specified, specifiedColor=<rgba 1 1 1 1>}
     uVRegion {bIsValid=false, max=<0 0>, min=<0 0>}
    }
    brushColor=<rgba 0.0116 0.016 0.0395 0.6>
    padding=<ltrb 16 2 16 2>
   }
  }
  HorizontalBoxDisclaimer {
   widgetClassPath=&HorizontalBox
   bIsVariable=true
   parent=CommonBorder_0
   widget=HorizontalBoxDisclaimer
   slot {class=BorderSlot, properties {padding=<ltrb 16 2 16 2>}}
  }
  DisclaimerText {
   widgetClassPath=&CommonTextBlock
   bIsVariable=true
   parent=HorizontalBoxDisclaimer
   widget=DisclaimerText
   properties {
    justification=Center
    renderOpacity=0.8
    style=/FortUILibrary/Styles/Typography/Foundation/TextStyle-Base-XS.TextStyle-Base-XS_C
    text=This purchase is not eligible for refund.
    wrapTextAt=450
    font {
     bForceMonospaced=false
     bMaterialIsStencil=false
     fontMaterial=null
     fontObject=&NotoSans
     letterSpacing=0
     monospacedWidth=1
     size=16
     skewAmount=0
     typefaceFontName=NotoSans-Regular
     outlineSettings {
      bApplyOutlineToDropShadows=false
      bMiteredCorners=false
      bSeparateFillAlpha=false
      outlineMaterial=null
      outlineSize=0
      outlineColor=<rgba 0 0 0 1>
     }
    }
    shadowOffset=<0 0>
   }
   slot {class=HorizontalBoxSlot, properties {verticalAlignment=VAlign_Center, padding=<ltrb 0 4 0 4>, size {sizeRule=Fill, value=1}}}
  }
 ]
}
