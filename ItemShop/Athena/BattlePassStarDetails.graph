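# Item shop widget that shows a banner noting an item comes with bonus Battle Pass stars.
# A purchasable-gradient border wraps a horizontal row with a material-driven star icon, a count text and a fixed explanatory message.
# The SetNumOfBattlePassStars custom event takes an integer and formats it as "+ {x}" into the count text.
# The graph only sets that text and handles no visibility or purchase logic.
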WidgetBlueprint /ItemShop/Athena/BattlePassStarDetails
graphs [EventGraph {nodes [
   n1 = SetNumOfBattlePassStars!() -> (OutputDelegate, then, NumOfBattlePassStars:int) {is=title K2Node_CustomEvent}
   > "SetText (Text)"(
    self=Get T_BattleStarInfoSupplemental() -> (T_BattleStarInfoSupplemental:Common Text Object Reference)
    InText=Format Text!(Format="LOCTEXT(\"+ {x}\")", x:int=@n1.NumOfBattlePassStars) -> (Result)
   )
  ]}]
variables []
widgets {
 info {inheritedWidgetCount=0, parentClass=&UserWidget, rootWidgetClass=&CommonBorder, widgetCount=6}
 widgets [
  CommonBorder_0 {
   widgetClassPath=&CommonBorder
   widget=CommonBorder_0
   properties {
    horizontalAlignment=HAlign_Left
    style=/FortUILibrary/Styles/Border/Border_Grad_Purchasable.Border_Grad_Purchasable_C
    padding=<ltrb 16 2 16 2>
   }
  }
  HB_BattlePassStarsSupplemental {
   widgetClassPath=&HorizontalBox
   bIsVariable=true
   parent=CommonBorder_0
   widget=HB_BattlePassStarsSupplemental
   slot {class=BorderSlot, properties {horizontalAlignment=HAlign_Left, padding=<ltrb 16 2 16 2>}}
  }
  ScaleBox_0 {widgetClassPath=&ScaleBox, parent=HB_BattlePassStarsSupplemental, widget=ScaleBox_0, slot {class=HorizontalBoxSlot}}
  Image_BattleStarSupplemental {
   widgetClassPath=&Image
   bIsVariable=true
   parent=ScaleBox_0
   widget=Image_BattleStarSupplemental
   properties {
    visibility=SelfHitTestInvisible
    brush {
     bIsDynamicallyLoaded=false
     drawAs=Image
     imageType=NoImage
     mirroring=NoMirror
     resourceName=None
     resourceObject=/Game/UI/Foundation/Materials/M-BattlePassStarStoreDisplay
     tiling=NoTile
     imageSize=<35 35>
     margin=<ltrb 0 0 0 0>
     outlineSettings {
      bUseBrushTransparency=false
      roundingType=HalfHeightRadius
      width=0
      color {colorUseRule=UseColor_Specified, specifiedColor=<rgba 0 0 0 0>}
      cornerRadii=<0 0 0 0>
     }
     tintColor {colorUseRule=UseColor_Specified, specifiedColor=<rgba 1 1 1 1>}
     uVRegion {bIsValid=false, max=<0 0>, min=<0 0>}
    }
    renderTransform {angle=0, scale=<1 1>, shear=<0 0>, translation=<0 -1>}
   }
   slot {class=ScaleBoxSlot}
  }
  T_BattleStarInfoSupplemental {
   widgetClassPath=&CommonTextBlock
   bIsVariable=true
   parent=HB_BattlePassStarsSupplemental
   widget=T_BattleStarInfoSupplemental
   properties {
    style=/FortUILibrary/Styles/Typography/Foundation/TextStyle-Header-S.TextStyle-Header-S_C
    text=+ 10
    font {
     bForceMonospaced=false
     bMaterialIsStencil=false
     fontMaterial=null
     fontObject=&Burbank
     letterSpacing=0
     monospacedWidth=1
     size=24
     skewAmount=0
     typefaceFontName=Burbank-Black
     outlineSettings {
      bApplyOutlineToDropShadows=false
      bMiteredCorners=false
      bSeparateFillAlpha=false
      outlineMaterial=null
      outlineSize=0
      outlineColor=<rgba 0 0 0 1>
     }
    }
    shadowOffset=<0 0>
   }
   slot {class=HorizontalBoxSlot, properties {verticalAlignment=VAlign_Center, padding=<ltrb 4 0 0 0>}}
  }
  CommonTextBlock_0 {
   widgetClassPath=&CommonTextBlock
   parent=HB_BattlePassStarsSupplemental
   widget=CommonTextBlock_0
   properties {
    style=/FortUILibrary/Styles/Typography/Foundation/TextStyle-Base-XS-B-70pc.TextStyle-Base-XS-B-70pc_C
    text="This item comes with Battle Pass stars!"
    wrapTextAt=450
    colorAndOpacity {colorUseRule=UseColor_Specified, specifiedColor=<rgba 1 1 1 0.7>}
    font {
     bForceMonospaced=false
     bMaterialIsStencil=false
     fontMaterial=null
     fontObject=&NotoSans
     letterSpacing=0
     monospacedWidth=1
     size=16
     skewAmount=0
     typefaceFontName=NotoSans-Bold
     outlineSettings {
      bApplyOutlineToDropShadows=false
      bMiteredCorners=false
      bSeparateFillAlpha=false
      outlineMaterial=null
      outlineSize=0
      outlineColor=<rgba 0 0 0 1>
     }
    }
    shadowOffset=<0 0>
   }
   slot {class=HorizontalBoxSlot, properties {verticalAlignment=VAlign_Center, padding=<ltrb 10 4 0 4>}}
  }
 ]
}
