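# Purchase confirmation widget for the Athena item shop, showing an offer's name, granted items and price over a blurred lightbox.
# When an offer is set, UpdateFromOffer uppercases the name, rebuilds the horizontal reward tiles and fills currency icon and prices, with strikethrough for sales.
# It shows the nonrefundable disclaimer only for nonrefundable offers in Athena.
# CONFIRM and BACK deactivate the widget, then fire Confirm or Decline dispatchers, leaving purchase logic to listeners.
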WidgetBlueprint /ItemShop/Athena/AthenaItemShopConfirmation
graphs [
 Confirm {nodes [Confirm() {is=title K2Node_FunctionEntry}]}
 Decline {nodes [Decline() {is=title K2Node_FunctionEntry}]}
 EventGraph {nodes [
   "On Button Clicked (CancelButton)"!() -> (OutputDelegate, then, Button:Common Button Legacy Object Reference) {is=title K2Node_ComponentBoundEvent}
   > DeactivateWidget()
   > Call Decline!(execute:Exec, self:Athena Item Shop Confirmation Object Reference) -> (then:Exec) {is=title K2Node_CallDelegate}
   "On Button Clicked (PurchaseButton)"!() -> (OutputDelegate, then, Button:Common Button Legacy Object Reference) {is=title K2Node_ComponentBoundEvent}
   > DeactivateWidget()
   > Call Confirm!(execute:Exec, self:Athena Item Shop Confirmation Object Reference) -> (then:Exec) {is=title K2Node_CallDelegate}
   Event On Offer Set() {is=title K2Node_Event}
   > UpdateFromOffer()
   Event Construct()
   Event On Activated()
  ]}
 SetPurchaseNumber {nodes [
   n1 = SetPurchaseNumber!() -> (then, InNumToPurchse:int) {is=title K2Node_FunctionEntry}
   > Set NumberToPurchase(NumberToPurchase:int=@n1) -> (Output_Get:int=0)
  ]}
 SetupPrice {nodes [
   SetupPrice!() -> (
    then
    PriceContainer:Widget Object Reference
    NormalPriceText:Text Object Reference
    SalePriceContainer:Widget Object Reference
    SalePriceText:Text Object Reference
    CurrencyImage:Image Object Reference
   ) {is=title K2Node_FunctionEntry}
   n1 = GetOfferInfo!(self:Athena Item Shop Offer Details Widget Base Object Reference) -> (ReturnValue:Fort Store Front Offer Info Object Reference)
   n2 = GetPriceItem!(self:Fort Store Front Offer Info Object Reference=@n1, PriceIndex:int=0) -> (RequiredItemCount:int=0, ReturnValue:Fort Account Item Definition Object Reference)
   > n3 = Is Valid(InputObject=@n2.ReturnValue)
   n4 = Get PriceContainer() -> (PriceContainer:Widget Object Reference)
   SetVisibility(execute=@n3.Is Not Valid, self=@n4, InVisibility=Collapsed)
   SetVisibility(execute=@n3.Is Valid, self=@n4, InVisibility=SelfHitTestInvisible)
   > Sequence()
   n5 = Get NumberToPurchase() -> (NumberToPurchase:int=0)
   > "SetText (Text)"(
    self=Get NormalPriceText() -> (NormalPriceText:Text Object Reference)
    InText=GetRegularPrice!(self:Fort Store Front Offer Info Object Reference=@n1, PriceIndex:int=0, ItemQuantity:int=@n5) -> (ReturnValue:Text)
   )
   n6 = GetSalePrice!(self:Fort Store Front Offer Info Object Reference=@n1, PriceIndex:int=0, ItemQuantity:int=@n5) -> (SalePrice:Text, ReturnValue:bool)
   > SetVisibility(
    self=Get SalePriceContainer() -> (SalePriceContainer:Widget Object Reference)
    InVisibility=Select(Option 0=Collapsed, Option 1=SelfHitTestInvisible, Index=@n6.ReturnValue)
   )
   > "SetText (Text)"(self=Get SalePriceText() -> (SalePriceText:Text Object Reference), InText=@n6)
   n7 = Get CurrencyImage() -> (CurrencyImage:Image Object Reference)
   > SetVisibility(self=@n7, InVisibility=SelfHitTestInvisible)
   > SetBrush(
    self=@n7
    InBrush=GetSmallPreviewImageBrush!(self:Fort Item Definition Object Reference=@n2.ReturnValue) -> (ReturnValue:Slate Brush Structure)
   )
  ]}
 UpdateCatalogOfferName {nodes [
   UpdateCatalogOfferName() {is=title K2Node_FunctionEntry}
   n1 = GetOfferInfo()
   > "SetText (Text)"(
    self=Get OfferName() -> (OfferName:Common Text Object Reference)
    InText=TextToUpper!(InText:"Text (by ref)"=GetName!(self:Fort Store Front Offer Info Object Reference=@n1) -> (ReturnValue:Text)) -> (ReturnValue:Text)
   )
  ]}
 UpdateFromOffer {nodes [
   UpdateFromOffer() {is=title K2Node_FunctionEntry}
   > n1 = Sequence!(execute) -> (then_0, then_1, then_2:Exec)
   SetupPrice!(
    execute=@n1.then_1
    self
    PriceContainer:Widget Object Reference=Get CurrencyVB() -> (CurrencyVB:Vertical Box Object Reference)
    NormalPriceText:Text Object Reference=Get TextOriginalPrice1() -> (TextOriginalPrice1:Common Text Object Reference)
    SalePriceContainer:Widget Object Reference=Get O_Sale_PreviousPrice() -> (O_Sale_PreviousPrice:Overlay Object Reference)
    SalePriceText:Text Object Reference=Get TextOfferDisplayPrice1() -> (TextOfferDisplayPrice1:Common Text Object Reference)
    CurrencyImage:Image Object Reference=Get Currency1ItemImage2() -> (Currency1ItemImage2:Image Object Reference)
   ) -> (then) {is=title K2Node_CallFunction}
   n2 = IsRefundable!(self:Fort Store Front Offer Info Object Reference=GetOfferInfo()) -> (ReturnValue:bool)
   n3 = GetSubGame!(self:Fort Global UIContext Object Reference=Get FortGlobalUIContext!() -> (ReturnValue:Fort Global UIContext Object Reference) {is=title K2Node_GetBlueprintContextEx}) -> (ReturnValue:ESubGame Enum=Campaign)
   n4 = Branch(execute=@n1.then_2, Condition=AND Boolean(A=NOT Boolean(A=@n2), B="Equal (Enum)"(A:ESubGame Enum=@n3, B:ESubGame Enum=Athena)))
   n5 = Get NonrefundableText() -> (NonrefundableText:Athena Item Shop Refund Disclaimer Object Reference)
   > SetVisibility(self=@n5, InVisibility=SelfHitTestInvisible)
   SetVisibility(execute=@n4.else, self=@n5, InVisibility=Hidden)
   UpdateCatalogOfferName(execute=@n1.then_0) {is=title K2Node_CallFunction}
   > UpdateItemList()
  ]}
 UpdateItemList {nodes [
   UpdateItemList() {is=title K2Node_FunctionEntry}
   > ClearChildren(self=Get ItemRewardsSB() -> (ItemRewardsSB:Scroll Box Object Reference))
   > n1 = For Each Loop(Array:Array of Fort Account Item Definition Object References=GetGrantedItemDefinitions!(self:Fort Store Front Offer Info Object Reference=GetOfferInfo()) -> (GrantedItems:Array of Fort Account Item Definition Object References)) -> (Array Element:Fort Account Item Definition Object Reference)
   > Set GrantedItem(GrantedItem:Fort Account Item Definition Object Reference=@n1) -> (Output_Get:Fort Account Item Definition Object Reference)
   > n2 = Create Widget!(
    execute:Exec
    Class:User Widget Class Reference=/FortniteCoreUI/Frontend/AthenaStoreReward.AthenaStoreReward_C
    OwningPlayer:Player Controller Object Reference=GetOwningPlayer()
    Style:Common Button Style Class Reference=/Game/UI/Frontend/ItemManagement/InventoryPanel/ItemTileButtonStyles/ItemManagementTileButtonStyle-AthenaLocker.ItemManagementTileButtonStyle-AthenaLocker_C
    bApplyAlphaOnDisable:bool=True
    bLocked:bool=False
    bSelectable:bool=False
    bShouldSelectUponReceivingFocus:bool=False
    bInteractableWhenSelected:bool=False
    bToggleable:bool=False
    bTriggerClickedAfterSelection:bool=False
    bDisplayInputActionWhenNotInteractable:bool=True
    bHideInputActionWithKeyboard:bool=False
    bShouldUseFallbackDefaultInputAction:bool=True
    bRequiresHold:bool=False
    HoldData:Common UIHold Data Class Reference
    ClickMethod:EButtonClickMethod Enum=DownAndUp
    InputPriority:int=0
    TriggeringInputAction:Data Table Row Handle Structure
   ) -> (then:Exec, ReturnValue:Athena Store Reward Object Reference)
   > n3 = Set StoreRewardItem(StoreRewardItem:Athena Store Reward Object Reference=@n2) -> (Output_Get:Athena Store Reward Object Reference)
   > SetupRewardItem!(
    execute
    self:Fort Athena Reward Object Reference=@n3
    ItemDefToRepresent:Item Definition Base Object Reference=Get GrantedItem() -> (GrantedItem:Fort Account Item Definition Object Reference)
    Quantity:int=1
    ItemVM:Fort Item VM Object Reference
   ) -> (then)
   > UpdateRewardState!(
    execute
    self:Fort Athena Reward Object Reference=Get StoreRewardItem() -> (StoreRewardItem:Athena Store Reward Object Reference)
    StateToSetup:EFortAthenaRewardState Enum=LevelAchieved
    bEnabled:bool=true
   ) -> (then)
   > SetViewOnHovered!(execute, self:Fort Athena Reward Object Reference=Get StoreRewardItem(), bShouldViewOnHovered:bool=false) -> (then)
   > EnableRarityFlare(self:Athena Store Reward Object Reference=Get StoreRewardItem())
   > n4 = AddChild(self=Get ItemRewardsSB(), Content=Get StoreRewardItem())
   > n5 = Cast To ScrollBoxSlot!(execute, Object=@n4) -> (then, CastFailed, AsScroll Box Slot:Scroll Box Slot Object Reference) {is=title K2Node_DynamicCast}
   > n6 = Set ScrollBoxSlot(ScrollBoxSlot:Scroll Box Slot Object Reference=@n5) -> (Output_Get:Scroll Box Slot Object Reference)
   > SetVerticalAlignment(self:Scroll Box Slot Object Reference=@n6, InVerticalAlignment=VAlign_Center)
   > SetPadding!(
    execute
    self:Scroll Box Slot Object Reference=Get ScrollBoxSlot() -> (ScrollBoxSlot:Scroll Box Slot Object Reference)
    InPadding:Margin Structure=Make Margin(Left=8, Top=0, Right=8, Bottom=0)
   ) -> (then)
  ]}
]
variables [
 CanChangeQuantity {type=boolean}
 CanPurchase {type=boolean}
 CanPurchaseLeft {type=boolean}
 CanPurchaseRight {type=boolean}
 Confirm {}
 Decline {}
 IsPurchasing {type=boolean}
 NumberToPurchase {type=integer, default=1}
]
widgets {
 info {
  inheritedWidgetCount=4
  namedSlotCount=1
  parentClass=/Script/FortniteUI.AthenaItemShopOfferDetailsWidgetBase
  rootWidgetClass=&VerticalBox
  widgetCount=31
 }
 widgets [
  VerticalBox_1 {widgetClassPath=&VerticalBox, widget=VerticalBox_1}
  BackgroundBlur_2 {
   widgetClassPath=&BackgroundBlur
   parent=VerticalBox_1
   widget=BackgroundBlur_2
   properties {blurStrength=16}
   slot {class=VerticalBoxSlot, properties {size {sizeRule=Fill, value=1}}}
  }
  Lightbox {
   widgetClassPath=/Game/UI/Foundation/Lightbox/Lightbox
   bIsVariable=true
   parent=BackgroundBlur_2
   widget=Lightbox
   slot {class=BackgroundBlurSlot, properties {padding=<ltrb 0 0 0 0>}}
  }
  SafeZone_3 {widgetClassPath=&SafeZone, namedSlotHost=Lightbox, widget=SafeZone_3}
  CommonBorderPurchaseContainer {
   widgetClassPath=&CommonBorder
   parent=SafeZone_3
   widget=CommonBorderPurchaseContainer
   properties {padding=<ltrb 0 0 0 0>}
   slot {class=SafeZoneSlot, properties {padding=<ltrb 64 64 64 64>}}
  }
  Overlay_0 {
   widgetClassPath=&Overlay
   parent=CommonBorderPurchaseContainer
   widget=Overlay_0
   slot {class=BorderSlot, properties {padding=<ltrb 0 0 0 0>}}
  }
  VerticalBoxItemDetails {
   widgetClassPath=&VerticalBox
   parent=Overlay_0
   widget=VerticalBoxItemDetails
   slot {class=OverlaySlot, properties {horizontalAlignment=HAlign_Center, verticalAlignment=VAlign_Center}}
  }
  CommonTextBlock_0 {
   widgetClassPath=&CommonTextBlock
   bIsVariable=true
   parent=VerticalBoxItemDetails
   widget=CommonTextBlock_0
   properties {
    justification=Center
    scrollStyle=&TextScrollStyle-Base_C
    style=/FortUILibrary/Styles/Typography/Foundation/TextStyle-Header-L-40pc.TextStyle-Header-L-40pc_C
    text=PURCHASE?
    colorAndOpacity {colorUseRule=UseColor_Specified, specifiedColor=<rgba 1 1 1 0.4>}
    font {
     bForceMonospaced=false
     bMaterialIsStencil=false
     fontMaterial=null
     fontObject=&Burbank
     letterSpacing=0
     monospacedWidth=1
     size=48
     skewAmount=0
     typefaceFontName=Burbank-Black
     outlineSettings {
      bApplyOutlineToDropShadows=false
      bMiteredCorners=false
      bSeparateFillAlpha=false
      outlineMaterial=null
      outlineSize=0
      outlineColor=<rgba 0 0 0 1>
     }
    }
    shadowOffset=<0 0>
   }
   slot {class=VerticalBoxSlot, properties {horizontalAlignment=HAlign_Center, padding=<ltrb 0 0 0 -16>}}
  }
  OfferName {
   widgetClassPath=&CommonTextBlock
   bIsVariable=true
   parent=VerticalBoxItemDetails
   widget=OfferName
   properties {
    justification=Center
    scrollStyle=&TextScrollStyle-Base_C
    style=/FortUILibrary/Styles/Typography/Foundation/TextStyle-Header-L-70pc.TextStyle-Header-L-70pc_C
    text="{CATALOG OFFER NAME}"
    colorAndOpacity {colorUseRule=UseColor_Specified, specifiedColor=<rgba 1 1 1 0.7>}
    font {
     bForceMonospaced=false
     bMaterialIsStencil=false
     fontMaterial=null
     fontObject=&Burbank
     letterSpacing=0
     monospacedWidth=1
     size=48
     skewAmount=0
     typefaceFontName=Burbank-Black
     outlineSettings {
      bApplyOutlineToDropShadows=false
      bMiteredCorners=false
      bSeparateFillAlpha=false
      outlineMaterial=null
      outlineSize=0
      outlineColor=<rgba 0 0 0 1>
     }
    }
    shadowOffset=<0 0>
   }
   slot {class=VerticalBoxSlot, properties {horizontalAlignment=HAlign_Center}}
  }
  ItemRewardsSB {
   widgetClassPath=&ScrollBox
   bIsVariable=true
   parent=VerticalBoxItemDetails
   widget=ItemRewardsSB
   properties {
    clipping=Inherit
    navigation=&self:WidgetTree.ItemRewardsSB.WidgetNavigation_0
    orientation=Orient_Horizontal
    visibility=HitTestInvisible
    widgetStyle {
     barThickness=9
     bottomShadowBrush {
      bIsDynamicallyLoaded=false
      drawAs=NoDrawType
      imageType=FullColor
      mirroring=NoMirror
      resourceName=../../../Engine/Content/Slate/Common/ScrollBoxShadowBottom.png
      resourceObject=null
      tiling=NoTile
      imageSize=<16 8>
      margin=<ltrb 0.5 0 0.5 1>
      outlineSettings {
       bUseBrushTransparency=false
       roundingType=HalfHeightRadius
       width=0
       color {colorUseRule=UseColor_Specified, specifiedColor=<rgba 0 0 0 0>}
       cornerRadii=<0 0 0 0>
      }
      tintColor {colorUseRule=UseColor_Specified, specifiedColor=<rgba 1 1 1 1>}
      uVRegion {bIsValid=false, max=<0 0>, min=<0 0>}
     }
     horizontalScrolledContentPadding=<ltrb 0 0 1 0>
     leftShadowBrush {
      bIsDynamicallyLoaded=false
      drawAs=NoDrawType
      imageType=FullColor
      mirroring=NoMirror
      resourceName=../../../Engine/Content/Slate/Common/ScrollBoxShadowLeft.png
      resourceObject=null
      tiling=NoTile
      imageSize=<8 16>
      margin=<ltrb 1 0.5 0 0.5>
      outlineSettings {
       bUseBrushTransparency=false
       roundingType=HalfHeightRadius
       width=0
       color {colorUseRule=UseColor_Specified, specifiedColor=<rgba 0 0 0 0>}
       cornerRadii=<0 0 0 0>
      }
      tintColor {colorUseRule=UseColor_Specified, specifiedColor=<rgba 1 1 1 1>}
      uVRegion {bIsValid=false, max=<0 0>, min=<0 0>}
     }
     rightShadowBrush {
      bIsDynamicallyLoaded=false
      drawAs=NoDrawType
      imageType=FullColor
      mirroring=NoMirror
      resourceName=../../../Engine/Content/Slate/Common/ScrollBoxShadowRight.png
      resourceObject=null
      tiling=NoTile
      imageSize=<8 16>
      margin=<ltrb 0 0.5 1 0.5>
      outlineSettings {
       bUseBrushTransparency=false
       roundingType=HalfHeightRadius
       width=0
       color {colorUseRule=UseColor_Specified, specifiedColor=<rgba 0 0 0 0>}
       cornerRadii=<0 0 0 0>
      }
      tintColor {colorUseRule=UseColor_Specified, specifiedColor=<rgba 1 1 1 1>}
      uVRegion {bIsValid=false, max=<0 0>, min=<0 0>}
     }
     topShadowBrush {
      bIsDynamicallyLoaded=false
      drawAs=NoDrawType
      imageType=FullColor
      mirroring=NoMirror
      resourceName=../../../Engine/Content/Slate/Common/ScrollBoxShadowTop.png
      resourceObject=null
      tiling=NoTile
      imageSize=<16 8>
      margin=<ltrb 0.5 1 0.5 0>
      outlineSettings {
       bUseBrushTransparency=false
       roundingType=HalfHeightRadius
       width=0
       color {colorUseRule=UseColor_Specified, specifiedColor=<rgba 0 0 0 0>}
       cornerRadii=<0 0 0 0>
      }
      tintColor {colorUseRule=UseColor_Specified, specifiedColor=<rgba 1 1 1 1>}
      uVRegion {bIsValid=false, max=<0 0>, min=<0 0>}
     }
     verticalScrolledContentPadding=<ltrb 0 0 0 1>
    }
   }
   slot {class=VerticalBoxSlot, properties {horizontalAlignment=HAlign_Center, verticalAlignment=VAlign_Top, padding=<ltrb 16 32 16 48>}}
  }
  SizeBox_2 {
   widgetClassPath=&SizeBox
   parent=VerticalBoxItemDetails
   widget=SizeBox_2
   properties {bOverride_MinDesiredWidth=true, minDesiredWidth=576}
   slot {class=VerticalBoxSlot, properties {horizontalAlignment=HAlign_Center}}
  }
  VerticalBoxPurchaseDetails {widgetClassPath=&VerticalBox, parent=SizeBox_2, widget=VerticalBoxPurchaseDetails, slot {class=SizeBoxSlot}}
  CurrencyVB {
   widgetClassPath=&VerticalBox
   bIsVariable=true
   parent=VerticalBoxPurchaseDetails
   widget=CurrencyVB
   slot {class=VerticalBoxSlot, properties {padding=<ltrb 0 0 0 16>}}
  }
  VerticalBox_5 {
   widgetClassPath=&VerticalBox
   parent=CurrencyVB
   widget=VerticalBox_5
   slot {class=VerticalBoxSlot, properties {horizontalAlignment=HAlign_Center, padding=<ltrb 0 0 0 16>}}
  }
  HorizontalBox_8 {
   widgetClassPath=&HorizontalBox
   parent=VerticalBox_5
   widget=HorizontalBox_8
   slot {class=VerticalBoxSlot, properties {horizontalAlignment=HAlign_Center}}
  }
  SizeBox_5 {
   widgetClassPath=&SizeBox
   parent=HorizontalBox_8
   widget=SizeBox_5
   properties {bOverride_HeightOverride=true, bOverride_WidthOverride=true, heightOverride=40, widthOverride=40}
   slot {class=HorizontalBoxSlot, properties {horizontalAlignment=HAlign_Left, verticalAlignment=VAlign_Center, padding=<ltrb 0 0 8 0>}}
  }
  Currency1ItemImage2 {
   widgetClassPath=&Image
   bIsVariable=true
   parent=SizeBox_5
   widget=Currency1ItemImage2
   properties {visibility=SelfHitTestInvisible}
   slot {class=SizeBoxSlot}
  }
  TextOfferDisplayPrice1 {
   widgetClassPath=&CommonTextBlock
   bIsVariable=true
   parent=HorizontalBox_8
   widget=TextOfferDisplayPrice1
   properties {
    justification=Center
    style=/FortUILibrary/Styles/Typography/Foundation/TextStyle-Base-L-B.TextStyle-Base-L-B_C
    text="1,200"
    font {
     bForceMonospaced=false
     bMaterialIsStencil=false
     fontMaterial=null
     fontObject=&NotoSans
     letterSpacing=0
     monospacedWidth=1
     size=32
     skewAmount=0
     typefaceFontName=NotoSans-Bold
     outlineSettings {
      bApplyOutlineToDropShadows=false
      bMiteredCorners=false
      bSeparateFillAlpha=false
      outlineMaterial=null
      outlineSize=0
      outlineColor=<rgba 0 0 0 1>
     }
    }
    shadowOffset=<0 0>
   }
   slot {class=HorizontalBoxSlot, properties {size {sizeRule=Fill, value=1}}}
  }
  O_Sale_PreviousPrice {
   widgetClassPath=&Overlay
   bIsVariable=true
   parent=HorizontalBox_8
   widget=O_Sale_PreviousPrice
   properties {visibility=Collapsed}
   slot {class=HorizontalBoxSlot, properties {horizontalAlignment=HAlign_Right, verticalAlignment=VAlign_Center}}
  }
  TextOriginalPrice1 {
   widgetClassPath=&CommonTextBlock
   bIsVariable=true
   parent=O_Sale_PreviousPrice
   widget=TextOriginalPrice1
   properties {
    justification=Center
    style=/FortUILibrary/Styles/Typography/Foundation/TextStyle-Base-M-B-70pc.TextStyle-Base-M-B-70pc_C
    text="1,500"
    colorAndOpacity {colorUseRule=UseColor_Specified, specifiedColor=<rgba 1 1 1 0.7>}
    font {
     bForceMonospaced=false
     bMaterialIsStencil=false
     fontMaterial=null
     fontObject=&NotoSans
     letterSpacing=0
     monospacedWidth=1
     size=24
     skewAmount=0
     typefaceFontName=NotoSans-Bold
     outlineSettings {
      bApplyOutlineToDropShadows=false
      bMiteredCorners=false
      bSeparateFillAlpha=false
      outlineMaterial=null
      outlineSize=0
      outlineColor=<rgba 0 0 0 1>
     }
    }
    shadowOffset=<0 0>
   }
   slot {class=OverlaySlot, properties {horizontalAlignment=HAlign_Right, verticalAlignment=VAlign_Fill, padding=<ltrb 12 0 12 0>}}
  }
  I_StrikeThrough {
   widgetClassPath=&Image
   parent=O_Sale_PreviousPrice
   widget=I_StrikeThrough
   properties {
    visibility=SelfHitTestInvisible
    brush {
     bIsDynamicallyLoaded=false
     drawAs=Image
     imageType=NoImage
     mirroring=NoMirror
     resourceName=None
     resourceObject=/Game/UI/Foundation/Materials/M_UI_StrikeThrough
     tiling=NoTile
     imageSize=<32 4>
     margin=<ltrb 0 0 0 0>
     outlineSettings {
      bUseBrushTransparency=false
      roundingType=HalfHeightRadius
      width=0
      color {colorUseRule=UseColor_Specified, specifiedColor=<rgba 0 0 0 0>}
      cornerRadii=<0 0 0 0>
     }
     tintColor {colorUseRule=UseColor_Specified, specifiedColor=<rgba 1 1 1 1>}
     uVRegion {bIsValid=false, max=<0 0>, min=<0 0>}
    }
    colorAndOpacity=<rgba 1 1 1 0.7>
   }
   slot {class=OverlaySlot, properties {horizontalAlignment=HAlign_Fill, verticalAlignment=VAlign_Fill}}
  }
  VerticalBox_2 {
   widgetClassPath=&VerticalBox
   parent=CurrencyVB
   widget=VerticalBox_2
   slot {class=VerticalBoxSlot, properties {horizontalAlignment=HAlign_Center}}
  }
  PurchaseButton {
   widgetClassPath=&IconTextButton
   bIsVariable=true
   parent=VerticalBox_2
   widget=PurchaseButton
   properties {
    bDisplayInputActionWhenNotInteractable=false
    buttonText=CONFIRM
    contentAlignment=HAlign_Center
    minWidth=384
    navigation=&self:WidgetTree.PurchaseButton.WidgetNavigation_0
    onButtonBaseClicked=(&self:WidgetTree.PurchaseButton.HandleOnButtonClicked)
    onButtonBaseDoubleClicked=(&self:WidgetTree.PurchaseButton.HandleOnButtonDoubleClicked)
    onButtonBaseHovered=(&self:WidgetTree.PurchaseButton.HandleOnButtonHovered)
    onButtonBaseUnhovered=(&self:WidgetTree.PurchaseButton.HandleOnButtonUnhovered)
    onSelectedChangedBase=(&self:WidgetTree.PurchaseButton.HandleOnSelectedChanged)
    style=/FortUILibrary/Foundation/Buttons/ButtonStyles/ButtonStyle_Feature_L_Yellow.ButtonStyle_Feature_L_Yellow_C
    hoveredSlateSoundOverride {resourceObject=/Game/Sounds/Fort_UI_Sounds/FrontEnd/ItemShop/Fort_UI_Shop_Menu_Hover_Cue}
    pressedSlateSoundOverride {resourceObject=/Game/Sounds/Fort_UI_Sounds/FrontEnd/ItemShop/Fort_UI_Shop_Menu_Click_Purchase_Cue}
    triggeringInputAction {dataTable=&InputActionDataTable, rowName=GenericForward}
   }
   slot {class=VerticalBoxSlot, properties {padding=<ltrb 0 0 0 16>}}
  }
  CancelButton {
   widgetClassPath=&IconTextButton
   bIsVariable=true
   parent=VerticalBox_2
   widget=CancelButton
   properties {
    buttonText=BACK
    contentAlignment=HAlign_Center
    onButtonBaseClicked=(&self:WidgetTree.CancelButton.HandleOnButtonClicked)
    onButtonBaseDoubleClicked=(&self:WidgetTree.CancelButton.HandleOnButtonDoubleClicked)
    onButtonBaseHovered=(&self:WidgetTree.CancelButton.HandleOnButtonHovered)
    onButtonBaseUnhovered=(&self:WidgetTree.CancelButton.HandleOnButtonUnhovered)
    onSelectedChangedBase=(&self:WidgetTree.CancelButton.HandleOnSelectedChanged)
    style=/FortUILibrary/Foundation/Buttons/ButtonStyles/TopBar/ButtonStyle-Outline-FadedBlue.ButtonStyle-Outline-FadedBlue_C
    hoveredSlateSoundOverride {resourceObject=&Fort_UI_Shop_Menu_Hover_Cue}
    pressedSlateSoundOverride {resourceObject=/Game/Sounds/Fort_UI_Sounds/FrontEnd/ItemShop/Fort_UI_Shop_Menu_Click_Simple_B_Cue}
    triggeringInputAction {dataTable=&InputActionDataTable, rowName=GenericBack}
   }
   slot {class=VerticalBoxSlot}
  }
  AthenaDirectAcquisitonDetails_BattlePassStarDetails {
   widgetClassPath=&/BattlePassStarDetails
   bIsVariable=true
   parent=VerticalBoxPurchaseDetails
   widget=AthenaDirectAcquisitonDetails_BattlePassStarDetails
   properties {visibility=Collapsed}
   slot {class=VerticalBoxSlot, properties {padding=<ltrb 0 0 0 8>}}
  }
  NonrefundableText {
   widgetClassPath=&/AthenaItemShop_RefundDisclaimer
   bIsVariable=true
   parent=VerticalBoxPurchaseDetails
   widget=NonrefundableText
   slot {class=VerticalBoxSlot, properties {padding=<ltrb 0 0 0 8>}}
  }
  MTXButton {
   widgetClassPath=/Game/UI/Frontend/TopBar/MTXButton
   bIsVariable=true
   parent=Overlay_0
   widget=MTXButton
   properties {
    onButtonBaseClicked=(&self:WidgetTree.MTXButton.HandleOnButtonClicked)
    onButtonBaseDoubleClicked=(&self:WidgetTree.MTXButton.HandleOnButtonDoubleClicked)
    onButtonBaseHovered=(&self:WidgetTree.MTXButton.HandleOnButtonHovered)
    onButtonBaseUnhovered=(&self:WidgetTree.MTXButton.HandleOnButtonUnhovered)
    onSelectedChangedBase=(&self:WidgetTree.MTXButton.HandleOnSelectedChanged)
    visibility=HitTestInvisible
   }
   slot {class=OverlaySlot, properties {horizontalAlignment=HAlign_Right}}
  }
  Text_CosmeticWarning {widgetClassPath=&CommonTextBlock, bInherited=true, widget=null}
  TextUndoAvailability {widgetClassPath=&CommonTextBlock, bInherited=true, widget=null}
  Text_UndoBottomBarWarning {widgetClassPath=&CommonTextBlock, bInherited=true, widget=null}
  Button_Undo {widgetClassPath=&CommonButtonBase, bInherited=true, widget=null}
 ]
}
