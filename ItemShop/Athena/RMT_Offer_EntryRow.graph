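# Item Shop offer entry row widget: a small skewed image beside an auto-wrapping rich text line, with placeholder text for rarity, item name and item type.
# Update Color and Text stores the text and color, then calls SetColorAndText, which also runs on Construct.
# That function tints the image and the text's default color with the same color and sets the text.
# The text block uses a cosmetic-gate rich text decorator.
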WidgetBlueprint /ItemShop/Athena/RMT_Offer_EntryRow
graphs [
 EventGraph {nodes [
   Event Construct()
   > SetColorAndText()
  ]}
 SetColorAndText {nodes [
   SetColorAndText() {is=title K2Node_FunctionEntry}
   n1 = Get Linear color() -> (Linear color:Linear Color Structure)
   > SetColorAndOpacity(self:Image Object Reference=Get Image_27() -> (Image_27:Image Object Reference), InColorAndOpacity:Linear Color Structure=@n1)
   n2 = Get CommonRichTextBlock_96() -> (CommonRichTextBlock_96:Common Rich Text Block Object Reference)
   > SetDefaultColorAndOpacity!(execute, self:Rich Text Block Object Reference=@n2, InColorAndOpacity:Slate Color Structure=Make Slate Color(SpecifiedColor=@n1)) -> (then)
   > SetText(self=@n2, InText=Get text() -> (text:Text))
  ]}
 Update Color and Text {nodes [
   n1 = Update Color and Text!() -> (then, text:Text, Linear color:Linear Color Structure) {is=title K2Node_FunctionEntry}
   > Set text(text:Text=@n1) -> (Output_Get:Text)
   > Set Linear color(Linear color:Linear Color Structure=@n1.Linear color) -> (Output_Get:Linear Color Structure)
   > SetColorAndText() {is=title K2Node_CallFunction}
  ]}
]
variables [
 Linear color {type=LinearColor, default=<rgba 0 0 0 0>}
 text {type=string, default=""}
]
widgets {
 info {inheritedWidgetCount=0, parentClass=&UserWidget, rootWidgetClass=&HorizontalBox, widgetCount=3}
 widgets [
  HorizontalBox_14 {widgetClassPath=&HorizontalBox, widget=HorizontalBox_14}
  Image_27 {
   widgetClassPath=&Image
   bIsVariable=true
   parent=HorizontalBox_14
   widget=Image_27
   properties {brush {
     bIsDynamicallyLoaded=false
     drawAs=Image
     imageType=NoImage
     mirroring=NoMirror
     resourceName=None
     resourceObject=/Game/Athena/UI/ReplayBrowser/M_Ui_Skewed_Vfor9slices_Inst_Inst
     tiling=NoTile
     imageSize=<16 16>
     margin=<ltrb 0 0 0 0>
     outlineSettings {
      bUseBrushTransparency=false
      roundingType=HalfHeightRadius
      width=0
      color {colorUseRule=UseColor_Specified, specifiedColor=<rgba 0 0 0 0>}
      cornerRadii=<0 0 0 0>
     }
     tintColor {colorUseRule=UseColor_Specified, specifiedColor=<rgba 1 1 1 1>}
     uVRegion {bIsValid=false, max=<0 0>, min=<0 0>}
    }}
   slot {class=HorizontalBoxSlot, properties {horizontalAlignment=HAlign_Left, verticalAlignment=VAlign_Top, padding=<ltrb 0 8 8 0>}}
  }
  CommonRichTextBlock_96 {
   widgetClassPath=&CommonRichTextBlock
   bIsVariable=true
   parent=HorizontalBox_14
   widget=CommonRichTextBlock_96
   properties {
    autoWrapText=true
    bOverrideDefaultStyle=true
    lineHeightPercentage=0.9
    text="Rarity<text color=\"FFFFFFFF\" fontface=\"black\"> item name</> Item type <img id=\"WarningIcon\"/>"
    decoratorClasses [/Game/Athena/UI/Frontend/CosmeticGating/Data/RichTextDecorator_CosmeticGate.RichTextDecorator_CosmeticGate_C]
    defaultTextStyleOverride {
     overflowPolicy=Clip
     transformPolicy=None
     colorAndOpacity {colorUseRule=UseColor_Specified, specifiedColor=<rgba 1 1 1 0.75>}
     font {
      bForceMonospaced=false
      bMaterialIsStencil=false
      fontMaterial=null
      fontObject=&BurbankSmall
      letterSpacing=0
      monospacedWidth=1
      size=16
      skewAmount=0
      typefaceFontName=Black
      outlineSettings {
       bApplyOutlineToDropShadows=false
       bMiteredCorners=false
       bSeparateFillAlpha=false
       outlineMaterial=null
       outlineSize=0
       outlineColor=<rgba 0 0 0 1>
      }
     }
     highlightColor {colorUseRule=UseColor_Specified, specifiedColor=<rgba 0 0 0 1>}
     highlightShape {
      bIsDynamicallyLoaded=false
      drawAs=Image
      imageType=NoImage
      mirroring=NoMirror
      resourceName=None
      resourceObject=null
      tiling=NoTile
      imageSize=<32 32>
      margin=<ltrb 0 0 0 0>
      outlineSettings {
       bUseBrushTransparency=false
       roundingType=HalfHeightRadius
       width=0
       color {colorUseRule=UseColor_Specified, specifiedColor=<rgba 0 0 0 0>}
       cornerRadii=<0 0 0 0>
      }
      tintColor {colorUseRule=UseColor_Specified, specifiedColor=<rgba 1 1 1 1>}
      uVRegion {bIsValid=false, max=<0 0>, min=<0 0>}
     }
     selectedBackgroundColor {colorUseRule=UseColor_Foreground, specifiedColor=<rgba 1 0 1 1>}
     shadowColorAndOpacity=<rgba 0 0 0 1>
     shadowOffset=<0 0>
     strikeBrush {
      bIsDynamicallyLoaded=false
      drawAs=Image
      imageType=NoImage
      mirroring=NoMirror
      resourceName=None
      resourceObject=null
      tiling=NoTile
      imageSize=<32 32>
      margin=<ltrb 0 0 0 0>
      outlineSettings {
       bUseBrushTransparency=false
       roundingType=HalfHeightRadius
       width=0
       color {colorUseRule=UseColor_Specified, specifiedColor=<rgba 0 0 0 0>}
       cornerRadii=<0 0 0 0>
      }
      tintColor {colorUseRule=UseColor_Specified, specifiedColor=<rgba 1 1 1 1>}
      uVRegion {bIsValid=false, max=<0 0>, min=<0 0>}
     }
     underlineBrush {
      bIsDynamicallyLoaded=false
      drawAs=Image
      imageType=NoImage
      mirroring=NoMirror
      resourceName=None
      resourceObject=null
      tiling=NoTile
      imageSize=<32 32>
      margin=<ltrb 0 0 0 0>
      outlineSettings {
       bUseBrushTransparency=false
       roundingType=HalfHeightRadius
       width=0
       color {colorUseRule=UseColor_Specified, specifiedColor=<rgba 0 0 0 0>}
       cornerRadii=<0 0 0 0>
      }
      tintColor {colorUseRule=UseColor_Specified, specifiedColor=<rgba 1 1 1 1>}
      uVRegion {bIsValid=false, max=<0 0>, min=<0 0>}
     }
    }
    margin=<ltrb 0 0 0 8>
   }
   slot {class=HorizontalBoxSlot, properties {verticalAlignment=VAlign_Top, size {sizeRule=Fill, value=1}}}
  }
 ]
}
