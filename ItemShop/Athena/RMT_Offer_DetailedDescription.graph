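# Item Shop widget that shows an offer's detailed description: a short description rich text block above a vertical list of detail attribute rows.
# SetupDataDetails clears the entry box, sets the description text and secondary color from the MTX details struct, then creates an RMT_Offer_EntryRow per attribute in the tertiary color.
# When content gating is disabled, it strips the inline WarningIcon image tag from each attribute name.
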WidgetBlueprint /ItemShop/Athena/RMT_Offer_DetailedDescription
graphs [
 EventGraph {links [], nodes []}
 SetupDataDetails {nodes [
   SetupDataDetails!() -> (
    then
    FortMtxDescriptionAndDetails:"Fort Mtx Description And Details Structure (by ref)"
    SecondaryColor:Linear Color Structure
    TertiaryColor:Linear Color Structure
   ) {is=title K2Node_FunctionEntry}
   > Reset!(
    execute
    self:Dynamic Entry Box Object Reference=Get DynamicEntryBox_Items() -> (DynamicEntryBox_Items:Dynamic Entry Box Object Reference)
    bDeleteWidgets:bool=true
   ) -> (then)
   n1 = Break Fort Mtx Description and Details!(FortMtxDescriptionAndDetails:"Fort Mtx Description And Details Structure (by ref)"=Get FortMtxDescriptionAndDetails() -> (FortMtxDescriptionAndDetails:"Fort Mtx Description And Details Structure (by ref)")) -> (ShortDescription:Text, AssociatedDetailsAttributes:Array of Fort Mtx Details Attribute Structures) {is=title K2Node_BreakStruct}
   n2 = Get CommonRichTextBlock_Desc() -> (CommonRichTextBlock_Desc:Common Rich Text Block Object Reference)
   > SetText(self=@n2, InText=@n1)
   > SetDefaultColorAndOpacity!(
    execute
    self:Rich Text Block Object Reference=@n2
    InColorAndOpacity_SpecifiedColor:Linear Color Structure=Get SecondaryColor() -> (SecondaryColor:Linear Color Structure)
    InColorAndOpacity_ColorUseRule:ESlateColorStylingMode Enum
   ) -> (then)
   > n3 = For Each Loop(Array:Array of Fort Mtx Details Attribute Structures=@n1.AssociatedDetailsAttributes) -> (Array Element:Fort Mtx Details Attribute Structure)
   > n4 = Create Entry!(execute, self:Dynamic Entry Box Object Reference=Get DynamicEntryBox_Items()) -> (then, ReturnValue:User Widget Object Reference)
   n5 = Break Fort Mtx Details Attribute!(FortMtxDetailsAttribute:"Fort Mtx Details Attribute Structure (by ref)"=@n3) -> (Name:Text) {is=title K2Node_BreakStruct}
   n6 = Replace(SourceString="To String (Text)"(InText=@n5), From="<img id=\"WarningIcon\"/>")
   > Update Color and Text!(
    execute
    self:RMT Offer Entry Row Object Reference=Cast To RMT_Offer_EntryRow!(Object=@n4) -> (AsRMT Offer Entry Row:RMT Offer Entry Row Object Reference, bSuccess:bool=~) {is=title K2Node_DynamicCast}
    text:Text=Select(Option 0:Text="To Text (String)"(InString=@n6), Option 1:Text=@n5, Index=IsContentGatingEnabled!() -> (ReturnValue:bool)) -> (ReturnValue:Text)
    Linear color:Linear Color Structure=Get TertiaryColor() -> (TertiaryColor:Linear Color Structure)
   ) -> (then)
  ]}
]
variables []
widgets {
 info {inheritedWidgetCount=0, parentClass=&UserWidget, rootWidgetClass=&VerticalBox, widgetCount=3}
 widgets [
  VerticalBox_0 {widgetClassPath=&VerticalBox, widget=VerticalBox_0}
  CommonRichTextBlock_Desc {
   widgetClassPath=&CommonRichTextBlock
   bIsVariable=true
   parent=VerticalBox_0
   widget=CommonRichTextBlock_Desc
   properties {
    autoWrapText=true
    bOverrideDefaultStyle=true
    lineHeightPercentage=0.9
    text="<text color=\"FFFFFFFF\" fontface=\"black\" size=\"24\" Case=\"AllCaps\">Short description</>\r\nInclusion text"
    defaultTextStyleOverride {
     overflowPolicy=Clip
     transformPolicy=None
     colorAndOpacity {colorUseRule=UseColor_Specified, specifiedColor=<rgba 0.9948 0.9952 1 0.75>}
     font {
      bForceMonospaced=false
      bMaterialIsStencil=false
      fontMaterial=null
      fontObject=&BurbankSmall
      letterSpacing=0
      monospacedWidth=1
      size=16
      skewAmount=0
      typefaceFontName=Bold
      outlineSettings {
       bApplyOutlineToDropShadows=false
       bMiteredCorners=false
       bSeparateFillAlpha=false
       outlineMaterial=null
       outlineSize=0
       outlineColor=<rgba 0 0 0 1>
      }
     }
     highlightColor {colorUseRule=UseColor_Specified, specifiedColor=<rgba 0 0 0 1>}
     highlightShape {
      bIsDynamicallyLoaded=false
      drawAs=Image
      imageType=NoImage
      mirroring=NoMirror
      resourceName=None
      resourceObject=null
      tiling=NoTile
      imageSize=<32 32>
      margin=<ltrb 0 0 0 0>
      outlineSettings {
       bUseBrushTransparency=false
       roundingType=HalfHeightRadius
       width=0
       color {colorUseRule=UseColor_Specified, specifiedColor=<rgba 0 0 0 0>}
       cornerRadii=<0 0 0 0>
      }
      tintColor {colorUseRule=UseColor_Specified, specifiedColor=<rgba 1 1 1 1>}
      uVRegion {bIsValid=false, max=<0 0>, min=<0 0>}
     }
     selectedBackgroundColor {colorUseRule=UseColor_Foreground, specifiedColor=<rgba 1 0 1 1>}
     shadowColorAndOpacity=<rgba 0 0 0 1>
     shadowOffset=<0 0>
     strikeBrush {
      bIsDynamicallyLoaded=false
      drawAs=Image
      imageType=NoImage
      mirroring=NoMirror
      resourceName=None
      resourceObject=null
      tiling=NoTile
      imageSize=<32 32>
      margin=<ltrb 0 0 0 0>
      outlineSettings {
       bUseBrushTransparency=false
       roundingType=HalfHeightRadius
       width=0
       color {colorUseRule=UseColor_Specified, specifiedColor=<rgba 0 0 0 0>}
       cornerRadii=<0 0 0 0>
      }
      tintColor {colorUseRule=UseColor_Specified, specifiedColor=<rgba 1 1 1 1>}
      uVRegion {bIsValid=false, max=<0 0>, min=<0 0>}
     }
     underlineBrush {
      bIsDynamicallyLoaded=false
      drawAs=Image
      imageType=NoImage
      mirroring=NoMirror
      resourceName=None
      resourceObject=null
      tiling=NoTile
      imageSize=<32 32>
      margin=<ltrb 0 0 0 0>
      outlineSettings {
       bUseBrushTransparency=false
       roundingType=HalfHeightRadius
       width=0
       color {colorUseRule=UseColor_Specified, specifiedColor=<rgba 0 0 0 0>}
       cornerRadii=<0 0 0 0>
      }
      tintColor {colorUseRule=UseColor_Specified, specifiedColor=<rgba 1 1 1 1>}
      uVRegion {bIsValid=false, max=<0 0>, min=<0 0>}
     }
    }
   }
   slot {class=VerticalBoxSlot, properties {padding=<ltrb 10 0 10 16>}}
  }
  DynamicEntryBox_Items {
   widgetClassPath=&DynamicEntryBox
   bIsVariable=true
   parent=VerticalBox_0
   widget=DynamicEntryBox_Items
   properties {entryBoxType=Vertical, entryWidgetClass=&/RMT_Offer_EntryRow.RMT_Offer_EntryRow_C}
   slot {class=VerticalBoxSlot}
  }
 ]
}
