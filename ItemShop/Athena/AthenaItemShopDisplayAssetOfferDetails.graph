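# Offer details screen widget for the Athena Item Shop, showing an offer's key art, colors, included items, V-Bucks amounts and legal text from its display asset.
# Init fills the layout and plays an animated intro, while activation also preloads item previews.
# Purchases can route through cosmetic-gate, PIN or right-of-withdrawal modals, and Inspect async-loads the product detail screen.
# A purchase trace ID tags analytics for real-money offers and is conditionally cleared on deactivation.
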WidgetBlueprint /ItemShop/Athena/AthenaItemShopDisplayAssetOfferDetails
animations [Intro {
  binding_count=8
  length_seconds=1.25
  object=&self:Intro
  sequence {bindings [
    Button_PurchaseB {
     id=#1
     display_name=Button_Purchase
     possessed_class=IconTextButton_C
     tracks [{
       type=MovieScene2DTransformTrack
       sections [{
         type=MovieScene2DTransformSection
         channels [
          Translation.X {type=MovieSceneScriptingFloatChannel}
          Translation.Y {type=MovieSceneScriptingFloatChannel}
          Angle {type=MovieSceneScriptingFloatChannel}
          Scale.X {type=MovieSceneScriptingFloatChannel, keys [[21,0],[23,1.1],[25,1]]}
          Scale.Y {type=MovieSceneScriptingFloatChannel, keys [[21,0],[23,1.1],[25,1]]}
          Shear.X {type=MovieSceneScriptingFloatChannel}
          Shear.Y {type=MovieSceneScriptingFloatChannel}
         ]
        }]
      }]
    }
    VerticalBox_Content {
     id=#2
     possessed_class=VerticalBox
     tracks [{
       type=MovieScene2DTransformTrack
       sections [{
         type=MovieScene2DTransformSection
         channels [
          Translation.X {type=MovieSceneScriptingFloatChannel, keys [[15.0007,-512],[20.0007,8],[22.0007,0]]}
          Translation.Y {type=MovieSceneScriptingFloatChannel}
          Angle {type=MovieSceneScriptingFloatChannel}
          Scale.X {type=MovieSceneScriptingFloatChannel}
          Scale.Y {type=MovieSceneScriptingFloatChannel}
          Shear.X {type=MovieSceneScriptingFloatChannel}
          Shear.Y {type=MovieSceneScriptingFloatChannel}
         ]
        }]
      }]
    }
    SizeBox_SidePannel {
     id=#3
     possessed_class=SizeBox
     tracks [{
       type=MovieScene2DTransformTrack
       sections [{
         type=MovieScene2DTransformSection
         channels [
          Translation.X {type=MovieSceneScriptingFloatChannel, keys [[0,-800],[15.0007,-800],[20.0003,0]]}
          Translation.Y {type=MovieSceneScriptingFloatChannel}
          Angle {type=MovieSceneScriptingFloatChannel}
          Scale.X {type=MovieSceneScriptingFloatChannel}
          Scale.Y {type=MovieSceneScriptingFloatChannel}
          Shear.X {type=MovieSceneScriptingFloatChannel}
          Shear.Y {type=MovieSceneScriptingFloatChannel}
         ]
        }]
      }]
    }
    SpeedLines {
     id=#4
     possessed_class=Image
     tracks [{
       type=MovieSceneWidgetMaterialTrack
       sections [{
         type=MovieSceneParameterSection
         channels [PullBack {type=MovieSceneScriptingFloatChannel, keys [[0,1],[4,0.1],[8,0.2],[15.0003,0.2],[23.0007,1]]}]
        }]
      }]
    }
    Spacer_90 {
     id=#5
     possessed_class=Spacer
     tracks [{
       type=MovieSceneFloatVectorTrack
       display_name=Size
       sections [{
         type=MovieSceneFloatVectorSection
         channels [
          Vector.X {type=MovieSceneScriptingFloatChannel, keys [[0,50],[15.0007,50],[20.0003,650]]}
          Vector.Y {type=MovieSceneScriptingFloatChannel, const [0,1]}
         ]
        }]
      }]
    }
    Button_Inspect {
     id=#6
     possessed_class=IconTextButton_C
     tracks [{
       type=MovieScene2DTransformTrack
       sections [{
         type=MovieScene2DTransformSection
         channels [
          Translation.X {type=MovieSceneScriptingFloatChannel}
          Translation.Y {type=MovieSceneScriptingFloatChannel}
          Angle {type=MovieSceneScriptingFloatChannel}
          Scale.X {type=MovieSceneScriptingFloatChannel, keys [[19,0],[22.0003,1]]}
          Scale.Y {type=MovieSceneScriptingFloatChannel, keys [[19,0],[22.0003,1]]}
          Shear.X {type=MovieSceneScriptingFloatChannel}
          Shear.Y {type=MovieSceneScriptingFloatChannel}
         ]
        }]
      }]
    }
    Button_SendGift {
     id=#7
     possessed_class=IconTextButton_C
     tracks [{
       type=MovieScene2DTransformTrack
       sections [{
         type=MovieScene2DTransformSection
         channels [
          Translation.X {type=MovieSceneScriptingFloatChannel}
          Translation.Y {type=MovieSceneScriptingFloatChannel}
          Angle {type=MovieSceneScriptingFloatChannel}
          Scale.X {type=MovieSceneScriptingFloatChannel, keys [[20.0003,0],[23.0007,1]]}
          Scale.Y {type=MovieSceneScriptingFloatChannel, keys [[20.0003,0],[23.0007,1]]}
          Shear.X {type=MovieSceneScriptingFloatChannel}
          Shear.Y {type=MovieSceneScriptingFloatChannel}
         ]
        }]
      }]
    }
    KeyArt {
     id=#8
     possessed_class=CommonLazyImage
     tracks [
      {
       type=MovieSceneColorTrack
       display_name=Color and Opacity
       sections [{
         type=MovieSceneColorSection
         channels [
          Color.R {type=MovieSceneScriptingFloatChannel}
          Color.G {type=MovieSceneScriptingFloatChannel}
          Color.B {type=MovieSceneScriptingFloatChannel}
          Color.A {type=MovieSceneScriptingFloatChannel, keys [[0,0],[6,1]]}
         ]
        }]
      }
      {
       type=MovieScene2DTransformTrack
       sections [{
         type=MovieScene2DTransformSection
         channels [
          Translation.X {type=MovieSceneScriptingFloatChannel}
          Translation.Y {type=MovieSceneScriptingFloatChannel}
          Angle {type=MovieSceneScriptingFloatChannel}
          Scale.X {type=MovieSceneScriptingFloatChannel, keys [[0,3],[4,0.8],[6,1]]}
          Scale.Y {type=MovieSceneScriptingFloatChannel, keys [[0,3],[4,0.8],[6,1]]}
          Shear.X {type=MovieSceneScriptingFloatChannel}
          Shear.Y {type=MovieSceneScriptingFloatChannel}
         ]
        }]
      }
     ]
    }
   ]}
 }]
graphs [
 BuildAnalyticsInfo {nodes [
   BuildAnalyticsInfo() {is=title K2Node_FunctionEntry}
   n1 = Get PurchaseUXTraceId() -> (PurchaseUXTraceId:String)
   > n2 = Branch(Condition=IsEmpty(InString=@n1))
   n3 = "Add String (Non-Localized)"!(
    Info:"Flow Analytics Additional Info Structure (by ref)"=Make Flow Analytics Additional Info!() -> (FlowAnalyticsAdditionalInfo:Flow Analytics Additional Info Structure) {is=title K2Node_MakeStruct}
    Key:String=Offer
    Value:String=GetDisplayAssetName!(self:Fort Item Shop Offer VM Object Reference=Get OfferVM() -> (OfferVM:Fort Item Shop Offer VM Object Reference)) -> (ReturnValue:String)
   ) -> (ReturnValue:Flow Analytics Additional Info Structure)
   > Return Node!(execute, ReturnValue:Flow Analytics Additional Info Structure=@n3)
   Return Node(execute=@n2.else, ReturnValue="Add String (Non-Localized)"(Info=@n3, Key=PurchaseUXTraceId, Value=@n1))
  ]}
 CanPurchaseMtxPriceOffer {nodes [
   n1 = CanPurchaseMtxPriceOffer!() -> (then, PirceIndex:int) {is=title K2Node_FunctionEntry}
   > n2 = CanPurchaseOffer!(
    execute
    self:Fort Mcp Context Object Reference=Get FortMcpContext!() -> (ReturnValue:Fort Mcp Context Object Reference) {is=title K2Node_GetBlueprintContextEx}
    OfferId:String=GetOfferId!(self:Fort Store Front Offer Info Object Reference=GetOfferInfo!(self:Athena Item Shop Offer Details Widget Base Object Reference) -> (ReturnValue:Fort Store Front Offer Info Object Reference)) -> (ReturnValue:String)
    PriceIndex:int=@n1
    Quantity:int=1
   ) -> (then, ReturnValue:EOfferPurchaseError Enum=NoError)
   > Return Node(ReturnValue:bool="Equal (Enum)"(A:EOfferPurchaseError Enum=@n2, B:EOfferPurchaseError Enum=NoError))
  ]}
 Content Blocking {nodes [
   n1 = Content Blocking!() -> (then, IsContentBlocked:bool=~, ContentBlockedText:Text) {is=title K2Node_FunctionEntry}
   > Set Data!(
    execute
    self:WBP Content Purchase Blocker Object Reference=Get WBP_ContentPurchaseBlocker() -> (WBP_ContentPurchaseBlocker:WBP Content Purchase Blocker Object Reference)
    InText:Text=@n1.ContentBlockedText
   ) -> (then) {is=title K2Node_CallFunction}
   > SetVisibility(self=Get WBP_ContentPurchaseBlocker(), InVisibility=Select(Option 0=Collapsed, Option 1=SelfHitTestInvisible, Index=@n1))
   > SetVisibility(
    self=Get VerticalBox_Content() -> (VerticalBox_Content:Vertical Box Object Reference)
    InVisibility=Select(Option 0=SelfHitTestInvisible, Option 1=Collapsed, Index=@n1)
   )
  ]}
 Custom Navigate Up {nodes [
   n1 = Custom Navigate Up!() -> (then, Navigation:EUINavigation Enum) {is=title K2Node_FunctionEntry}
   > n2 = Switch on EUINavigation!(execute, Selection:EUINavigation Enum=@n1) -> (Left:Exec, Right:Exec, Up:Exec, Down:Exec, Next:Exec, Previous:Exec, Invalid:Exec) {is=title K2Node_SwitchEnum}
   n3 = Get Button_Purchase() -> (Button_Purchase:Fort CTAButton Object Reference)
   n4 = Branch(
    execute=@n2.Up
    Condition=AND Boolean(
     A=IsFocusableButton!(self:Athena Item Shop Offer Details Widget Base Object Reference, InButton:Widget Object Reference=@n3) -> (ReturnValue:bool)
     B=Get bIsEnabled!(self:Widget Object Reference=@n3) -> (bIsEnabled:bool)
    )
   )
   > Return Node(ReturnValue:Widget Object Reference=@n3)
   Return Node(execute=@n4.else)
  ]}
 EventGraph {nodes [
   "On Clicked (Button_Purchase)"!() -> (OutputDelegate, then, Button:Common Button Base Object Reference) {is=title K2Node_ComponentBoundEvent}
   > n1 = Branch(Condition=Get IsRestricted() -> (IsRestricted:bool))
   > n2 = Branch(Condition=Get ShouldShowPinModal() -> (ShouldShowPinModal:bool))
   > RequestPurchasePinModal(self:Athena Item Shop Offer Details Widget Base Object Reference)
   n3 = Create Widget!(
    execute:Exec=@n2.else
    Class:User Widget Class Reference=/FortniteCoreUI/Frontend/CosmeticGating/WBP_CosmeticGate_ShopModal.WBP_CosmeticGate_ShopModal_C
    OwningPlayer:Player Controller Object Reference
    FortItemShopOfferVM:Fort Item Shop Offer VM Object Reference=Get OfferVM()
   ) -> (then:Exec, ReturnValue:WBP Cosmetic Gate Shop Modal Object Reference)
   > Bind Event to On Purchase Clicked(self:WBP Cosmetic Gate Shop Modal Object Reference=@n3, Delegate=Create Event()) {is=title K2Node_AddDelegate}
   > TryPushWidgetAsModalPanel!(execute, PushWidget:User Widget Object Reference=@n3) -> (then)
   AttemptPurchase(execute=@n1.else, self:Athena Item Shop Offer Details Widget Base Object Reference)
   n4 = LoadAndInitImages!() -> (OutputDelegate, then, DisplayAsset:Fort Mtx Offer Data Object Reference) {is=title K2Node_CustomEvent}
   > SetDesiredSizeOverride!(
    execute
    self:Image Object Reference=Get KeyArt() -> (KeyArt:Fort Lazy Image Object Reference)
    DesiredSize:Vector 2D Structure=Get DetailsImageSize!(self:Fort Mtx Offer Data Object Reference=@n4.DisplayAsset) -> (DetailsImageSize:Vector 2D Structure)
   ) -> (then)
   > SetDetailsImage!(execute, self, DisplayAsset:Fort Mtx Offer Data Object Reference=@n4.DisplayAsset) -> (then)
   > n5 = Async Load Asset(Asset=Get DetailsBadge!(self:Fort Mtx Offer Data Object Reference=@n4.DisplayAsset) -> (DetailsBadge:Material Interface Soft Object Reference))
   n6 = Cast To MaterialInterface!(execute=@n5.Completed, Object=@n5) -> (then, CastFailed, AsMaterial Interface:Material Interface Object Reference) {is=title K2Node_DynamicCast}
   n7 = Get Image_Badge() -> (Image_Badge:Image Object Reference)
   > SetVisibility(self=@n7, InVisibility=SelfHitTestInvisible)
   > SetBrushFromMaterial(self=@n7, Material=@n6)
   SetVisibility(execute=@n6.CastFailed, self=@n7, InVisibility=Collapsed)
   OnLoadingError() {is=title K2Node_CustomEvent}
   > n8 = Branch(Condition=Get bIsActive() -> (bIsActive:bool))
   n9 = OnLoadingErrorButtonClicked!() -> (OutputDelegate, then, ButtonIndex:int) {is=title K2Node_CustomEvent}
   n10 = Push Simple UIKit Modal Dialog!(
    execute=@n8.then
    Context:User Widget Object Reference=Self-Reference()
    VM Class:UI Kit Dialog Viewmodel Class Reference=/ItemShop/MultiProduct/Data/DialogVM_ShopLoadingError.DialogVM_ShopLoadingError_C
    OnDialogButtonClicked:"Delegate (by ref)"=@n9
   ) -> (then, DialogActivatableWidget:WBP UIKit Dialog Object Reference, DialogVM:UI Kit Dialog Viewmodel Object Reference)
   OnCpcModalRequested() {is=title K2Node_CustomEvent}
   > n11 = Construct Object from Class(Class=/FortniteCoreUI/Frontend/RightOfWithdrawal/DialogVM_RightOfWithdrawal.DialogVM_RightOfWithdrawal_C) -> (ReturnValue:Dialog VM Right Of Withdrawal Object Reference)
   > n12 = Create Widget!(
    execute
    Class=/FortniteCoreUI/Frontend/RightOfWithdrawal/WBP_RightOfWithdrawal_Dialog.WBP_RightOfWithdrawal_Dialog_C
    OwningPlayer=GetOwningPlayer()
    DialogVM_RightOfWithdrawal:Dialog VM Right Of Withdrawal Object Reference=@n11
    BodyText:Text="LOCTABLE(\"/ItemShop/MultiProduct/Data/ST_ItemShop\", \"CPC_DefaultText\")"
   ) -> (then, ReturnValue:WBP Right Of Withdrawal Dialog Object Reference)
   n13 = OnCpcDialogDismissed!() -> (OutputDelegate, then, Accepted:bool=~) {is=title K2Node_CustomEvent}
   Unbind Event from On Dialog Dismissed(execute=@n12.then, self:WBP Right Of Withdrawal Dialog Object Reference=@n12, Delegate=@n13) {is=title K2Node_RemoveDelegate}
   > Bind Event to On Dialog Dismissed(self:WBP Right Of Withdrawal Dialog Object Reference=@n12, Delegate=@n13) {is=title K2Node_AddDelegate}
   > TryPushWidgetAsModalPanel(PushWidget=@n12)
   n14 = OnInspectScreenError() {is=title K2Node_CustomEvent}
   > DeactivateWidget()
   n15 = On2DAssetsLoaded!() -> (OutputDelegate, then, HasLoaded2D:bool=~, Error:"Fort Download Error Structure (by ref)") {is=title K2Node_CustomEvent}
   > n16 = Set LoadState2D(LoadState2D:E_ItemPreview_LoadState Enum=Select(Option 0:E_ItemPreview_LoadState Enum=NewEnumerator1, Option 1:E_ItemPreview_LoadState Enum=NewEnumerator2, Index=@n15.HasLoaded2D) -> (ReturnValue:E_ItemPreview_LoadState Enum)) -> (Output_Get:E_ItemPreview_LoadState Enum=NewEnumerator3)
   > n17 = Switch on E_ItemPreview_LoadState!(execute, Selection:E_ItemPreview_LoadState Enum=@n16) -> (NewEnumerator3:Exec, NewEnumerator0:Exec, NewEnumerator2:Exec, NewEnumerator1:Exec) {is=title K2Node_SwitchEnum}
   n18 = Branch(
    execute=[@n24.then, @n17.NewEnumerator1]
    Condition="Equal (Enum)"(
     A:E_ItemPreview_LoadState Enum=Get LoadState2D() -> (LoadState2D:E_ItemPreview_LoadState Enum=NewEnumerator3)
     B:E_ItemPreview_LoadState Enum=Get LoadState3D() -> (LoadState3D:E_ItemPreview_LoadState Enum=NewEnumerator3)
    )
   )
   OnLoadingError(execute=[@n18.then, @n24.else]) {is=title K2Node_CallFunction}
   n19 = NOT Boolean(A=IsPriceInRealMoney!(self:Fort Store Front Offer Info Object Reference=GetOfferInfo(), PriceIndex:int=0) -> (ReturnValue:bool))
   n20 = Branch(execute=@n17.NewEnumerator2, Condition=@n19)
   SetLoadGuardButtonProgress!(execute=[@n20.then, @n23.NewEnumerator2], self, LoadingProgress:double=1) -> (then)
   n21 = On3DAssetsLoaded!() -> (OutputDelegate, then, HasLoaded3D:bool=~, Error:"Fort Download Error Structure (by ref)") {is=title K2Node_CustomEvent}
   > n22 = Set LoadState3D(LoadState3D:E_ItemPreview_LoadState Enum=Select(Option 0=NewEnumerator1, Option 1=NewEnumerator2, Index=@n21.HasLoaded3D)) -> (Output_Get:E_ItemPreview_LoadState Enum=NewEnumerator3)
   > n23 = Switch on E_ItemPreview_LoadState(Selection=@n22)
   n24 = Branch(execute=@n23.NewEnumerator1, Condition=@n19)
   n25 = SetDetailsImage!() -> (OutputDelegate, then, DisplayAsset:Fort Mtx Offer Data Object Reference) {is=title K2Node_CustomEvent}
   n26 = GetSoftDetailsImageURL!(self:Fort Mtx Offer Data Object Reference=@n25.DisplayAsset, World:World Object Reference, TextureSize:int=1040) -> (ReturnValue:String)
   > n27 = Branch(Condition=NOT Boolean(A=IsEmpty(InString=@n26)))
   > SetBrushFromURL!(
    execute
    self:Fort Lazy Image Object Reference=Get KeyArt()
    CMSImageURL:String=@n26
    bMatchSize:bool=false
    Delay:float=0
    DownloadPriority:ECMSDownloadPriority Enum=Normal
   ) -> (then)
   SetBrushFromLazyDisplayAsset!(
    execute=@n27.else
    self:Common Lazy Image Object Reference=Get KeyArt()
    LazyObject:"Soft Object Reference (by ref)"=Get SoftDetailsImage!(self:Fort Mtx Offer Data Object Reference=@n25.DisplayAsset) -> (SoftDetailsImage:Soft Object Reference)
    bMatchTextureSize:bool=false
   ) -> (then)
   Event OnPurchasingStarted() {is=title K2Node_Event}
   > SetPurchaseButtonText!(execute, self, ButtonText:Text="LOCTEXT(\"PURCHASING...\")") -> (then)
   Event Construct()
   > Set ContentPurchaseFocusWidget(ContentPurchaseFocusWidget:Widget Object Reference=Get WBP_UIKit_Button_Loud!(self:WBP Content Purchase Blocker Object Reference=Get WBP_ContentPurchaseBlocker()) -> (WBP_UIKit_Button_Loud:WBP UIKit Button Loud Object Reference)) -> (Output_Get:Widget Object Reference)
   > n28 = Sequence()
   n29 = GetPlatformName!() -> (ReturnValue:String)
   Branch(
    execute=@n28.then_1
    Condition=OR Boolean!(
     A="Equal Exactly (String)"(A=@n29, B=Switch)
     B="Equal, Case Insensitive (String)"(A=@n29, B=PS4)
     C:bool="Equal, Case Insensitive (String)"(A=@n29, B=PS5)
    ) -> (ReturnValue)
   )
   n30 = Get CommonTextBlock_VbucksLegal() -> (CommonTextBlock_VbucksLegal:Common Text Object Reference)
   > Set Justification!(execute, Justification:ETextJustify Enum=Left, self:Text Layout Widget Object Reference=@n30) -> (then, Output_Get:ETextJustify Enum=Left)
   n31 = SelectInt(A=700, B=650, bPickA="Equal Exactly (String)"(A=GetPlatformName(), B=Switch))
   > SetWrapTextWidth!(execute, self:Common Text Object Reference=@n30, InWrapTextAt:int=@n31) -> (then)
   n32 = Get Text_UndoBottomBarWarning() -> (Text_UndoBottomBarWarning:Common Text Object Reference)
   > Set Justification(Justification=Left, self=@n32) -> (Output_Get=Left)
   > SetWrapTextWidth(self=@n32, InWrapTextAt=@n31)
   > SetWrapTextWidth(self=Get Text_CosmeticWarning() -> (Text_CosmeticWarning:Common Text Object Reference), InWrapTextAt=@n31)
   Branch(execute=@n28.then_0, Condition=IsMobileGame())
   n33 = Get CommonTextBlock_VbucksLegal()
   > Set Justification(Justification=Left, self=@n33) -> (Output_Get=Left)
   > SetWrapTextWidth(self=@n33, InWrapTextAt=700)
   n34 = Get Text_UndoBottomBarWarning()
   > Set Justification(Justification=Left, self=@n34) -> (Output_Get=Left)
   > SetWrapTextWidth(self=@n34, InWrapTextAt=700)
   > SetWrapTextWidth(self=Get Text_CosmeticWarning(), InWrapTextAt=700)
   n35 = Event SetLoadingOverlayVisibility!() -> (OutputDelegate, then, InVisibility:ESlateVisibility Enum=Visible) {is=title K2Node_Event}
   > SetVisibility(self=Get Loading_Overlay() -> (Loading_Overlay:Overlay Object Reference), InVisibility=@n35.InVisibility)
   Event AttemptPurchase() {is=title K2Node_Event}
   > n36 = ShouldShowCpcDialog!(execute, Context:Widget Object Reference=Self-Reference(), OfferVM:Fort Item Shop Offer VM Object Reference=Get OfferVM()) -> (then, ReturnValue:bool)
   > n37 = Branch(Condition=@n36)
   > OnCpcModalRequested() {is=title K2Node_CallFunction}
   Set ClearPurchaseUXTraceIdOnExit(execute=@n37.else, ClearPurchaseUXTraceIdOnExit:bool=false) -> (Output_Get:bool)
   > Parent: AttemptPurchase() {is=title K2Node_CallParentFunction}
   Event On Offer Set() {is=title K2Node_Event}
   n38 = GetRuntimeOptions()
   n39 = Get bAltUpdateFlow!(self:Fort Runtime Options Object Reference=@n38) -> (bAltUpdateFlow:bool)
   > n40 = Branch(Condition=@n39)
   Init(execute=@n40.else)
   n41 = Event OnAnimationFinished!() -> (OutputDelegate, then, Animation:Widget Animation Object Reference) {is=title K2Node_Event}
   > Branch(Condition="Equal (Object)"(A=@n41.Animation, B=Get Intro()))
   > RequestRefreshFocus(self:Common Activatable Widget Object Reference)
   n42 = Event UpdatePurchaseBlocked!() -> (OutputDelegate, then, bIsBlocked:bool, PurchaseNotAllowedReason:"Text (by ref)") {is=title K2Node_Event}
   > Content Blocking!(execute, self, IsContentBlocked:bool=@n42.bIsBlocked, ContentBlockedText:Text=@n42.PurchaseNotAllowedReason) -> (then) {is=title K2Node_CallFunction}
   n43 = Event SetMtxButtonVisible!() -> (OutputDelegate, then, bVisible:bool) {is=title K2Node_Event}
   > SetVisibility(
    self=Get VBox_MtxContainer() -> (VBox_MtxContainer:Vertical Box Object Reference)
    InVisibility=Select(Option 0:ESlateVisibility Enum=Collapsed, Option 1:ESlateVisibility Enum=SelfHitTestInvisible, Index=@n43.bVisible) -> (ReturnValue:ESlateVisibility Enum)
   )
   Event ResetPurchaseButtonText() {is=title K2Node_Event}
   > UpdatePurchaseButton()
   n44 = Event ShowPurchasedOverlay!() -> (OutputDelegate, then, QuantityPurchased:int=0, PurchasedItems:Array of Purchased Item Info Structures, OfferId:String) {is=title K2Node_Event}
   > n45 = GetItemTemplateIDs!(execute, self, ItemInfo:Array of Purchased Item Info Structures=@n44.PurchasedItems) -> (then, ItemTemplateIDs:Array of Strings)
   > ShowPurchasedOverlayWidget!(
    execute
    self:Athena Item Shop Offer Details Widget Base Object Reference
    QuantityPurchased:int=@n44.QuantityPurchased
    PurchasedItemTemplateIds:Array of Strings=@n45
    bShowAsRefund:bool=false
    OfferId:String=@n44.OfferId
   ) -> (then)
   > RefreshUndoState!(
    execute
    self:Athena Item Shop Offer Details Widget Base Object Reference
    OfferPurchased:Fort Store Front Offer Info Object Reference=GetOfferInfoById!(self:Athena Item Shop Offer Details Widget Base Object Reference, OfferId:String=@n44.OfferId) -> (ReturnValue:Fort Store Front Offer Info Object Reference)
   ) -> (then)
   n46 = Event OnSetPaymentLegalText!() -> (OutputDelegate, then, LegalText:"Text (by ref)") {is=title K2Node_Event}
   > "SetText (Text)"(self=Get Text_PaymentLegalText() -> (Text_PaymentLegalText:Common Text Object Reference), InText=@n46.LegalText)
   > SetVisibility(
    self=Get Text_PaymentLegalText()
    InVisibility=Select(Option 0=SelfHitTestInvisible, Option 1=Collapsed, Index=TextIsEmpty(InText=@n46.LegalText))
   )
   Event On Activated()
   > n47 = Branch(Condition=Get bShowRMTDescriptionPopup!(self:Fort Runtime Options Object Reference=@n38) -> (bShowRMTDescriptionPopup:bool))
   n48 = Branch(execute=[@n47.else, @n47.then], Condition=@n39)
   > n49 = Init()
   UpdatePurchaseButton(execute=[@n49.then, @n48.else])
   > n50 = Set FortItemShopOfferVM(FortItemShopOfferVM:Fort Item Shop Offer VM Object Reference=Get OfferVM()) -> (Output_Get:Fort Item Shop Offer VM Object Reference)
   > PreloadGrantedItemPreviews!(
    execute
    self:Fort Item Shop Offer VM Object Reference=@n50
    bAllow2dPreviews:bool=false
    On2dAssetsPreloaded:"Delegate (by ref)"=Create Event()
    On3dAssetsPreloaded:"Delegate (by ref)"=Create Event()
   ) -> (then)
   Event HandleInspectClicked() {is=title K2Node_Event}
   > Branch(Condition=IsValid(Object=Get OfferData() -> (OfferData:Fort Store Front Offer Info Object Reference)))
   > n51 = Sequence()
   > SendShopInteractionAnalytic!(execute, self:Athena Item Shop Offer Details Widget Base Object Reference, Interaction:String=SubPageViews, HoldTime:float=0) -> (then)
   > Fire Event Element Interacted(
    Element=Get Button_Inspect() -> (Button_Inspect:Fort CTAButton Object Reference)
    AdditionalInfo=BuildAnalyticsInfo!() -> (ReturnValue:Flow Analytics Additional Info Structure) {is=title K2Node_CallFunction}
   )
   > Set ClearPurchaseUXTraceIdOnExit(ClearPurchaseUXTraceIdOnExit=false)
   n52 = Async Load Class Asset!(execute:Exec=@n51.then_1, AssetClass:Object Soft Class Reference=/ItemShop/MultiProduct/WBP_ItemShop_PDP.WBP_ItemShop_PDP_C) -> (then:Exec, Completed:Exec, Class:Object Class Reference) {is=title K2Node_LoadAssetClass}
   n53 = Cast To CommonActivatableWidget Class!(execute:Exec=@n52.Completed, Class:Object Class Reference=@n52) -> (then:Exec, CastFailed:Exec, AsCommon Activatable Widget:Common Activatable Widget Class Reference) {is=title K2Node_ClassDynamicCast}
   > n54 = PushPrimaryContentClassAsync!(execute:Exec, WidgetClass:Common Activatable Widget Class Reference=@n53) -> (then:Exec, OnPrimaryContentPushed:Exec, ActivatableWidget:Common Activatable Widget Object Reference) {is=title K2Node_AsyncAction}
   n55 = Cast To FortMPItemShopOfferInspectionScreen!(execute=@n54.OnPrimaryContentPushed, Object=@n54) -> (then, CastFailed, AsFort MPItem Shop Offer Inspection Screen:Fort MPItem Shop Offer Inspection Screen Object Reference) {is=title K2Node_DynamicCast}
   > SetViewModels!(
    execute
    self:Fort MPItem Shop Offer Inspection Screen Object Reference=@n55
    OfferVM:Fort Item Shop Offer VM Object Reference=Get OfferVM()
    OfferGroupVM:Fort Item Shop Offer Group VM Object Reference
   ) -> (then)
   n56 = Create Event()
   > Unbind Event from On End Purchase(self:Fort MPItem Shop Offer Inspection Screen Object Reference=@n55, Delegate=@n56) {is=title K2Node_RemoveDelegate}
   > Bind Event to On End Purchase(self:Fort MPItem Shop Offer Inspection Screen Object Reference=@n55, Delegate=@n56) {is=title K2Node_AddDelegate}
   > Unbind Event from On Load Error(self:Fort MPItem Shop Offer Inspection Screen Object Reference=@n55, Delegate=@n14) {is=title K2Node_RemoveDelegate}
   > Bind Event to On Load Error(self:Fort MPItem Shop Offer Inspection Screen Object Reference=@n55, Delegate=@n14) {is=title K2Node_AddDelegate}
   Event On Deactivated()
   > n57 = Set LoadState2D(LoadState2D=NewEnumerator0) -> (Output_Get=NewEnumerator3)
   > Set LoadState3D(LoadState3D=@n57) -> (Output_Get=NewEnumerator3)
   > ReleasePreloadedGrantedItemPreviews(self:Fort Item Shop Offer VM Object Reference=Get OfferVM())
   > Set FortItemShopOfferVM()
   > Branch(Condition=Get ClearPurchaseUXTraceIdOnExit() -> (ClearPurchaseUXTraceIdOnExit:bool))
   > ClearPurchaseUXTraceIdIfMatches!(
    execute
    self:Fort Mcp Context Object Reference=Get FortMcpContext()
    ExpectedUXTraceId:String=Get PurchaseUXTraceId()
    DebugSourceContext:String=GetDisplayName(Object=Self-Reference())
   ) -> (then, ReturnValue:bool)
   > Set PurchaseUXTraceId(PurchaseUXTraceId:String) -> (Output_Get:String)
   Parent: AttemptPurchase(execute=@n59.then)
   Unbind Event from Dialog Button Clicked Dispatcher(
    execute=@n58.then
    self:WBP UIKit Dialog Internal Object Reference=Get Dialog_Internal!(self:WBP UIKit Dialog Object Reference=@n10) -> (Dialog_Internal:WBP UIKit Dialog Internal Object Reference)
    Delegate=@n9
   ) {is=title K2Node_RemoveDelegate}
   n58 = DeactivateWidget(execute=@n9.then, self=Self-Reference())
   Branch(execute=@n13.then, Condition=@n13.Accepted)
   > n59 = Set ClearPurchaseUXTraceIdOnExit(ClearPurchaseUXTraceIdOnExit=false)
  ]}
 GetItemTemplateIDs {nodes [
   n1 = GetItemTemplateIDs!() -> (then, ItemInfo:Array of Purchased Item Info Structures) {is=title K2Node_FunctionEntry}
   > n2 = For Each Loop(Array:Array of Purchased Item Info Structures=@n1) -> (Array Element:Purchased Item Info Structure)
   n3 = Break Purchased Item Info!(PurchasedItemInfo:"Purchased Item Info Structure (by ref)"=@n2) -> (Item:Fort Item Object Reference, Quantity:int) {is=title K2Node_BreakStruct}
   n4 = Get TemplateIDs() -> (TemplateIDs:Array of Strings)
   > Add(
    TargetArray=@n4
    NewItem=GetTemplateId!(self:Mcp Item Definition Base Object Reference=GetItemDefinition!(self:Fort Item Object Reference=@n3) -> (ReturnValue:Fort Item Definition Object Reference)) -> (ReturnValue:String)
   ) {is=title K2Node_CallArrayFunction}
   Return Node!(execute=@n2.Completed, ItemTemplateIDs:Array of Strings=@n4)
  ]}
 GetWidgetForFramingViewedItem {nodes [
   GetWidgetForFramingViewedItem() {is=title K2Node_FunctionEntry}
   > Return Node!(execute, ReturnValue:Widget Object Reference=Self-Reference())
  ]}
 HandleLegalText {nodes [
   HandleLegalText() {is=title K2Node_FunctionEntry}
   n1 = GetOfferInfo()
   > ShowLegalTextInPopup!(
    execute
    self:Athena Item Shop Display Asset Offer Details Object Reference
    ItemName:"Text (by ref)"=GetName!(self:Fort Store Front Offer Info Object Reference=@n1) -> (ReturnValue:Text)
    LegalText:"Text (by ref)"=GetDescription!(self:Fort Store Front Offer Info Object Reference=@n1) -> (ReturnValue:Text)
   ) -> (then)
   > Return Node!(execute, Passthrough:bool=false)
  ]}
 Init {nodes [
   Init() {is=title K2Node_FunctionEntry}
   n1 = GetOfferInfo()
   > Set DisplayAsset(DisplayAsset:Fort Mtx Offer Data Object Reference=GetOfferDisplayAsset!(self:Fort Store Front Offer Info Object Reference=@n1) -> (ReturnValue:Fort Mtx Offer Data Object Reference)) -> (Output_Get:Fort Mtx Offer Data Object Reference)
   > SetVisibility(
    self=Get VBox_NonRefundableBullets() -> (VBox_NonRefundableBullets:Vertical Box Object Reference)
    InVisibility=Select(
     Option 0=SelfHitTestInvisible
     Option 1=Collapsed
     Index=IsRefundable!(self:Fort Store Front Offer Info Object Reference=@n1) -> (ReturnValue:bool)
    )
   )
   > SetVisibility(
    self=Get VBox_NonRefundableRmtBullets() -> (VBox_NonRefundableRmtBullets:Vertical Box Object Reference)
    InVisibility=Select(Option 0=Collapsed, Option 1=SelfHitTestInvisible, Index=IsPriceInRealMoney(self=@n1, PriceIndex=0))
   )
   > UpdateHeader()
   n2 = Get DisplayAsset() -> (DisplayAsset:Fort Mtx Offer Data Object Reference)
   > Set colors from data Asset!(
    execute
    self
    Upsell Primary:Linear Color Structure=Get UpsellPrimaryColor!(self:Fort Mtx Offer Data Object Reference=@n2) -> (UpsellPrimaryColor:Linear Color Structure)
    Upsell Secondary:Linear Color Structure=Get UpsellSecondaryColor!(self:Fort Mtx Offer Data Object Reference=@n2) -> (UpsellSecondaryColor:Linear Color Structure)
    Text Color:Linear Color Structure=Get UpsellTextColor!(self:Fort Mtx Offer Data Object Reference=@n2) -> (UpsellTextColor:Linear Color Structure)
    Use Base Color:bool=Get bUseBaseColors!(self:Fort Mtx Offer Data Object Reference=@n2) -> (bUseBaseColors:bool)
    Background:Linear Color Structure=Get Background!(self:Fort Mtx Offer Data Object Reference=@n2) -> (Background:Linear Color Structure)
    Gradient:Linear Color Structure=Break Fort Mtx Gradient!(FortMtxGradient:"Fort Mtx Gradient Structure (by ref)"=Get Gradient!(self:Fort Mtx Offer Data Object Reference=@n2) -> (Gradient:Fort Mtx Gradient Structure)) -> (Start:Linear Color Structure) {is=title K2Node_BreakStruct}
   ) -> (then) {is=title K2Node_CallFunction}
   > "SetText (Text)"(self=Get TextName() -> (TextName:Common Text Object Reference), InText=GetName(self=GetOfferInfo()))
   n3 = Get SubTitleText!(self:Fort Mtx Offer Data Object Reference=Get DisplayAsset()) -> (SubTitleText:Text)
   n4 = Get CommonRichTextBlock_SubTitle() -> (CommonRichTextBlock_SubTitle:Common Rich Text Block Object Reference)
   > SetText(self=@n4, InText=@n3)
   n5 = MakeLiteralText()
   > SetVisibility(
    self=@n4
    InVisibility=Select(
     Option 0=SelfHitTestInvisible
     Option 1=Collapsed
     Index="Equal Exactly (Text)"!(A:"Text (by ref)"=@n3, B:"Text (by ref)"=@n5) -> (ReturnValue:bool)
    )
   )
   n6 = Get DisclaimerText!(self:Fort Mtx Offer Data Object Reference=Get DisplayAsset()) -> (DisclaimerText:Text)
   n7 = Get CommonRichTextBlock_Legal() -> (CommonRichTextBlock_Legal:Common Rich Text Block Object Reference)
   > SetText(self=@n7, InText=@n6)
   n8 = MakeLiteralText()
   > SetVisibility(self=@n7, InVisibility=Select(Option 0=SelfHitTestInvisible, Option 1=Collapsed, Index="Equal Exactly (Text)"(A=@n6, B=@n8)))
   > InitCurrency()
   n9 = Get DisplayAsset()
   > InitDetailAttributes!(
    execute
    self
    UpsellSecondary:Linear Color Structure=Get UpsellSecondaryColor(self=@n9)
    TertiaryColor:Linear Color Structure=Get UpsellTextColor(self=@n9)
    DisplayAsset:Fort Mtx Offer Data Object Reference=@n9
   ) -> (then)
   > UpdatePurchaseButton()
   > PlayAnimationForward(InAnimation=Get Intro())
   > SpawnSound2D!(
    execute
    Sound:Sound Base Object Reference=Get IntroSound() -> (IntroSound:Sound Base Object Reference)
    VolumeMultiplier:float=1
    PitchMultiplier:float=1
    StartTime:float=0
    ConcurrencySettings:Sound Concurrency Object Reference
    bPersistAcrossLevelTransition:bool=false
    bAutoDestroy:bool=true
   ) -> (then, ReturnValue:Audio Component Object Reference)
   > LoadAndInitImages!(execute, self, DisplayAsset:Fort Mtx Offer Data Object Reference=Get DisplayAsset()) -> (then) {is=title K2Node_CallFunction}
   > UpdateCosmeticGateWarning()
   > UpdateLegalText()
   > Set FortItemShopOfferVM!(
    execute
    FortItemShopOfferVM=Get OfferVM()
    self:WBP Item Shop 2d PDP Price Object Reference=Get WBP_ItemShop_2dPDP_Price() -> (WBP_ItemShop_2dPDP_Price:WBP Item Shop 2d PDP Price Object Reference)
   ) -> (then, Output_Get)
   > n10 = Branch(Condition=IsPriceInRealMoney(self=GetOfferInfo(), PriceIndex=0))
   > n11 = Branch(Condition=IsEmpty(InString=Get PurchaseUXTraceId()))
   > n12 = GeneratePurchaseUXTraceId!(
    execute
    self:Fort Mcp Context Object Reference=Get FortMcpContext()
    EpicOfferId:String=GetEpicOfferId!(self:Fort Store Front Offer Info Object Reference=GetOfferInfo()) -> (ReturnValue:String)
    DebugSourceContext:String=GetDisplayName(Object=Self-Reference())
   ) -> (then, ReturnValue:String)
   > n13 = Set PurchaseUXTraceId(PurchaseUXTraceId=@n12)
   n14 = Set ClearPurchaseUXTraceIdOnExit(execute=[@n13.then, @n11.else], ClearPurchaseUXTraceIdOnExit=true)
   Fire Event Sub View Opened!(
    execute=[@n14.then, @n10.else]
    View:Widget Object Reference=Self-Reference()
    AdditionalInfo:Flow Analytics Additional Info Structure=BuildAnalyticsInfo()
    bIsOverlay:bool=false
   ) -> (then)
   > SetPurchaseButtonStyle!(execute, Purchase Button:Fort CTAButton Object Reference=Get Button_Purchase(), GetVbucks Button:Fort CTAButton Object Reference) -> (then)
  ]}
 InitCurrency {nodes [
   InitCurrency() {is=title K2Node_FunctionEntry}
   > n1 = Sequence!(execute) -> (then_0, then_1, then_2:Exec)
   n2 = GetTotalQuantity!(self:Fort Store Front Offer Info Object Reference=GetOfferInfo()) -> (ReturnValue:int=0)
   > SetVisibility(
    self=Get MTXArea() -> (MTXArea:Size Box Object Reference)
    InVisibility=Select(Option 0=Collapsed, Option 1=SelfHitTestInvisible, Index="integer > integer"(A=@n2, B=0) {is=title K2Node_CallFunction})
   )
   n3 = GetBonusQuantity!(self:Fort Store Front Offer Info Object Reference=GetOfferInfo()) -> (ReturnValue:int=0)
   SetVisibility(
    execute=@n1.then_1
    self=Get BonusMTXArea() -> (BonusMTXArea:Horizontal Box Object Reference)
    InVisibility=Select(Option 0=Collapsed, Option 1=SelfHitTestInvisible, Index="integer > integer"(A=@n3, B=0))
   )
   n4 = GetOfferInfo()
   "SetText (Text)"(
    execute=@n1.then_2
    self=Get CommonText_BaseQuantity() -> (CommonText_BaseQuantity:Common Text Object Reference)
    InText="To Text (Integer)"(Value=GetBaseQuantity!(self:Fort Store Front Offer Info Object Reference=@n4) -> (ReturnValue:int=0))
   )
   > "SetText (Text)"(
    self=Get CommonText_BonusQuantity() -> (CommonText_BonusQuantity:Common Text Object Reference)
    InText="To Text (Integer)"(Value=GetBonusQuantity(self=@n4) -> (ReturnValue=0))
   )
   n5 = GetTotalQuantity(self=GetOfferInfo()) -> (ReturnValue=0)
   > SetVisibility(
    self=Get CommonTextBlock_VbucksLegal()
    InVisibility=Select(Option 0=Collapsed, Option 1=SelfHitTestInvisible, Index="integer > integer"(A=@n5, B=0))
   )
  ]}
 InitDetailAttributes {nodes [
   n1 = InitDetailAttributes!() -> (then, UpsellSecondary:Linear Color Structure, TertiaryColor:Linear Color Structure, DisplayAsset:Fort Mtx Offer Data Object Reference) {is=title K2Node_FunctionEntry}
   > Reset!(
    execute
    self:Dynamic Entry Box Object Reference=Get DynamicEntryBox_IncludedItems() -> (DynamicEntryBox_IncludedItems:Dynamic Entry Box Object Reference)
    bDeleteWidgets:bool=true
   ) -> (then)
   > n2 = For Each Loop(Array:Array of Fort Mtx Description And Details Structures=GetDescriptionAndDetailsList!(self:Fort Mtx Offer Data Object Reference=@n1.DisplayAsset) -> (ReturnValue:Array of Fort Mtx Description And Details Structures)) -> (Array Element:Fort Mtx Description And Details Structure)
   > n3 = Create Entry!(execute, self:Dynamic Entry Box Object Reference=Get DynamicEntryBox_IncludedItems()) -> (then, ReturnValue:User Widget Object Reference)
   > SetupDataDetails!(
    execute
    self:RMT Offer Detailed Description Object Reference=Cast To RMT_Offer_DetailedDescription!(Object=@n3) -> (AsRMT Offer Detailed Description:RMT Offer Detailed Description Object Reference, bSuccess:bool=~) {is=title K2Node_DynamicCast}
    FortMtxDescriptionAndDetails:"Fort Mtx Description And Details Structure (by ref)"=@n2
    SecondaryColor:Linear Color Structure=Get UpsellSecondary() -> (UpsellSecondary:Linear Color Structure)
    TertiaryColor:Linear Color Structure=Get TertiaryColor() -> (TertiaryColor:Linear Color Structure)
   ) -> (then)
  ]}
 IsGiftingStorefrontOffer {nodes [
   IsGiftingStorefrontOffer() {is=title K2Node_FunctionEntry}
   > Return Node!(
    execute
    ReturnValue:bool=IsGiftingStorefrontOffer!(self:Fort Store Front Offer Info Object Reference=GetOfferInfo()) -> (ReturnValue:bool) {is=title K2Node_CallFunction}
   )
  ]}
 PopDetails {nodes [
   PopDetails() {is=title K2Node_FunctionEntry}
   > PopContentWidget!(execute) -> (then)
  ]}
 RegisterInput {nodes [
   RegisterInput() {is=title K2Node_FunctionEntry}
   > Branch(Condition=IsPriceInMTX!(self:Fort Store Front Offer Info Object Reference=GetOfferInfo(), PriceIndex:int=0) -> (ReturnValue:bool))
   Make Data Table Row Handle!(DataTable:Data Table Object Reference=&InputActionDataTable, RowName:Name=LegalText) -> (DataTableRowHandle:Data Table Row Handle Structure) {is=title K2Node_MakeStruct}
  ]}
 Set colors from data Asset {nodes [
   Set colors from data Asset!() -> (
    then
    Upsell Primary:Linear Color Structure
    Upsell Secondary:Linear Color Structure
    Text Color:Linear Color Structure
    Use Base Color:bool=~
    Background:Linear Color Structure
    Gradient:Linear Color Structure
   ) {is=title K2Node_FunctionEntry}
   > SetColorAndOpacity(
    self:Image Object Reference=Get BaseColor() -> (BaseColor:Image Object Reference)
    InColorAndOpacity:Linear Color Structure=Select(
     Option 0:Linear Color Structure=Get Background!() -> (Background)
     Option 1:Linear Color Structure=Get Upsell Primary() -> (Upsell Primary:Linear Color Structure)
     Index=Get Use Base Color() -> (Use Base Color:bool)
    ) -> (ReturnValue:Linear Color Structure)
   )
   > SetColorAndOpacity(
    self=[Get Highlight() -> (Highlight:Image Object Reference), Get SpeedLines() -> (SpeedLines:Image Object Reference)]
    InColorAndOpacity=Select(
     Option 0=Get Gradient!() -> (Gradient:Linear Color Structure)
     Option 1=Get Upsell Secondary() -> (Upsell Secondary:Linear Color Structure)
     Index=Get Use Base Color()
    )
   )
   > n1 = GetDynamicMaterial(self=Get OverlayGradientDetails() -> (OverlayGradientDetails:Image Object Reference))
   > n2 = Set OverlayGradientMID(OverlayGradientMID:Material Instance Dynamic Object Reference=@n1) -> (Output_Get:Material Instance Dynamic Object Reference)
   > SetVectorParameterValue(self=@n2, ParameterName=RGB1, Value=Get Upsell Primary())
   > SetVectorParameterValue(
    self=Get OverlayGradientMID() -> (OverlayGradientMID:Material Instance Dynamic Object Reference)
    ParameterName=RGB2
    Value=Get Upsell Secondary()
   )
   n3 = Get Upsell Secondary()
   > SetColorAndOpacity(
    self=[
     Get Image_Separator_Up() -> (Image_Separator_Up:Image Object Reference)
     Get Image_Separator_Down() -> (Image_Separator_Down:Image Object Reference)
    ]
    InColorAndOpacity=@n3
   )
   > SetColorAndOpacity(self:Text Object Reference=Get CommonTextBlock_VbucksLegal(), InColorAndOpacity:Slate Color Structure=Make Slate Color(SpecifiedColor=@n3))
   n4 = Get Upsell Secondary()
   > SetBrushColor(self=Get CommonBorder_VBucks() -> (CommonBorder_VBucks:Common Border Object Reference), InBrushColor=@n4)
   > SetDefaultColorAndOpacity!(
    execute
    self:Rich Text Block Object Reference=Get CommonRichTextBlock_Legal()
    InColorAndOpacity:Slate Color Structure=Make Slate Color(SpecifiedColor=@n4)
   ) -> (then)
   > SetDefaultColorAndOpacity(
    self=[Get CommonRichTextBlock_SubTitle(), Get RichText_Header() -> (RichText_Header:Common Rich Text Block Object Reference)]
    InColorAndOpacity=Make Slate Color(SpecifiedColor=Get Text Color() -> (Text Color:Linear Color Structure))
   )
   > SetColorAndOpacity(self:Image Object Reference=Get Image_LTM() -> (Image_LTM:Image Object Reference), InColorAndOpacity:Linear Color Structure=Get Text Color())
   n5 = Get ScrollBox_Details() -> (ScrollBox_Details:Scroll Box Object Reference)
   n6 = Break Scroll Bar Style!(ScrollBarStyle:"Scroll Bar Style Structure (by ref)"=Get WidgetBarStyle!(self:Scroll Box Object Reference=@n5) -> (WidgetBarStyle:Scroll Bar Style Structure)) -> (
    HorizontalBackgroundImage:Slate Brush Structure
    VerticalBackgroundImage:Slate Brush Structure
    VerticalTopSlotImage:Slate Brush Structure
    HorizontalTopSlotImage:Slate Brush Structure
    VerticalBottomSlotImage:Slate Brush Structure
    HorizontalBottomSlotImage:Slate Brush Structure
    NormalThumbImage:Slate Brush Structure
    HoveredThumbImage:Slate Brush Structure
    DraggedThumbImage:Slate Brush Structure
    Thickness:float=~
   ) {is=title K2Node_BreakStruct}
   n7 = Break Slate Brush!(SlateBrush:"Slate Brush Structure (by ref)"=@n6.HoveredThumbImage) -> (
    TintColor:Slate Color Structure
    DrawAs:ESlateBrushDrawType Enum
    Tiling:ESlateBrushTileType Enum
    Mirroring:ESlateBrushMirrorType Enum
    ImageSize:Vector 2D Structure
    Margin:Margin Structure
    ResourceObject:Object Reference
    OutlineSettings:Slate Brush Outline Settings Structure
   ) {is=title K2Node_BreakStruct}
   n8 = Make Slate Color(SpecifiedColor=Get Text Color())
   n9 = Break Slate Brush(SlateBrush=@n6.NormalThumbImage)
   n10 = Make Slate Color(SpecifiedColor=Get Upsell Secondary())
   n11 = Break Slate Brush(SlateBrush=@n6.DraggedThumbImage)
   n12 = Break Slate Brush(SlateBrush=@n6.VerticalBackgroundImage)
   n13 = BreakColor(InColor=Get Upsell Primary())
   n14 = Make Slate Color(SpecifiedColor=Make Linear Color!(R:float=@n13, G:float=@n13.G, B:float=@n13.B, A:float=0.5) -> (LinearColor:Linear Color Structure) {is=title K2Node_MakeStruct})
   > Set WidgetBarStyle!(
    execute
    WidgetBarStyle:Scroll Bar Style Structure=Make Scroll Bar Style!(
     HorizontalBackgroundImage:Slate Brush Structure=@n6
     VerticalBackgroundImage:Slate Brush Structure=Make Slate Brush!(
      TintColor:Slate Color Structure=@n14
      DrawAs:ESlateBrushDrawType Enum=@n12.DrawAs
      Tiling:ESlateBrushTileType Enum=@n12.Tiling
      Mirroring:ESlateBrushMirrorType Enum=@n12.Mirroring
      ImageSize:Vector 2D Structure=@n12.ImageSize
      Margin:Margin Structure=@n12.Margin
      ResourceObject:Object Reference=@n12.ResourceObject
      OutlineSettings:Slate Brush Outline Settings Structure=(CornerRadii=(X=0,Y=0,Z=0,W=0),Color=(SpecifiedColor=(R=0,G=0,B=0,A=0),ColorUseRule=UseColor_Specified),Width=0,RoundingType=HalfHeightRadius,bUseBrushTransparency=False)
     ) -> (SlateBrush:Slate Brush Structure) {is=title K2Node_MakeStruct}
     VerticalTopSlotImage:Slate Brush Structure=@n6.VerticalTopSlotImage
     HorizontalTopSlotImage:Slate Brush Structure=@n6.HorizontalTopSlotImage
     VerticalBottomSlotImage:Slate Brush Structure=@n6.VerticalBottomSlotImage
     HorizontalBottomSlotImage:Slate Brush Structure=@n6.HorizontalBottomSlotImage
     NormalThumbImage:Slate Brush Structure=Make Slate Brush(
      TintColor=@n10
      DrawAs=@n9.DrawAs
      Tiling=@n9.Tiling
      Mirroring=@n9.Mirroring
      ImageSize=@n9.ImageSize
      Margin=@n9.Margin
      ResourceObject=@n9.ResourceObject
      OutlineSettings=(CornerRadii=(X=0,Y=0,Z=0,W=0),Color=(SpecifiedColor=(R=0,G=0,B=0,A=0),ColorUseRule=UseColor_Specified),Width=0,RoundingType=HalfHeightRadius,bUseBrushTransparency=False)
     )
     HoveredThumbImage:Slate Brush Structure=Make Slate Brush(
      TintColor=@n8
      DrawAs=@n7.DrawAs
      Tiling=@n7.Tiling
      Mirroring=@n7.Mirroring
      ImageSize=@n7.ImageSize
      Margin=@n7.Margin
      ResourceObject=@n7.ResourceObject
      OutlineSettings=(CornerRadii=(X=0,Y=0,Z=0,W=0),Color=(SpecifiedColor=(R=0,G=0,B=0,A=0),ColorUseRule=UseColor_Specified),Width=0,RoundingType=HalfHeightRadius,bUseBrushTransparency=False)
     )
     DraggedThumbImage:Slate Brush Structure=Make Slate Brush(
      TintColor=@n8
      DrawAs=@n11.DrawAs
      Tiling=@n11.Tiling
      Mirroring=@n11.Mirroring
      ImageSize=@n11.ImageSize
      Margin=@n11.Margin
      ResourceObject=@n11.ResourceObject
      OutlineSettings=(CornerRadii=(X=0,Y=0,Z=0,W=0),Color=(SpecifiedColor=(R=0,G=0,B=0,A=0),ColorUseRule=UseColor_Specified),Width=0,RoundingType=HalfHeightRadius,bUseBrushTransparency=False)
     )
     Thickness:float=16
    ) -> (ScrollBarStyle:Scroll Bar Style Structure) {is=title K2Node_MakeStruct}
    self:Scroll Box Object Reference=@n5
   ) -> (then, Output_Get:Scroll Bar Style Structure)
  ]}
 SetLoadGuardButtonProgress {nodes [
   SetLoadGuardButtonProgress!() -> (then, LoadingProgress:double=~) {is=title K2Node_FunctionEntry}
   n1 = Get Button_LoadGuard() -> (Button_LoadGuard:WBP UIKit Button Load Guard Object Reference)
   n2 = Get LoadingProgress() -> (LoadingProgress:double)
   > Set HasChangedState(HasChangedState:bool="Not Equal (Boolean)"(A=GetIsLoading!(self:WBP UIKit Button Load Guard Object Reference=@n1) -> (IsLoading:bool), B="float < float"(A=@n2, B=1)) -> (ReturnValue=~)) -> (Output_Get:bool)
   > SetLoadingProgress!(
    execute
    self:WBP UIKit Button Load Guard Object Reference=Get Button_LoadGuard()
    InLoadProgress:double=Get LoadingProgress()
    ShowTransition:bool=true
   ) -> (then)
   n3 = Get FortItemShopOfferVM() -> (FortItemShopOfferVM:Fort Item Shop Offer VM Object Reference)
   > SetLoadingText!(
    execute
    self:WBP UIKit Button Load Guard Object Reference=Get Button_LoadGuard()
    InLoadingText:Text=Format Text!(
     Format="LOCTEXT(\"Loading {NumLoaded}/{NumToLoad}\")"
     NumLoaded:int=Get NumPreloadedGrantedItemPreviews!(self:Fort Item Shop Offer VM Object Reference=@n3) -> (NumPreloadedGrantedItemPreviews:int=0)
     NumToLoad:int=Get NumGrantedItemPreviewsToPreload!(self:Fort Item Shop Offer VM Object Reference=@n3) -> (NumGrantedItemPreviewsToPreload:int=0)
    ) -> (Result)
   ) -> (then)
   > Branch(Condition=Get HasChangedState() -> (HasChangedState:bool))
   > RequestRefreshFocus()
  ]}
 SetPurchaseButtonText {nodes [
   n1 = SetPurchaseButtonText!() -> (then, ButtonText:Text) {is=title K2Node_FunctionEntry}
   > SetText!(execute, self:Fort CTAButton Object Reference=Get Button_Purchase(), ButtonText:"Text (by ref)"=@n1) -> (then)
  ]}
 ShowPurchaseConfirmation {nodes [
   ShowPurchaseConfirmation() {is=title K2Node_FunctionEntry}
   > n1 = Create Widget!(execute, Class=&/AthenaItemShopConfirmation.AthenaItemShopConfirmation_C, OwningPlayer=GetOwningPlayer()) -> (then, ReturnValue:Athena Item Shop Confirmation Object Reference)
   > n2 = Set Confirmation(Confirmation:Athena Item Shop Confirmation Object Reference=@n1) -> (Output_Get:Athena Item Shop Confirmation Object Reference)
   > Bind Event to Confirm(self:Athena Item Shop Confirmation Object Reference=@n2, Delegate=Create Event()) {is=title K2Node_AddDelegate}
   > Bind Event to Decline(
    self:Athena Item Shop Confirmation Object Reference=Get Confirmation() -> (Confirmation:Athena Item Shop Confirmation Object Reference)
    Delegate=Create Event()
   ) {is=title K2Node_AddDelegate}
   > SetPurchaseNumber!(execute, self:Athena Item Shop Confirmation Object Reference=Get Confirmation(), InNumToPurchse:int=1) -> (then)
   > SetOffer!(
    execute
    self:Athena Item Shop Offer Details Widget Base Object Reference=Get Confirmation()
    InOfferData:Fort Store Front Offer Info Object Reference=GetOfferInfo()
   ) -> (then)
   > TryPushWidgetAsModalPanel(PushWidget=Get Confirmation())
  ]}
 UpdateCosmeticGateWarning {nodes [
   UpdateCosmeticGateWarning() {is=title K2Node_FunctionEntry}
   > Set ShouldShowPinModal(ShouldShowPinModal:bool=false) -> (Output_Get:bool)
   > n1 = For Each Loop(Array:Array of Fort Item VM Object References=Get GrantedItems!(self:Fort Item Shop Offer VM Object Reference=Get OfferVM()) -> (GrantedItems:Array of Fort Item VM Object References)) -> (Array Element:Fort Item VM Object Reference)
   n2 = ConvertCosmeticItemDefinitionsToInterfaces!(
    execute=@n1.Completed
    CosmeticItemDefinitions:Array of Athena Cosmetic Item Definition Object References=Get ItemDefs() -> (ItemDefs:Array of Athena Cosmetic Item Definition Object References)
   ) -> (then, OutCosmeticItemGatingInterfaces:Array of Cosmetic Item Gating Interface Interfaces)
   > n3 = Set CosmeticGatingInterfaces(CosmeticGatingInterfaces:Array of Cosmetic Item Gating Interface Interfaces=@n2) -> (Output_Get:Array of Cosmetic Item Gating Interface Interfaces)
   > n4 = ItemsHaveVocalRatingRestriction!(
    execute
    LocalPlayer:Local Player Object Reference=GetOwningLocalPlayer!(self:Widget Object Reference) -> (ReturnValue:Local Player Object Reference)
    CosmeticItems:Array of Cosmetic Item Gating Interface Interfaces=@n3
   ) -> (then, ReturnValue:bool)
   > n5 = IsPlayerEligibleToUseRestrictedItems!(
    execute
    LocalPlayer:Local Player Object Reference=GetOwningLocalPlayer()
    CosmeticItems:Array of Cosmetic Item Gating Interface Interfaces=Get CosmeticGatingInterfaces() -> (CosmeticGatingInterfaces:Array of Cosmetic Item Gating Interface Interfaces)
   ) -> (then, ReturnValue:bool)
   n6 = Get HasLoopFoundUnrestrictedOutfit() -> (HasLoopFoundUnrestrictedOutfit:bool)
   > Set ShouldShowPinModal(ShouldShowPinModal=NOT Boolean(A=OR Boolean(A=@n5, B=@n6, C=@n4)))
   > n7 = GetRatingRestrictionTextForItems!(
    execute
    LocalPlayer:Local Player Object Reference=GetOwningLocalPlayer()
    CosmeticItems:Array of Cosmetic Item Gating Interface Interfaces=Get CosmeticGatingInterfaces()
   ) -> (then, ReturnValue:Text)
   > n8 = Branch(Condition=TextIsEmpty(InText=@n7))
   > SetVisibility(self=Get CosmeticGate_Warning() -> (CosmeticGate_Warning:WBP UIKit Info Box Object Reference), InVisibility=Collapsed)
   > Set IsRestricted(IsRestricted:bool=false) -> (Output_Get:bool)
   n9 = Get CosmeticGate_Warning()
   SetText!(execute=@n8.else, self:WBP UIKit Info Box Object Reference=@n9, InText:Text=@n7) -> (then)
   > SetVisibility(self=@n9, InVisibility=SelfHitTestInvisible)
   > Set IsRestricted(IsRestricted=true)
   Set LoopItem(execute=@n1.LoopBody, LoopItem:Fort Item VM Object Reference=@n1) -> (Output_Get:Fort Item VM Object Reference)
   > n10 = Cast To AthenaCosmeticItemDefinition!(execute, Object=Get ItemDefinition!(self:Fort Item VM Object Reference=@n1) -> (ItemDefinition:Fort Item Definition Object Reference)) -> (then, CastFailed, AsAthena Cosmetic Item Definition:Athena Cosmetic Item Definition Object Reference) {is=title K2Node_DynamicCast}
   > Add(
    TargetArray:Array of Athena Cosmetic Item Definition Object References=Get ItemDefs()
    NewItem:"Athena Cosmetic Item Definition Object Reference (by ref)"=@n10
   )
   > Branch(Condition=NOT Boolean(A=Get HasLoopFoundUnrestrictedOutfit()))
   > n11 = HasRatingRestrictions!(
    execute
    LocalPlayer:Local Player Object Reference=GetOwningLocalPlayer()
    CosmeticItemGatingInterface:"Cosmetic Item Gating Interface Interface (by ref)"=@n10
   ) -> (then, ReturnValue:bool)
   n12 = GetItemType!(ItemDefinition:Item Definition Base Object Reference=@n10) -> (ReturnValue:EFortItemType Enum=WorldItem)
   > Branch(Condition=AND Boolean(A=NOT Boolean(A=@n11), B="Equal (Enum)"(A:EFortItemType Enum=@n12, B:EFortItemType Enum=AthenaCharacter)))
   > Set HasLoopFoundUnrestrictedOutfit(HasLoopFoundUnrestrictedOutfit:bool=true) -> (Output_Get:bool)
  ]}
 UpdateHeader {nodes [
   UpdateHeader() {is=title K2Node_FunctionEntry}
   > n1 = Get!(execute:Exec) -> (then:Exec, else:Exec, OfferVM:Fort Item Shop Offer VM Object Reference)
   n2 = Get OutDate!(self:Fort Item Shop Offer VM Object Reference=@n1) -> (OutDate:Date Time Structure)
   > SetText(
    self:Rich Text Block Object Reference=Get RichText_Header()
    InText:"Text (by ref)"=Format Text!(
     Format="LOCTEXT(\"This offer will no longer be available as of {Date} at {Time} your local time ({GmtOffset}).\")"
     Date:Text="As Date (from UTC)"!(InDateTime:"Date Time Structure (by ref)"=@n2, InTimeZone:String, InDateStyle:EDateTimeStyle Enum=Medium) -> (ReturnValue:Text)
     Time:Text=AsShortTime!(InDateTime:"Date Time Structure (by ref)"=@n2) -> (ReturnValue:Text)
     GmtOffset:Text=GetLocalGMTOffset!(InDateTime:"Date Time Structure (by ref)"=@n2) -> (ReturnValue:Text)
    ) -> (Result)
   )
   > SetVisibility(self=Get HorizontalBox_LTM() -> (HorizontalBox_LTM:Horizontal Box Object Reference), InVisibility=SelfHitTestInvisible)
   SetText(execute=@n1.else, self=Get RichText_Header())
   > SetVisibility(self=Get HorizontalBox_LTM(), InVisibility=Collapsed)
  ]}
 UpdateLegalText {nodes [
   UpdateLegalText() {is=title K2Node_FunctionEntry}
   > Set Text(Text=GetLegalText!(self:Fort Item Shop Offer VM Object Reference=Get OfferVM()) -> (ReturnValue:Text), self=Get Text_CosmeticWarning())
   n1 = GetOfferDisplayAsset(self=Get OfferInfo!(self:Fort Item Shop Offer VM Object Reference=Get OfferVM()) -> (OfferInfo:Athena Item Shop Offer Info Object Reference))
   > SetColorAndOpacity(
    self:Text Object Reference=Get Text_CosmeticWarning()
    InColorAndOpacity:Slate Color Structure=Make Slate Color(SpecifiedColor=Get LegalTextColor!(self:Fort Mtx Offer Data Object Reference=@n1) -> (LegalTextColor:Linear Color Structure))
   )
  ]}
 UpdatePurchaseButton {nodes [
   UpdatePurchaseButton() {is=title K2Node_FunctionEntry}
   n1 = GetOfferInfo()
   > Set IsOwned(IsOwned:bool=AreGrantedItemsOwned!(self:Fort Store Front Offer Info Object Reference=@n1) -> (ReturnValue:bool)) -> (Output_Get:bool)
   > n2 = Branch(Condition=IsPriceInMTX(self=@n1, PriceIndex=Get PriceIndex() -> (PriceIndex:int=0)))
   > n3 = Branch(Condition=Get IsOwned() -> (IsOwned:bool))
   > SetRequiresHold!(execute, self:Common Button Base Object Reference=Get Button_Purchase(), bInRequiresHold:bool=false) -> (then)
   > SetPurchaseButtonText(ButtonText="LOCTEXT(\"Owned\")") {is=title K2Node_CallFunction}
   n4 = Branch(
    execute=@n3.else
    Condition=CanPurchaseMtxPriceOffer!(PirceIndex:int=Get PriceIndex() -> (PriceIndex=0)) -> (ReturnValue:bool) {is=title K2Node_CallFunction}
   )
   > n5 = ShouldShowRightOfWithdrawalModal!(execute, Context:Widget Object Reference) -> (then, ReturnValue:bool)
   > SetRequiresHold(self=Get Button_Purchase(), bInRequiresHold=NOT Boolean(A=@n5))
   > SetPurchaseButtonText(ButtonText="LOCTEXT(\"Purchase\")")
   SetRequiresHold(execute=@n4.else, self=Get Button_Purchase(), bInRequiresHold=false)
   > n6 = IsReloadMtxEnabled!(execute, self:Athena Item Shop Offer Details Widget Base Object Reference) -> (then, ReturnValue:bool)
   > n7 = Branch(Condition=@n6)
   > SetPurchaseButtonText(ButtonText="LOCTEXT(\"RELOAD V-BUCKS\")")
   SetPurchaseButtonText(execute=@n7.else, ButtonText="LOCTEXT(\"GET V-BUCKS\")")
   SetPurchaseButtonText(
    execute=@n2.else
    ButtonText=Select(Option 0:Text="LOCTEXT(\"Purchase\")", Option 1:Text="LOCTEXT(\"Owned\")", Index=Get IsOwned()) -> (ReturnValue:Text)
   )
   > SetRequiresHold(self=Get Button_Purchase(), bInRequiresHold=false)
  ]}
 VM_PreloadedItemsProgressUpdated {nodes [
   VM_PreloadedItemsProgressUpdated!() -> (then, LoadingProgress:double=~) {is=title K2Node_FunctionEntry}
   n1 = NOT Boolean(A=IsPriceInRealMoney(self=GetOfferInfo(), PriceIndex=0))
   n2 = "Equal (Enum)"(A:E_ItemPreview_LoadState Enum=Get LoadState2D() -> (LoadState2D=NewEnumerator3), B:E_ItemPreview_LoadState Enum=NewEnumerator2)
   n3 = Get LoadState3D() -> (LoadState3D=NewEnumerator3)
   > n4 = Branch(Condition=OR Boolean!(A=AND Boolean(A=@n2, B=@n1), B="Equal (Enum)"(A=@n3, B=NewEnumerator2)) -> (ReturnValue))
   n5 = Get NumGrantedItemPreviewsToPreload(self=Get OfferVM()) -> (NumGrantedItemPreviewsToPreload=0)
   SetLoadGuardButtonProgress(
    execute=@n4.else
    LoadingProgress=Select(Option 0:double=Get LoadingProgress(), Option 1:double=1, Index="Equal (Integer)"(A=@n5)) -> (ReturnValue:double)
   ) {is=title K2Node_CallFunction}
  ]}
]
variables [
 ClearPurchaseUXTraceIdOnExit {type=boolean, default=true}
 CloseCount {type=integer}
 CurrentFocusedButtonIndex {type=integer}
 IntroSound {type=SoundBase, default=/Game/Sounds/Fort_UI_Sounds/FrontEnd/ItemShop/Fort_UI_Shop_OfferDetails_Cue}
 IsRestricted {type=boolean}
 LoadState2D {type=E_ItemPreview_LoadState, default=Unset}
 LoadState3D {type=E_ItemPreview_LoadState, default=Unset}
 NavigableButtonsArray {type="array<Widget>", default []}
 PriceItem {type=FortAccountItemDefinition, default=null}
 PurchaseUXTraceId {type=string, default=""}
 ShortDescriptions {type="array<CommonRichTextBlock>", default []}
 ShouldShowPinModal {type=boolean}
]
widgets {
 info {
  inheritedWidgetCount=10
  namedSlotCount=1
  parentClass=/Script/FortniteUI.AthenaItemShopDisplayAssetOfferDetails
  rootWidgetClass=&Overlay
  widgetCount=88
 }
 widgets [
  Overlay_2 {widgetClassPath=&Overlay, widget=Overlay_2}
  BaseColor {
   widgetClassPath=&Image
   bIsVariable=true
   parent=Overlay_2
   widget=BaseColor
   properties {colorAndOpacity=<rgba 0.0097 0.0075 0.0382 1>}
   slot {class=OverlaySlot, properties {horizontalAlignment=HAlign_Fill, verticalAlignment=VAlign_Fill}}
  }
  SpeedLines {
   widgetClassPath=&Image
   bIsVariable=true
   parent=Overlay_2
   widget=SpeedLines
   properties {
    visibility=SelfHitTestInvisible
    brush {
     bIsDynamicallyLoaded=false
     drawAs=Image
     imageType=NoImage
     mirroring=NoMirror
     resourceName=None
     resourceObject=/Game/Athena/UI/Frontend/Art/M_UI_SpeedLines_MOTD
     tiling=NoTile
     imageSize=<32 32>
     margin=<ltrb 0 0 0 0>
     outlineSettings {
      bUseBrushTransparency=false
      roundingType=HalfHeightRadius
      width=0
      color {colorUseRule=UseColor_Specified, specifiedColor=<rgba 0 0 0 0>}
      cornerRadii=<0 0 0 0>
     }
     tintColor {colorUseRule=UseColor_Specified, specifiedColor=<rgba 1 1 1 0.15>}
     uVRegion {bIsValid=false, max=<0 0>, min=<0 0>}
    }
    colorAndOpacity=<rgba 0.0844 0.0953 0.2384 1>
   }
   slot {class=OverlaySlot, properties {horizontalAlignment=HAlign_Fill, verticalAlignment=VAlign_Fill}}
  }
  SafeZone_0 {
   widgetClassPath=&SafeZone
   parent=Overlay_2
   widget=SafeZone_0
   properties {padBottom=false}
   slot {class=OverlaySlot, properties {horizontalAlignment=HAlign_Fill, verticalAlignment=VAlign_Fill}}
  }
  Overlay_4 {widgetClassPath=&Overlay, parent=SafeZone_0, widget=Overlay_4, slot {class=SafeZoneSlot}}
  HorizontalBox_4 {
   widgetClassPath=&HorizontalBox
   parent=Overlay_4
   widget=HorizontalBox_4
   properties {clipping=ClipToBounds}
   slot {class=OverlaySlot, properties {horizontalAlignment=HAlign_Fill, verticalAlignment=VAlign_Fill}}
  }
  Spacer_90 {
   widgetClassPath=&Spacer
   parent=HorizontalBox_4
   widget=Spacer_90
   properties {size=<650 0>}
   slot {class=HorizontalBoxSlot, properties {horizontalAlignment=HAlign_Left}}
  }
  Overlay_1 {
   widgetClassPath=&Overlay
   parent=HorizontalBox_4
   widget=Overlay_1
   slot {class=HorizontalBoxSlot, properties {size {sizeRule=Fill, value=1}}}
  }
  Highlight {
   widgetClassPath=&Image
   bIsVariable=true
   parent=Overlay_1
   widget=Highlight
   properties {
    visibility=SelfHitTestInvisible
    brush {
     bIsDynamicallyLoaded=false
     drawAs=Image
     imageType=NoImage
     mirroring=NoMirror
     resourceName=None
     resourceObject=/Game/UI/Foundation/Materials/M-Store-Highlight
     tiling=NoTile
     imageSize=<1024 1024>
     margin=<ltrb 0 0 0 0>
     outlineSettings {
      bUseBrushTransparency=false
      roundingType=HalfHeightRadius
      width=0
      color {colorUseRule=UseColor_Specified, specifiedColor=<rgba 0 0 0 0>}
      cornerRadii=<0 0 0 0>
     }
     tintColor {colorUseRule=UseColor_Specified, specifiedColor=<rgba 1 1 1 0.75>}
     uVRegion {bIsValid=false, max=<0 0>, min=<0 0>}
    }
    colorAndOpacity=<rgba 0.0844 0.0953 0.2384 1>
    renderTransform {angle=0, scale=<2.25 1.5>, shear=<0 0>, translation=<0 0>}
   }
   slot {class=OverlaySlot, properties {horizontalAlignment=HAlign_Center, verticalAlignment=VAlign_Center}}
  }
  HorizontalBox_0 {
   widgetClassPath=&HorizontalBox
   parent=Overlay_1
   widget=HorizontalBox_0
   slot {class=OverlaySlot, properties {horizontalAlignment=HAlign_Fill, verticalAlignment=VAlign_Fill, padding=<ltrb 0 0 50 0>}}
  }
  ScaleBox_0 {
   widgetClassPath=&ScaleBox
   parent=HorizontalBox_0
   widget=ScaleBox_0
   properties {stretchDirection=DownOnly}
   slot {class=HorizontalBoxSlot, properties {size {sizeRule=Fill, value=1}}}
  }
  KeyArt {
   widgetClassPath=&FortLazyImage
   bIsVariable=true
   parent=ScaleBox_0
   widget=KeyArt
   properties {
    bOverride_Cursor=true
    brush {
     bIsDynamicallyLoaded=false
     drawAs=Image
     imageType=FullColor
     mirroring=NoMirror
     resourceName=None
     resourceObject=/Game/UI/ContentOnDemand/ItemDownloadIndicator/MI_ItemDownloadIndicator_SpinnerMaterial
     tiling=NoTile
     imageSize=<64 64>
     margin=<ltrb 0 0 0 0>
     outlineSettings {
      bUseBrushTransparency=false
      roundingType=HalfHeightRadius
      width=0
      color {colorUseRule=UseColor_Specified, specifiedColor=<rgba 0 0 0 0>}
      cornerRadii=<0 0 0 0>
     }
     tintColor {colorUseRule=UseColor_Specified, specifiedColor=<rgba 1 1 1 1>}
     uVRegion {bIsValid=false, max=<0 0>, min=<0 0>}
    }
    loadingBackgroundBrush {
     bIsDynamicallyLoaded=false
     drawAs=Image
     imageType=FullColor
     mirroring=NoMirror
     resourceName=None
     resourceObject=&MI_ItemDownloadIndicator_SpinnerMaterial
     tiling=NoTile
     imageSize=<64 64>
     margin=<ltrb 0 0 0 0>
     outlineSettings {
      bUseBrushTransparency=false
      roundingType=HalfHeightRadius
      width=0
      color {colorUseRule=UseColor_Specified, specifiedColor=<rgba 0 0 0 0>}
      cornerRadii=<0 0 0 0>
     }
     tintColor {colorUseRule=UseColor_Specified, specifiedColor=<rgba 1 1 1 1>}
     uVRegion {bIsValid=false, max=<0 0>, min=<0 0>}
    }
   }
   slot {class=ScaleBoxSlot}
  }
  WBP_CaptureForPostBufferUpdate_66 {
   widgetClassPath=&WBP_CaptureForPostBufferUpdate
   bIsVariable=true
   parent=Overlay_1
   widget=WBP_CaptureForPostBufferUpdate_66
   slot {class=OverlaySlot, properties {horizontalAlignment=HAlign_Fill, verticalAlignment=VAlign_Fill}}
  }
  BorderDisclaimer {
   widgetClassPath=&CommonBorder
   bIsVariable=true
   parent=Overlay_1
   widget=BorderDisclaimer
   properties {padding=<ltrb 0 0 0 24>}
   slot {class=OverlaySlot, properties {horizontalAlignment=HAlign_Fill, verticalAlignment=VAlign_Bottom}}
  }
  VerticalBox_6 {
   widgetClassPath=&VerticalBox
   parent=BorderDisclaimer
   widget=VerticalBox_6
   slot {class=BorderSlot, properties {padding=<ltrb 0 0 0 24>}}
  }
  CommonTextBlock_VbucksLegal {
   widgetClassPath=&CommonTextBlock
   bIsVariable=true
   parent=VerticalBox_6
   widget=CommonTextBlock_VbucksLegal
   properties {
    autoWrapText=true
    justification=Center
    mobileFontSizeMultiplier=1.2
    text=All V-bucks purchased on one platform may not be redeemable on other platforms.
    colorAndOpacity {colorUseRule=UseColor_Specified, specifiedColor=<rgba 0.0844 0.0953 0.2384 1>}
    font {
     bForceMonospaced=false
     bMaterialIsStencil=false
     fontMaterial=null
     fontObject=&BurbankSmall
     letterSpacing=0
     monospacedWidth=1
     size=16
     skewAmount=0
     typefaceFontName=Bold
     outlineSettings {
      bApplyOutlineToDropShadows=false
      bMiteredCorners=false
      bSeparateFillAlpha=false
      outlineMaterial=null
      outlineSize=0
      outlineColor=<rgba 0 0 0 1>
     }
    }
    shadowOffset=<0 0>
   }
   slot {class=VerticalBoxSlot, properties {verticalAlignment=VAlign_Top}}
  }
  Text_UndoBottomBarWarning {
   widgetClassPath=&CommonTextBlock
   bInherited=true
   bIsVariable=true
   parent=VerticalBox_6
   widget=Text_UndoBottomBarWarning
   properties {
    justification=Center
    mobileFontSizeMultiplier=1.2
    text=Cancel Purchase is available until you equip or use your purchase during gameplay. Only items purchased within the last 24 hours are eligible for Cancel Purchase.
    visibility=Collapsed
    wrapTextAt=1000
    colorAndOpacity {colorUseRule=UseColor_Specified, specifiedColor=<rgba 0.4 0.4 0.4 1>}
    font {
     bForceMonospaced=false
     bMaterialIsStencil=false
     fontMaterial=null
     fontObject=&BurbankSmall
     letterSpacing=0
     monospacedWidth=1
     size=16
     skewAmount=0
     typefaceFontName=Bold
     outlineSettings {
      bApplyOutlineToDropShadows=false
      bMiteredCorners=false
      bSeparateFillAlpha=false
      outlineMaterial=null
      outlineSize=0
      outlineColor=<rgba 0 0 0 1>
     }
    }
    shadowOffset=<0 0>
   }
   slot {class=VerticalBoxSlot, properties {verticalAlignment=VAlign_Bottom}}
  }
  mobile_alignment {widgetClassPath=&Overlay, bIsVariable=true, parent=VerticalBox_6, widget=mobile_alignment, slot {class=VerticalBoxSlot}}
  VerticalBox_0 {
   widgetClassPath=&VerticalBox
   parent=mobile_alignment
   widget=VerticalBox_0
   slot {class=OverlaySlot, properties {horizontalAlignment=HAlign_Fill}}
  }
  CosmeticGate_Warning {
   widgetClassPath=/FortUILibrary/UIKit/InfoBox/WBP_UIKit_InfoBox
   bIsVariable=true
   parent=VerticalBox_0
   widget=CosmeticGate_Warning
   properties {showIcon=false}
   slot {class=VerticalBoxSlot, properties {horizontalAlignment=HAlign_Center, padding=<ltrb -8 8 0 8>}}
  }
  Text_PaymentLegalText {
   widgetClassPath=&CommonTextBlock
   bIsVariable=true
   parent=VerticalBox_0
   widget=Text_PaymentLegalText
   properties {
    text="[PH] Subject to applicable taxes."
    visibility=Collapsed
    wrapTextAt=950
    colorAndOpacity {colorUseRule=UseColor_Specified, specifiedColor=<rgba 0.4 0.4 0.4 1>}
    font {
     bForceMonospaced=false
     bMaterialIsStencil=false
     fontMaterial=null
     fontObject=&BurbankSmall
     letterSpacing=0
     monospacedWidth=1
     size=13
     skewAmount=0
     typefaceFontName=Bold
     outlineSettings {
      bApplyOutlineToDropShadows=false
      bMiteredCorners=false
      bSeparateFillAlpha=false
      outlineMaterial=null
      outlineSize=0
      outlineColor=<rgba 0 0 0 1>
     }
    }
    shadowOffset=<0 0>
   }
   slot {class=VerticalBoxSlot, properties {horizontalAlignment=HAlign_Center}}
  }
  Switcher_LegalText {
   widgetClassPath=&CommonVisibilitySwitcher
   bIsVariable=true
   parent=VerticalBox_0
   widget=Switcher_LegalText
   slot {class=VerticalBoxSlot, properties {horizontalAlignment=HAlign_Left}}
  }
  Text_CosmeticWarning {
   widgetClassPath=&CommonTextBlock
   bInherited=true
   bIsVariable=true
   parent=Switcher_LegalText
   widget=Text_CosmeticWarning
   properties {
    mobileFontSizeMultiplier=1.25
    text="[PH] These cosmetic items grant no competitive advantage. Outfits and wraps don't include weapons. Not all cosmetics are usable in all experiences. For more details on cosmetics variations and limitations of use, see fn.gg/cosmetics*"
    visibility=Collapsed
    wrapTextAt=700
    colorAndOpacity {colorUseRule=UseColor_Specified, specifiedColor=<rgba 0.4 0.4 0.4 1>}
    font {
     bForceMonospaced=false
     bMaterialIsStencil=false
     fontMaterial=null
     fontObject=&BurbankSmall
     letterSpacing=0
     monospacedWidth=1
     size=13
     skewAmount=0
     typefaceFontName=Bold
     outlineSettings {
      bApplyOutlineToDropShadows=false
      bMiteredCorners=false
      bSeparateFillAlpha=false
      outlineMaterial=null
      outlineSize=0
      outlineColor=<rgba 0 0 0 1>
     }
    }
    shadowOffset=<0 0>
   }
   slot {class=CommonVisibilitySwitcherSlot}
  }
  Text_RealPriceLegal {
   widgetClassPath=&CommonTextBlock
   parent=VerticalBox_0
   widget=Text_RealPriceLegal
   properties {
    text="*To buy exactly 1,000 V-Bucks, you’d pay €8.99."
    colorAndOpacity {colorUseRule=UseColor_Specified, specifiedColor=<rgba 1 1 1 0.65>}
    font {
     bForceMonospaced=false
     bMaterialIsStencil=false
     fontMaterial=null
     fontObject=&BurbankSmall
     letterSpacing=0
     monospacedWidth=1
     size=13
     skewAmount=0
     typefaceFontName=Bold
     outlineSettings {
      bApplyOutlineToDropShadows=false
      bMiteredCorners=false
      bSeparateFillAlpha=false
      outlineMaterial=null
      outlineSize=0
      outlineColor=<rgba 0 0 0 1>
     }
    }
    shadowOffset=<0 0>
   }
   slot {class=VerticalBoxSlot, properties {horizontalAlignment=HAlign_Left, padding=<ltrb 0 8 0 0>}}
  }
  SizeBox_SidePannel {
   widgetClassPath=&SizeBox
   parent=Overlay_4
   widget=SizeBox_SidePannel
   properties {bOverride_HeightOverride=true, bOverride_WidthOverride=true, heightOverride=932, widthOverride=700}
   slot {class=OverlaySlot, properties {verticalAlignment=VAlign_Fill}}
  }
  Overlay_0 {widgetClassPath=&Overlay, parent=SizeBox_SidePannel, widget=Overlay_0, slot {class=SizeBoxSlot}}
  OverlayGradientDetails {
   widgetClassPath=&Image
   bIsVariable=true
   parent=Overlay_0
   widget=OverlayGradientDetails
   properties {brush {
     bIsDynamicallyLoaded=false
     drawAs=Box
     imageType=NoImage
     mirroring=NoMirror
     resourceName=None
     resourceObject=/FortniteCoreUI/Frontend/StoreMain/Art/M_UI_RMT_SidePanel_Inst
     tiling=NoTile
     imageSize=<512 512>
     margin=<ltrb 0 0 0.25 0>
     outlineSettings {
      bUseBrushTransparency=false
      roundingType=HalfHeightRadius
      width=0
      color {colorUseRule=UseColor_Specified, specifiedColor=<rgba 0 0 0 0>}
      cornerRadii=<0 0 0 0>
     }
     tintColor {colorUseRule=UseColor_Specified, specifiedColor=<rgba 1 1 1 1>}
     uVRegion {bIsValid=false, max=<0 0>, min=<0 0>}
    }}
   slot {class=OverlaySlot, properties {horizontalAlignment=HAlign_Fill, verticalAlignment=VAlign_Fill, padding=<ltrb -128 -128 0 0>}}
  }
  Image_Badge {
   widgetClassPath=&Image
   bIsVariable=true
   parent=Overlay_0
   widget=Image_Badge
   properties {
    visibility=Collapsed
    brush {
     bIsDynamicallyLoaded=false
     drawAs=Image
     imageType=NoImage
     mirroring=NoMirror
     resourceName=None
     resourceObject=/FortniteCoreUI/Frontend/StoreMain/Art/M_Ui_AnnualPass_Upsell
     tiling=NoTile
     imageSize=<512 512>
     margin=<ltrb 0 0 0 0>
     outlineSettings {
      bUseBrushTransparency=false
      roundingType=HalfHeightRadius
      width=0
      color {colorUseRule=UseColor_Specified, specifiedColor=<rgba 0 0 0 0>}
      cornerRadii=<0 0 0 0>
     }
     tintColor {colorUseRule=UseColor_Specified, specifiedColor=<rgba 1 1 1 0.5>}
     uVRegion {bIsValid=false, max=<0 0>, min=<0 0>}
    }
    renderTransform {angle=0, scale=<1 1>, shear=<0 0>, translation=<-32 -64>}
   }
   slot {class=OverlaySlot, properties {horizontalAlignment=HAlign_Center}}
  }
  WBP_CaptureForPostBufferUpdate {
   widgetClassPath=&WBP_CaptureForPostBufferUpdate
   bIsVariable=true
   parent=Overlay_0
   slot {class=OverlaySlot, properties {horizontalAlignment=HAlign_Fill, verticalAlignment=VAlign_Fill}}
  }
  VerticalBox_Content {
   widgetClassPath=&VerticalBox
   bIsVariable=true
   parent=Overlay_0
   widget=VerticalBox_Content
   slot {class=OverlaySlot, properties {horizontalAlignment=HAlign_Fill, verticalAlignment=VAlign_Fill, padding=<ltrb 50 50 112 24>}}
  }
  VerticalBox_2 {
   widgetClassPath=&VerticalBox
   parent=VerticalBox_Content
   widget=VerticalBox_2
   slot {class=VerticalBoxSlot, properties {size {sizeRule=Fill, value=1}}}
  }
  TextName {
   widgetClassPath=&CommonTextBlock
   bIsVariable=true
   parent=VerticalBox_2
   widget=TextName
   properties {
    autoWrapText=true
    justification=Center
    lineHeightPercentage=0.85
    text=PLACEHOLDER PACK TITLE
    textTransformPolicy=ToUpper
    font {
     bForceMonospaced=false
     bMaterialIsStencil=false
     fontMaterial=null
     fontObject=&Burbank
     letterSpacing=0
     monospacedWidth=1
     size=40
     skewAmount=0
     typefaceFontName=BurbankRegular-Black
     outlineSettings {
      bApplyOutlineToDropShadows=false
      bMiteredCorners=false
      bSeparateFillAlpha=false
      outlineMaterial=null
      outlineSize=0
      outlineColor=<rgba 0 0 0 1>
     }
    }
    margin=<ltrb 0 0 0 8>
    shadowOffset=<0 0>
   }
   slot {class=VerticalBoxSlot}
  }
  CommonRichTextBlock_SubTitle {
   widgetClassPath=&CommonRichTextBlock
   bIsVariable=true
   parent=VerticalBox_2
   widget=CommonRichTextBlock_SubTitle
   properties {
    autoWrapText=true
    bOverrideDefaultStyle=true
    justification=Center
    text="this is a <text color=\"FFFFFFFF\" size=\"28\">Placeholder</> subtitles"
    textTransformPolicy=ToUpper
    defaultTextStyleOverride {
     overflowPolicy=Clip
     transformPolicy=None
     colorAndOpacity {colorUseRule=UseColor_Specified, specifiedColor=<rgba 1 1 1 0.75>}
     font {
      bForceMonospaced=false
      bMaterialIsStencil=false
      fontMaterial=null
      fontObject=&Burbank
      letterSpacing=0
      monospacedWidth=1
      size=24
      skewAmount=0
      typefaceFontName=BurbankRegular-Black
      outlineSettings {
       bApplyOutlineToDropShadows=false
       bMiteredCorners=false
       bSeparateFillAlpha=false
       outlineMaterial=null
       outlineSize=0
       outlineColor=<rgba 0 0 0 1>
      }
     }
     highlightColor {colorUseRule=UseColor_Specified, specifiedColor=<rgba 0 0 0 1>}
     highlightShape {
      bIsDynamicallyLoaded=false
      drawAs=Image
      imageType=NoImage
      mirroring=NoMirror
      resourceName=None
      resourceObject=null
      tiling=NoTile
      imageSize=<32 32>
      margin=<ltrb 0 0 0 0>
      outlineSettings {
       bUseBrushTransparency=false
       roundingType=HalfHeightRadius
       width=0
       color {colorUseRule=UseColor_Specified, specifiedColor=<rgba 0 0 0 0>}
       cornerRadii=<0 0 0 0>
      }
      tintColor {colorUseRule=UseColor_Specified, specifiedColor=<rgba 1 1 1 1>}
      uVRegion {bIsValid=false, max=<0 0>, min=<0 0>}
     }
     selectedBackgroundColor {colorUseRule=UseColor_Foreground, specifiedColor=<rgba 1 0 1 1>}
     shadowColorAndOpacity=<rgba 0 0 0 1>
     shadowOffset=<0 0>
     strikeBrush {
      bIsDynamicallyLoaded=false
      drawAs=Image
      imageType=NoImage
      mirroring=NoMirror
      resourceName=None
      resourceObject=null
      tiling=NoTile
      imageSize=<32 32>
      margin=<ltrb 0 0 0 0>
      outlineSettings {
       bUseBrushTransparency=false
       roundingType=HalfHeightRadius
       width=0
       color {colorUseRule=UseColor_Specified, specifiedColor=<rgba 0 0 0 0>}
       cornerRadii=<0 0 0 0>
      }
      tintColor {colorUseRule=UseColor_Specified, specifiedColor=<rgba 1 1 1 1>}
      uVRegion {bIsValid=false, max=<0 0>, min=<0 0>}
     }
     underlineBrush {
      bIsDynamicallyLoaded=false
      drawAs=Image
      imageType=NoImage
      mirroring=NoMirror
      resourceName=None
      resourceObject=null
      tiling=NoTile
      imageSize=<32 32>
      margin=<ltrb 0 0 0 0>
      outlineSettings {
       bUseBrushTransparency=false
       roundingType=HalfHeightRadius
       width=0
       color {colorUseRule=UseColor_Specified, specifiedColor=<rgba 0 0 0 0>}
       cornerRadii=<0 0 0 0>
      }
      tintColor {colorUseRule=UseColor_Specified, specifiedColor=<rgba 1 1 1 1>}
      uVRegion {bIsValid=false, max=<0 0>, min=<0 0>}
     }
    }
    renderTransform {angle=0, scale=<1 1>, shear=<-12 0>, translation=<0 0>}
   }
   slot {class=VerticalBoxSlot}
  }
  HorizontalBox_LTM {
   widgetClassPath=&HorizontalBox
   bIsVariable=true
   parent=VerticalBox_2
   widget=HorizontalBox_LTM
   slot {class=VerticalBoxSlot, properties {verticalAlignment=VAlign_Top, padding=<ltrb 0 0 0 10>}}
  }
  Image_LTM {
   widgetClassPath=&Image
   bIsVariable=true
   parent=HorizontalBox_LTM
   widget=Image_LTM
   properties {
    visibility=HitTestInvisible
    brush {
     bIsDynamicallyLoaded=false
     drawAs=Image
     imageType=NoImage
     mirroring=NoMirror
     resourceName=None
     resourceObject=/Game/Athena/UI/Frontend/ActivityBrowser/Art/M_UI_Timer_Icon_ShopOffer
     tiling=NoTile
     imageSize=<40 40>
     margin=<ltrb 0 0 0 0>
     outlineSettings {
      bUseBrushTransparency=false
      roundingType=HalfHeightRadius
      width=0
      color {colorUseRule=UseColor_Specified, specifiedColor=<rgba 0 0 0 0>}
      cornerRadii=<0 0 0 0>
     }
     tintColor {colorUseRule=UseColor_Specified, specifiedColor=<rgba 1 1 1 1>}
     uVRegion {bIsValid=false, max=<0 0>, min=<0 0>}
    }
   }
   slot {class=HorizontalBoxSlot, properties {horizontalAlignment=HAlign_Left, verticalAlignment=VAlign_Center, padding=<ltrb 0 0 8 0>}}
  }
  RichText_Header {
   widgetClassPath=&CommonRichTextBlock
   bIsVariable=true
   parent=HorizontalBox_LTM
   widget=RichText_Header
   properties {
    autoWrapText=true
    bOverrideDefaultStyle=true
    justification=Center
    lineHeightPercentage=0.8
    text="[PH] This offer will no longer be available as of September 16th at 9:00 PM your local time (GMT-3)"
    defaultTextStyleOverride {
     overflowPolicy=MultilineEllipsis
     transformPolicy=None
     colorAndOpacity {colorUseRule=UseColor_Specified, specifiedColor=<rgba 1 1 1 0.75>}
     font {
      bForceMonospaced=false
      bMaterialIsStencil=false
      fontMaterial=null
      fontObject=&BurbankSmall
      letterSpacing=0
      monospacedWidth=1
      size=16
      skewAmount=0
      typefaceFontName=Bold
      outlineSettings {
       bApplyOutlineToDropShadows=false
       bMiteredCorners=false
       bSeparateFillAlpha=false
       outlineMaterial=null
       outlineSize=0
       outlineColor=<rgba 0 0 0 1>
      }
     }
     highlightColor {colorUseRule=UseColor_Specified, specifiedColor=<rgba 0 0 0 1>}
     highlightShape {
      bIsDynamicallyLoaded=false
      drawAs=Image
      imageType=NoImage
      mirroring=NoMirror
      resourceName=None
      resourceObject=null
      tiling=NoTile
      imageSize=<32 32>
      margin=<ltrb 0 0 0 0>
      outlineSettings {
       bUseBrushTransparency=false
       roundingType=HalfHeightRadius
       width=0
       color {colorUseRule=UseColor_Specified, specifiedColor=<rgba 0 0 0 0>}
       cornerRadii=<0 0 0 0>
      }
      tintColor {colorUseRule=UseColor_Specified, specifiedColor=<rgba 1 1 1 1>}
      uVRegion {bIsValid=false, max=<0 0>, min=<0 0>}
     }
     selectedBackgroundColor {colorUseRule=UseColor_Foreground, specifiedColor=<rgba 1 0 1 1>}
     shadowColorAndOpacity=<rgba 0 0 0 1>
     shadowOffset=<0 0>
     strikeBrush {
      bIsDynamicallyLoaded=false
      drawAs=Image
      imageType=NoImage
      mirroring=NoMirror
      resourceName=None
      resourceObject=null
      tiling=NoTile
      imageSize=<32 32>
      margin=<ltrb 0 0 0 0>
      outlineSettings {
       bUseBrushTransparency=false
       roundingType=HalfHeightRadius
       width=0
       color {colorUseRule=UseColor_Specified, specifiedColor=<rgba 0 0 0 0>}
       cornerRadii=<0 0 0 0>
      }
      tintColor {colorUseRule=UseColor_Specified, specifiedColor=<rgba 1 1 1 1>}
      uVRegion {bIsValid=false, max=<0 0>, min=<0 0>}
     }
     underlineBrush {
      bIsDynamicallyLoaded=false
      drawAs=Image
      imageType=NoImage
      mirroring=NoMirror
      resourceName=None
      resourceObject=null
      tiling=NoTile
      imageSize=<32 32>
      margin=<ltrb 0 0 0 0>
      outlineSettings {
       bUseBrushTransparency=false
       roundingType=HalfHeightRadius
       width=0
       color {colorUseRule=UseColor_Specified, specifiedColor=<rgba 0 0 0 0>}
       cornerRadii=<0 0 0 0>
      }
      tintColor {colorUseRule=UseColor_Specified, specifiedColor=<rgba 1 1 1 1>}
      uVRegion {bIsValid=false, max=<0 0>, min=<0 0>}
     }
    }
   }
   slot {class=HorizontalBoxSlot, properties {verticalAlignment=VAlign_Center, size {sizeRule=Fill, value=1}}}
  }
  Overlay_3 {
   widgetClassPath=&Overlay
   parent=VerticalBox_2
   widget=Overlay_3
   slot {class=VerticalBoxSlot, properties {padding=<ltrb 0 12 0 0>, size {sizeRule=Fill, value=1}}}
  }
  VerticalBox_5 {
   widgetClassPath=&VerticalBox
   parent=Overlay_3
   widget=VerticalBox_5
   properties {clipping=ClipToBounds}
   slot {class=OverlaySlot, properties {horizontalAlignment=HAlign_Fill, padding=<ltrb 0 8 0 8>}}
  }
  ScrollBox_Details {
   widgetClassPath=&ScrollBox
   bInherited=true
   bIsVariable=true
   parent=VerticalBox_5
   widget=ScrollBox_Details
   properties {
    clipping=Inherit
    widgetBarStyle {
     thickness=16
     draggedThumbImage {
      bIsDynamicallyLoaded=false
      drawAs=Box
      imageType=FullColor
      mirroring=NoMirror
      resourceName=../../../Engine/Content/Slate/Common/Scrollbar_Thumb.png
      resourceObject=null
      tiling=NoTile
      imageSize=<0 0>
      margin=<ltrb 0.25 0.25 0.25 0.25>
      outlineSettings {
       bUseBrushTransparency=false
       roundingType=HalfHeightRadius
       width=0
       color {colorUseRule=UseColor_Specified, specifiedColor=<rgba 0 0 0 0>}
       cornerRadii=<0 0 0 0>
      }
      tintColor {colorUseRule=UseColor_Specified, specifiedColor=<rgba 1 1 1 1>}
      uVRegion {bIsValid=false, max=<0 0>, min=<0 0>}
     }
     horizontalBackgroundImage {
      bIsDynamicallyLoaded=false
      drawAs=NoDrawType
      imageType=NoImage
      mirroring=NoMirror
      resourceName=None
      resourceObject=null
      tiling=NoTile
      imageSize=<0 0>
      margin=<ltrb 0 0 0 0>
      outlineSettings {
       bUseBrushTransparency=false
       roundingType=HalfHeightRadius
       width=0
       color {colorUseRule=UseColor_Specified, specifiedColor=<rgba 0 0 0 0>}
       cornerRadii=<0 0 0 0>
      }
      tintColor {colorUseRule=UseColor_Specified, specifiedColor=<rgba 1 1 1 1>}
      uVRegion {bIsValid=false, max=<0 0>, min=<0 0>}
     }
     horizontalBottomSlotImage {
      bIsDynamicallyLoaded=false
      drawAs=Image
      imageType=FullColor
      mirroring=NoMirror
      resourceName=../../../Engine/Content/Slate/Common/Scrollbar_Background_Horizontal.png
      resourceObject=null
      tiling=NoTile
      imageSize=<8 8>
      margin=<ltrb 0 0 0 0>
      outlineSettings {
       bUseBrushTransparency=false
       roundingType=HalfHeightRadius
       width=0
       color {colorUseRule=UseColor_Specified, specifiedColor=<rgba 0 0 0 0>}
       cornerRadii=<0 0 0 0>
      }
      tintColor {colorUseRule=UseColor_Specified, specifiedColor=<rgba 1 1 1 1>}
      uVRegion {bIsValid=false, max=<0 0>, min=<0 0>}
     }
     horizontalTopSlotImage {
      bIsDynamicallyLoaded=false
      drawAs=Image
      imageType=FullColor
      mirroring=NoMirror
      resourceName=../../../Engine/Content/Slate/Common/Scrollbar_Background_Horizontal.png
      resourceObject=null
      tiling=NoTile
      imageSize=<8 8>
      margin=<ltrb 0 0 0 0>
      outlineSettings {
       bUseBrushTransparency=false
       roundingType=HalfHeightRadius
       width=0
       color {colorUseRule=UseColor_Specified, specifiedColor=<rgba 0 0 0 0>}
       cornerRadii=<0 0 0 0>
      }
      tintColor {colorUseRule=UseColor_Specified, specifiedColor=<rgba 1 1 1 1>}
      uVRegion {bIsValid=false, max=<0 0>, min=<0 0>}
     }
     hoveredThumbImage {
      bIsDynamicallyLoaded=false
      drawAs=Box
      imageType=FullColor
      mirroring=NoMirror
      resourceName=../../../Engine/Content/Slate/Common/Scrollbar_Thumb.png
      resourceObject=null
      tiling=NoTile
      imageSize=<0 0>
      margin=<ltrb 0.25 0.25 0.25 0.25>
      outlineSettings {
       bUseBrushTransparency=false
       roundingType=HalfHeightRadius
       width=0
       color {colorUseRule=UseColor_Specified, specifiedColor=<rgba 0 0 0 0>}
       cornerRadii=<0 0 0 0>
      }
      tintColor {colorUseRule=UseColor_Specified, specifiedColor=<rgba 1 1 1 1>}
      uVRegion {bIsValid=false, max=<0 0>, min=<0 0>}
     }
     normalThumbImage {
      bIsDynamicallyLoaded=false
      drawAs=Box
      imageType=FullColor
      mirroring=NoMirror
      resourceName=../../../Engine/Content/Slate/Common/Scrollbar_Thumb.png
      resourceObject=null
      tiling=NoTile
      imageSize=<0 0>
      margin=<ltrb 0.25 0.25 0.25 0.25>
      outlineSettings {
       bUseBrushTransparency=false
       roundingType=HalfHeightRadius
       width=0
       color {colorUseRule=UseColor_Specified, specifiedColor=<rgba 0 0 0 0>}
       cornerRadii=<0 0 0 0>
      }
      tintColor {colorUseRule=UseColor_Specified, specifiedColor=<rgba 1 1 1 1>}
      uVRegion {bIsValid=false, max=<0 0>, min=<0 0>}
     }
     verticalBackgroundImage {
      bIsDynamicallyLoaded=false
      drawAs=Image
      imageType=NoImage
      mirroring=NoMirror
      resourceName=None
      resourceObject=null
      tiling=NoTile
      imageSize=<0 0>
      margin=<ltrb 0 0 0 0>
      outlineSettings {
       bUseBrushTransparency=false
       roundingType=HalfHeightRadius
       width=0
       color {colorUseRule=UseColor_Specified, specifiedColor=<rgba 0 0 0 0>}
       cornerRadii=<0 0 0 0>
      }
      tintColor {colorUseRule=UseColor_Specified, specifiedColor=<rgba 0 0 0 1>}
      uVRegion {bIsValid=false, max=<0 0>, min=<0 0>}
     }
     verticalBottomSlotImage {
      bIsDynamicallyLoaded=false
      drawAs=NoDrawType
      imageType=FullColor
      mirroring=NoMirror
      resourceName=../../../Engine/Content/Slate/Common/Scrollbar_Background_Vertical.png
      resourceObject=null
      tiling=NoTile
      imageSize=<8 8>
      margin=<ltrb 0 0 0 0>
      outlineSettings {
       bUseBrushTransparency=false
       roundingType=HalfHeightRadius
       width=0
       color {colorUseRule=UseColor_Specified, specifiedColor=<rgba 0 0 0 0>}
       cornerRadii=<0 0 0 0>
      }
      tintColor {colorUseRule=UseColor_Specified, specifiedColor=<rgba 1 1 1 1>}
      uVRegion {bIsValid=false, max=<0 0>, min=<0 0>}
     }
     verticalTopSlotImage {
      bIsDynamicallyLoaded=false
      drawAs=NoDrawType
      imageType=FullColor
      mirroring=NoMirror
      resourceName=../../../Engine/Content/Slate/Common/Scrollbar_Background_Vertical.png
      resourceObject=null
      tiling=NoTile
      imageSize=<8 8>
      margin=<ltrb 0 0 0 0>
      outlineSettings {
       bUseBrushTransparency=false
       roundingType=HalfHeightRadius
       width=0
       color {colorUseRule=UseColor_Specified, specifiedColor=<rgba 0 0 0 0>}
       cornerRadii=<0 0 0 0>
      }
      tintColor {colorUseRule=UseColor_Specified, specifiedColor=<rgba 1 1 1 1>}
      uVRegion {bIsValid=false, max=<0 0>, min=<0 0>}
     }
    }
    widgetStyle {
     barThickness=9
     bottomShadowBrush {
      bIsDynamicallyLoaded=false
      drawAs=NoDrawType
      imageType=FullColor
      mirroring=NoMirror
      resourceName=../../../Engine/Content/Slate/Common/ScrollBoxShadowBottom.png
      resourceObject=null
      tiling=NoTile
      imageSize=<16 8>
      margin=<ltrb 0.5 0 0.5 1>
      outlineSettings {
       bUseBrushTransparency=false
       roundingType=HalfHeightRadius
       width=0
       color {colorUseRule=UseColor_Specified, specifiedColor=<rgba 0 0 0 0>}
       cornerRadii=<0 0 0 0>
      }
      tintColor {colorUseRule=UseColor_Specified, specifiedColor=<rgba 1 1 1 1>}
      uVRegion {bIsValid=false, max=<0 0>, min=<0 0>}
     }
     horizontalScrolledContentPadding=<ltrb 0 0 1 0>
     leftShadowBrush {
      bIsDynamicallyLoaded=false
      drawAs=Box
      imageType=FullColor
      mirroring=NoMirror
      resourceName=../../../Engine/Content/Slate/Common/ScrollBoxShadowLeft.png
      resourceObject=null
      tiling=NoTile
      imageSize=<8 16>
      margin=<ltrb 1 0.5 0 0.5>
      outlineSettings {
       bUseBrushTransparency=false
       roundingType=HalfHeightRadius
       width=0
       color {colorUseRule=UseColor_Specified, specifiedColor=<rgba 0 0 0 0>}
       cornerRadii=<0 0 0 0>
      }
      tintColor {colorUseRule=UseColor_Specified, specifiedColor=<rgba 1 1 1 1>}
      uVRegion {bIsValid=false, max=<0 0>, min=<0 0>}
     }
     rightShadowBrush {
      bIsDynamicallyLoaded=false
      drawAs=Box
      imageType=FullColor
      mirroring=NoMirror
      resourceName=../../../Engine/Content/Slate/Common/ScrollBoxShadowRight.png
      resourceObject=null
      tiling=NoTile
      imageSize=<8 16>
      margin=<ltrb 0 0.5 1 0.5>
      outlineSettings {
       bUseBrushTransparency=false
       roundingType=HalfHeightRadius
       width=0
       color {colorUseRule=UseColor_Specified, specifiedColor=<rgba 0 0 0 0>}
       cornerRadii=<0 0 0 0>
      }
      tintColor {colorUseRule=UseColor_Specified, specifiedColor=<rgba 1 1 1 1>}
      uVRegion {bIsValid=false, max=<0 0>, min=<0 0>}
     }
     topShadowBrush {
      bIsDynamicallyLoaded=false
      drawAs=NoDrawType
      imageType=FullColor
      mirroring=NoMirror
      resourceName=../../../Engine/Content/Slate/Common/ScrollBoxShadowTop.png
      resourceObject=null
      tiling=NoTile
      imageSize=<16 8>
      margin=<ltrb 0.5 1 0.5 0>
      outlineSettings {
       bUseBrushTransparency=false
       roundingType=HalfHeightRadius
       width=0
       color {colorUseRule=UseColor_Specified, specifiedColor=<rgba 0 0 0 0>}
       cornerRadii=<0 0 0 0>
      }
      tintColor {colorUseRule=UseColor_Specified, specifiedColor=<rgba 1 1 1 1>}
      uVRegion {bIsValid=false, max=<0 0>, min=<0 0>}
     }
     verticalScrolledContentPadding=<ltrb 0 0 0 1>
    }
   }
   slot {class=VerticalBoxSlot, properties {padding=<ltrb 0 20 0 20>, size {sizeRule=Fill, value=1}}}
  }
  DynamicEntryBox_IncludedItems {
   widgetClassPath=&DynamicEntryBox
   bIsVariable=true
   parent=ScrollBox_Details
   widget=DynamicEntryBox_IncludedItems
   properties {
    entryBoxType=Vertical
    entryWidgetClass=&/RMT_Offer_DetailedDescription.RMT_Offer_DetailedDescription_C
    numDesignerPreviewEntries=1
   }
   slot {class=ScrollBoxSlot}
  }
  CommonRichTextBlock_Legal {
   widgetClassPath=&CommonRichTextBlock
   bIsVariable=true
   parent=ScrollBox_Details
   widget=CommonRichTextBlock_Legal
   properties {
    autoWrapText=true
    bOverrideDefaultStyle=true
    text=Legal message.
    defaultTextStyleOverride {
     overflowPolicy=Clip
     transformPolicy=None
     colorAndOpacity {colorUseRule=UseColor_Specified, specifiedColor=<rgba 1 1 1 0.75>}
     font {
      bForceMonospaced=false
      bMaterialIsStencil=false
      fontMaterial=null
      fontObject=&BurbankSmall
      letterSpacing=0
      monospacedWidth=1
      size=16
      skewAmount=0
      typefaceFontName=Bold
      outlineSettings {
       bApplyOutlineToDropShadows=false
       bMiteredCorners=false
       bSeparateFillAlpha=false
       outlineMaterial=null
       outlineSize=0
       outlineColor=<rgba 0 0 0 1>
      }
     }
     highlightColor {colorUseRule=UseColor_Specified, specifiedColor=<rgba 0 0 0 1>}
     highlightShape {
      bIsDynamicallyLoaded=false
      drawAs=Image
      imageType=NoImage
      mirroring=NoMirror
      resourceName=None
      resourceObject=null
      tiling=NoTile
      imageSize=<32 32>
      margin=<ltrb 0 0 0 0>
      outlineSettings {
       bUseBrushTransparency=false
       roundingType=HalfHeightRadius
       width=0
       color {colorUseRule=UseColor_Specified, specifiedColor=<rgba 0 0 0 0>}
       cornerRadii=<0 0 0 0>
      }
      tintColor {colorUseRule=UseColor_Specified, specifiedColor=<rgba 1 1 1 1>}
      uVRegion {bIsValid=false, max=<0 0>, min=<0 0>}
     }
     selectedBackgroundColor {colorUseRule=UseColor_Foreground, specifiedColor=<rgba 1 0 1 1>}
     shadowColorAndOpacity=<rgba 0 0 0 1>
     shadowOffset=<0 0>
     strikeBrush {
      bIsDynamicallyLoaded=false
      drawAs=Image
      imageType=NoImage
      mirroring=NoMirror
      resourceName=None
      resourceObject=null
      tiling=NoTile
      imageSize=<32 32>
      margin=<ltrb 0 0 0 0>
      outlineSettings {
       bUseBrushTransparency=false
       roundingType=HalfHeightRadius
       width=0
       color {colorUseRule=UseColor_Specified, specifiedColor=<rgba 0 0 0 0>}
       cornerRadii=<0 0 0 0>
      }
      tintColor {colorUseRule=UseColor_Specified, specifiedColor=<rgba 1 1 1 1>}
      uVRegion {bIsValid=false, max=<0 0>, min=<0 0>}
     }
     underlineBrush {
      bIsDynamicallyLoaded=false
      drawAs=Image
      imageType=NoImage
      mirroring=NoMirror
      resourceName=None
      resourceObject=null
      tiling=NoTile
      imageSize=<32 32>
      margin=<ltrb 0 0 0 0>
      outlineSettings {
       bUseBrushTransparency=false
       roundingType=HalfHeightRadius
       width=0
       color {colorUseRule=UseColor_Specified, specifiedColor=<rgba 0 0 0 0>}
       cornerRadii=<0 0 0 0>
      }
      tintColor {colorUseRule=UseColor_Specified, specifiedColor=<rgba 1 1 1 1>}
      uVRegion {bIsValid=false, max=<0 0>, min=<0 0>}
     }
    }
   }
   slot {class=ScrollBoxSlot, properties {padding=<ltrb 10 16 10 0>}}
  }
  Image_Separator_Up {
   widgetClassPath=&Image
   bIsVariable=true
   parent=Overlay_3
   widget=Image_Separator_Up
   properties {
    brush {
     bIsDynamicallyLoaded=false
     drawAs=Box
     imageType=NoImage
     mirroring=NoMirror
     resourceName=None
     resourceObject=/Game/Athena/UI/ReplayBrowser/M_Ui_Skewed_Vfor9slices_Inst1
     tiling=NoTile
     imageSize=<16 16>
     margin=<ltrb 0.5 0.5 0.5 0.5>
     outlineSettings {
      bUseBrushTransparency=false
      roundingType=HalfHeightRadius
      width=0
      color {colorUseRule=UseColor_Specified, specifiedColor=<rgba 0 0 0 0>}
      cornerRadii=<0 0 0 0>
     }
     tintColor {colorUseRule=UseColor_Specified, specifiedColor=<rgba 1 1 1 1>}
     uVRegion {bIsValid=false, max=<0 0>, min=<0 0>}
    }
    colorAndOpacity=<rgba 0.0844 0.0953 0.2384 1>
   }
   slot {class=OverlaySlot, properties {horizontalAlignment=HAlign_Fill}}
  }
  Image_Separator_Down {
   widgetClassPath=&Image
   bIsVariable=true
   parent=Overlay_3
   widget=Image_Separator_Down
   properties {
    brush {
     bIsDynamicallyLoaded=false
     drawAs=Box
     imageType=NoImage
     mirroring=NoMirror
     resourceName=None
     resourceObject=&M_Ui_Skewed_Vfor9slices_Inst1
     tiling=NoTile
     imageSize=<16 16>
     margin=<ltrb 0.5 0.5 0.5 0.5>
     outlineSettings {
      bUseBrushTransparency=false
      roundingType=HalfHeightRadius
      width=0
      color {colorUseRule=UseColor_Specified, specifiedColor=<rgba 0 0 0 0>}
      cornerRadii=<0 0 0 0>
     }
     tintColor {colorUseRule=UseColor_Specified, specifiedColor=<rgba 1 1 1 1>}
     uVRegion {bIsValid=false, max=<0 0>, min=<0 0>}
    }
    colorAndOpacity=<rgba 0.0844 0.0953 0.2384 1>
    renderTransform {angle=0, scale=<-1 1>, shear=<0 0>, translation=<0 0>}
   }
   slot {class=OverlaySlot, properties {horizontalAlignment=HAlign_Fill, verticalAlignment=VAlign_Bottom}}
  }
  VerticalBox_4 {
   widgetClassPath=&VerticalBox
   parent=VerticalBox_Content
   widget=VerticalBox_4
   slot {class=VerticalBoxSlot, properties {verticalAlignment=VAlign_Bottom}}
  }
  MTXArea {
   widgetClassPath=&SizeBox
   bIsVariable=true
   parent=VerticalBox_4
   widget=MTXArea
   properties {bOverride_HeightOverride=true, heightOverride=110}
   slot {class=VerticalBoxSlot, properties {padding=<ltrb 0 -8 0 0>}}
  }
  CommonBorder_VBucks {
   widgetClassPath=&CommonBorder
   bIsVariable=true
   parent=MTXArea
   widget=CommonBorder_VBucks
   properties {
    background {
     bIsDynamicallyLoaded=false
     drawAs=Box
     imageType=NoImage
     mirroring=NoMirror
     resourceName=None
     resourceObject=/Game/Athena/UI/ReplayBrowser/M_Ui_Skewed_Vfor9slices_flatbottom
     tiling=NoTile
     imageSize=<64 64>
     margin=<ltrb 0.25 0 0.25 0>
     outlineSettings {
      bUseBrushTransparency=false
      roundingType=HalfHeightRadius
      width=0
      color {colorUseRule=UseColor_Specified, specifiedColor=<rgba 0 0 0 0>}
      cornerRadii=<0 0 0 0>
     }
     tintColor {colorUseRule=UseColor_Specified, specifiedColor=<rgba 1 1 1 1>}
     uVRegion {bIsValid=false, max=<0 0>, min=<0 0>}
    }
    brushColor=<rgba 0.0844 0.0953 0.2384 1>
    padding=<ltrb 0 0 0 0>
   }
   slot {class=SizeBoxSlot}
  }
  HorizontalBox_3 {
   widgetClassPath=&HorizontalBox
   parent=CommonBorder_VBucks
   widget=HorizontalBox_3
   slot {class=BorderSlot, properties {padding=<ltrb 0 0 0 0>}}
  }
  Spacer_274 {
   widgetClassPath=&Spacer
   parent=HorizontalBox_3
   widget=Spacer_274
   slot {class=HorizontalBoxSlot, properties {size {sizeRule=Fill, value=1}}}
  }
  VerticalBox_8 {
   widgetClassPath=&VerticalBox
   parent=HorizontalBox_3
   widget=VerticalBox_8
   slot {class=HorizontalBoxSlot, properties {horizontalAlignment=HAlign_Center, verticalAlignment=VAlign_Bottom}}
  }
  CommonText_BaseQuantity {
   widgetClassPath=&CommonTextBlock
   bIsVariable=true
   parent=VerticalBox_8
   widget=CommonText_BaseQuantity
   properties {
    justification=Center
    text=000
    font {
     bForceMonospaced=false
     bMaterialIsStencil=false
     fontMaterial=null
     fontObject=&BurbankSmall
     letterSpacing=0
     monospacedWidth=1
     size=24
     skewAmount=0
     typefaceFontName=Black
     outlineSettings {
      bApplyOutlineToDropShadows=false
      bMiteredCorners=false
      bSeparateFillAlpha=false
      outlineMaterial=null
      outlineSize=0
      outlineColor=<rgba 0 0 0 1>
     }
    }
    shadowOffset=<0 0>
   }
   slot {class=VerticalBoxSlot, properties {verticalAlignment=VAlign_Bottom, padding=<ltrb 0 4 0 0>}}
  }
  CommonText_Bucks {
   widgetClassPath=&CommonTextBlock
   parent=VerticalBox_8
   widget=CommonText_Bucks
   properties {
    justification=Center
    text=V-Bucks
    textTransformPolicy=ToUpper
    font {
     bForceMonospaced=false
     bMaterialIsStencil=false
     fontMaterial=null
     fontObject=&BurbankSmall
     letterSpacing=0
     monospacedWidth=1
     size=16
     skewAmount=0
     typefaceFontName=Black
     outlineSettings {
      bApplyOutlineToDropShadows=false
      bMiteredCorners=false
      bSeparateFillAlpha=false
      outlineMaterial=null
      outlineSize=0
      outlineColor=<rgba 0 0 0 1>
     }
    }
    renderTransform {angle=0, scale=<1 1>, shear=<0 0>, translation=<0 -4>}
    shadowOffset=<0 0>
   }
   slot {class=VerticalBoxSlot, properties {verticalAlignment=VAlign_Top}}
  }
  BonusMTXArea {
   widgetClassPath=&HorizontalBox
   bIsVariable=true
   parent=HorizontalBox_3
   widget=BonusMTXArea
   slot {class=HorizontalBoxSlot, properties {verticalAlignment=VAlign_Bottom}}
  }
  CommonTextBlock_7 {
   widgetClassPath=&CommonTextBlock
   parent=BonusMTXArea
   widget=CommonTextBlock_7
   properties {
    text=+
    font {
     bForceMonospaced=false
     bMaterialIsStencil=false
     fontMaterial=null
     fontObject=&BurbankSmall
     letterSpacing=0
     monospacedWidth=1
     size=32
     skewAmount=0
     typefaceFontName=Black
     outlineSettings {
      bApplyOutlineToDropShadows=false
      bMiteredCorners=false
      bSeparateFillAlpha=false
      outlineMaterial=null
      outlineSize=0
      outlineColor=<rgba 0 0 0 1>
     }
    }
    renderTransform {angle=0, scale=<1 1>, shear=<-12 0>, translation=<0 0>}
    shadowOffset=<0 0>
   }
   slot {class=HorizontalBoxSlot, properties {verticalAlignment=VAlign_Center, padding=<ltrb 8 0 8 0>}}
  }
  VerticalBox_3 {
   widgetClassPath=&VerticalBox
   parent=BonusMTXArea
   widget=VerticalBox_3
   slot {class=HorizontalBoxSlot, properties {horizontalAlignment=HAlign_Left, size {sizeRule=Fill, value=1}}}
  }
  CommonText_BonusQuantity {
   widgetClassPath=&CommonTextBlock
   bIsVariable=true
   parent=VerticalBox_3
   widget=CommonText_BonusQuantity
   properties {
    justification=Center
    text=000
    colorAndOpacity {colorUseRule=UseColor_Specified, specifiedColor=<rgba 1 0.8714 0.0723 1>}
    font {
     bForceMonospaced=false
     bMaterialIsStencil=false
     fontMaterial=null
     fontObject=&BurbankSmall
     letterSpacing=0
     monospacedWidth=1
     size=24
     skewAmount=0
     typefaceFontName=Black
     outlineSettings {
      bApplyOutlineToDropShadows=false
      bMiteredCorners=false
      bSeparateFillAlpha=false
      outlineMaterial=null
      outlineSize=0
      outlineColor=<rgba 0 0 0 1>
     }
    }
    shadowOffset=<0 0>
   }
   slot {class=VerticalBoxSlot, properties {verticalAlignment=VAlign_Bottom, padding=<ltrb 0 4 0 0>}}
  }
  CommonText_Bonus {
   widgetClassPath=&CommonTextBlock
   parent=VerticalBox_3
   widget=CommonText_Bonus
   properties {
    justification=Center
    text=BONUS
    colorAndOpacity {colorUseRule=UseColor_Specified, specifiedColor=<rgba 1 0.9301 0 1>}
    font {
     bForceMonospaced=false
     bMaterialIsStencil=false
     fontMaterial=null
     fontObject=&BurbankSmall
     letterSpacing=0
     monospacedWidth=1
     size=16
     skewAmount=0
     typefaceFontName=Black
     outlineSettings {
      bApplyOutlineToDropShadows=false
      bMiteredCorners=false
      bSeparateFillAlpha=false
      outlineMaterial=null
      outlineSize=0
      outlineColor=<rgba 0 0 0 1>
     }
    }
    renderTransform {angle=0, scale=<1 1>, shear=<0 0>, translation=<0 -4>}
    shadowOffset=<0 0>
   }
   slot {class=VerticalBoxSlot, properties {verticalAlignment=VAlign_Top}}
  }
  Spacer {
   widgetClassPath=&Spacer
   parent=HorizontalBox_3
   widget=Spacer
   slot {class=HorizontalBoxSlot, properties {size {sizeRule=Fill, value=1}}}
  }
  Image_999 {
   widgetClassPath=&Image
   bIsVariable=true
   parent=HorizontalBox_3
   widget=Image_999
   properties {brush {
     bIsDynamicallyLoaded=false
     drawAs=Image
     imageType=NoImage
     mirroring=NoMirror
     resourceName=None
     resourceObject=/FortniteCoreUI/Frontend/StoreMain/Art/V-Bucks_Icon_128
     tiling=NoTile
     imageSize=<128 128>
     margin=<ltrb 0 0 0 0>
     outlineSettings {
      bUseBrushTransparency=false
      roundingType=HalfHeightRadius
      width=0
      color {colorUseRule=UseColor_Specified, specifiedColor=<rgba 0 0 0 0>}
      cornerRadii=<0 0 0 0>
     }
     tintColor {colorUseRule=UseColor_Specified, specifiedColor=<rgba 1 1 1 1>}
     uVRegion {bIsValid=false, max=<0 0>, min=<0 0>}
    }}
   slot {class=HorizontalBoxSlot, properties {verticalAlignment=VAlign_Bottom, padding=<ltrb 0 -32 16 0>}}
  }
  VerticalBox_1 {
   widgetClassPath=&VerticalBox
   parent=VerticalBox_4
   widget=VerticalBox_1
   slot {class=VerticalBoxSlot, properties {verticalAlignment=VAlign_Bottom, padding=<ltrb 0 10 0 0>}}
  }
  WBP_ItemShop_2dPDP_Price {
   widgetClassPath=&/WBP_ItemShop_2dPDP_Price
   bIsVariable=true
   parent=VerticalBox_1
   widget=WBP_ItemShop_2dPDP_Price
   properties {realPriceLegalText_Optional=&self:WidgetTree.Text_RealPriceLegal}
   slot {class=VerticalBoxSlot, properties {horizontalAlignment=HAlign_Center, padding=<ltrb 0 4 0 4>}}
  }
  Button_LoadGuard {
   widgetClassPath=/FortUILibrary/UIKit/Buttons/WBP_UIKit_Button_LoadGuard
   bIsVariable=true
   parent=VerticalBox_1
   widget=Button_LoadGuard
   properties {defaultLoadingProgress=1}
   slot {class=VerticalBoxSlot, properties {padding=<ltrb 20 8 20 8>}}
  }
  VerticalBox_9 {widgetClassPath=&VerticalBox, namedSlotHost=Button_LoadGuard, widget=VerticalBox_9}
  Button_Purchase {
   widgetClassPath=/FortUILibrary/UIKit/Buttons/WBP_UIKit_Button_Loud
   bInherited=true
   bIsVariable=true
   parent=VerticalBox_9
   widget=Button_Purchase
   properties {bRequiresHold=true, holdData=/FortUILibrary/Input/PurchaseInteraction_HoldData.PurchaseInteraction_HoldData_C, minWidth=400, text=Purchase}
   slot {class=VerticalBoxSlot}
  }
  Button_SendGift {
   widgetClassPath=&WBP_UIKit_Button_Regular
   bInherited=true
   bIsVariable=true
   parent=VerticalBox_9
   widget=Button_SendGift
   properties {minWidth=400, text=GIFT TO FRIENDS}
   slot {class=VerticalBoxSlot, properties {padding=<ltrb 0 16 0 0>}}
  }
  Button_Inspect {
   widgetClassPath=&WBP_UIKit_Button_Regular
   bInherited=true
   bIsVariable=true
   parent=VerticalBox_1
   widget=Button_Inspect
   properties {minWidth=400, text=INSPECT ITEMS}
   slot {class=VerticalBoxSlot, properties {padding=<ltrb 20 8 20 8>}}
  }
  Button_Undo {
   widgetClassPath=&WBP_UIKit_Button_Regular
   bInherited=true
   bIsVariable=true
   parent=VerticalBox_1
   widget=Button_Undo
   properties {
    bRequiresHold=true
    holdData=&PurchaseInteraction_HoldData_C
    minWidth=400
    text=CANCEL PURCHASE
   }
   slot {class=VerticalBoxSlot, properties {padding=<ltrb 20 8 20 8>}}
  }
  TextUndoAvailability {
   widgetClassPath=&CommonTextBlock
   bInherited=true
   bIsVariable=true
   parent=VerticalBox_1
   widget=TextUndoAvailability
   properties {
    autoWrapText=true
    justification=Center
    text=Cancel Purchase is available until you equip or use your purchase during gameplay. Only items purchased within the last 24 hours are eligible for Cancel Purchase.
    visibility=Collapsed
    colorAndOpacity {colorUseRule=UseColor_Specified, specifiedColor=<rgba 0.159 0.3712 0.807 1>}
    font {
     bForceMonospaced=false
     bMaterialIsStencil=false
     fontMaterial=null
     fontObject=&BurbankSmall
     letterSpacing=0
     monospacedWidth=1
     size=16
     skewAmount=0
     typefaceFontName=Bold
     outlineSettings {
      bApplyOutlineToDropShadows=false
      bMiteredCorners=false
      bSeparateFillAlpha=false
      outlineMaterial=null
      outlineSize=0
      outlineColor=<rgba 0 0 0 1>
     }
    }
    shadowOffset=<0 0>
   }
   slot {class=VerticalBoxSlot, properties {padding=<ltrb 0 8 0 0>}}
  }
  VBox_NonRefundableBullets {
   widgetClassPath=&VerticalBox
   bIsVariable=true
   parent=VerticalBox_1
   widget=VBox_NonRefundableBullets
   slot {class=VerticalBoxSlot}
  }
  HorizontalBox_Line1 {
   widgetClassPath=&HorizontalBox
   parent=VBox_NonRefundableBullets
   widget=HorizontalBox_Line1
   slot {class=VerticalBoxSlot, properties {padding=<ltrb 0 14 0 0>}}
  }
  Image_166 {
   widgetClassPath=&Image
   parent=HorizontalBox_Line1
   widget=Image_166
   properties {brush {
     bIsDynamicallyLoaded=false
     drawAs=Image
     imageType=NoImage
     mirroring=NoMirror
     resourceName=None
     resourceObject=/Game/Athena/UI/ReplayBrowser/M_Ui_Skewed_Vfor9slices_Inst_Inst
     tiling=NoTile
     imageSize=<16 16>
     margin=<ltrb 0 0 0 0>
     outlineSettings {
      bUseBrushTransparency=false
      roundingType=HalfHeightRadius
      width=0
      color {colorUseRule=UseColor_Specified, specifiedColor=<rgba 0 0 0 0>}
      cornerRadii=<0 0 0 0>
     }
     tintColor {colorUseRule=UseColor_Specified, specifiedColor=<rgba 1 1 1 1>}
     uVRegion {bIsValid=false, max=<0 0>, min=<0 0>}
    }}
   slot {class=HorizontalBoxSlot, properties {verticalAlignment=VAlign_Top, padding=<ltrb 0 8 10 0>}}
  }
  CommonTextBlock_431 {
   widgetClassPath=&CommonTextBlock
   parent=HorizontalBox_Line1
   widget=CommonTextBlock_431
   properties {
    autoWrapText=true
    bIsScrollingEnabled=false
    text=These items are not eligible for returns through Fortnite Cancel Purchase or Return Tickets.
    font {
     bForceMonospaced=false
     bMaterialIsStencil=false
     fontMaterial=null
     fontObject=&BurbankSmall
     letterSpacing=0
     monospacedWidth=1
     size=16
     skewAmount=0
     typefaceFontName=Bold
     outlineSettings {
      bApplyOutlineToDropShadows=false
      bMiteredCorners=false
      bSeparateFillAlpha=false
      outlineMaterial=null
      outlineSize=0
      outlineColor=<rgba 0 0 0 1>
     }
    }
    shadowOffset=<0 0>
   }
   slot {class=HorizontalBoxSlot, properties {size {sizeRule=Fill, value=1}}}
  }
  VBox_NonRefundableRmtBullets {
   widgetClassPath=&VerticalBox
   bIsVariable=true
   parent=VBox_NonRefundableBullets
   widget=VBox_NonRefundableRmtBullets
   slot {class=VerticalBoxSlot}
  }
  HorizontalBox_Line2_1 {
   widgetClassPath=&HorizontalBox
   parent=VBox_NonRefundableRmtBullets
   widget=HorizontalBox_Line2_1
   slot {class=VerticalBoxSlot, properties {padding=<ltrb 0 14 0 0>}}
  }
  Image_2 {
   widgetClassPath=&Image
   parent=HorizontalBox_Line2_1
   widget=Image_2
   properties {brush {
     bIsDynamicallyLoaded=false
     drawAs=Image
     imageType=NoImage
     mirroring=NoMirror
     resourceName=None
     resourceObject=&M_Ui_Skewed_Vfor9slices_Inst_Inst
     tiling=NoTile
     imageSize=<16 16>
     margin=<ltrb 0 0 0 0>
     outlineSettings {
      bUseBrushTransparency=false
      roundingType=HalfHeightRadius
      width=0
      color {colorUseRule=UseColor_Specified, specifiedColor=<rgba 0 0 0 0>}
      cornerRadii=<0 0 0 0>
     }
     tintColor {colorUseRule=UseColor_Specified, specifiedColor=<rgba 1 1 1 1>}
     uVRegion {bIsValid=false, max=<0 0>, min=<0 0>}
    }}
   slot {class=HorizontalBoxSlot, properties {verticalAlignment=VAlign_Center, padding=<ltrb 0 0 10 0>}}
  }
  CommonTextBlock_1 {
   widgetClassPath=&CommonTextBlock
   parent=HorizontalBox_Line2_1
   widget=CommonTextBlock_1
   properties {
    autoWrapText=true
    bIsScrollingEnabled=false
    text="For more information please visit fn.gg⁄platformrefunds"
    font {
     bForceMonospaced=false
     bMaterialIsStencil=false
     fontMaterial=null
     fontObject=&BurbankSmall
     letterSpacing=0
     monospacedWidth=1
     size=16
     skewAmount=0
     typefaceFontName=Bold
     outlineSettings {
      bApplyOutlineToDropShadows=false
      bMiteredCorners=false
      bSeparateFillAlpha=false
      outlineMaterial=null
      outlineSize=0
      outlineColor=<rgba 0 0 0 1>
     }
    }
    shadowOffset=<0 0>
   }
   slot {class=HorizontalBoxSlot, properties {size {sizeRule=Fill, value=1}}}
  }
  HorizontalBox_Line2 {
   widgetClassPath=&HorizontalBox
   parent=VBox_NonRefundableRmtBullets
   widget=HorizontalBox_Line2
   slot {class=VerticalBoxSlot, properties {padding=<ltrb 0 14 0 0>}}
  }
  Image {
   widgetClassPath=&Image
   parent=HorizontalBox_Line2
   widget=Image
   properties {brush {
     bIsDynamicallyLoaded=false
     drawAs=Image
     imageType=NoImage
     mirroring=NoMirror
     resourceName=None
     resourceObject=&M_Ui_Skewed_Vfor9slices_Inst_Inst
     tiling=NoTile
     imageSize=<16 16>
     margin=<ltrb 0 0 0 0>
     outlineSettings {
      bUseBrushTransparency=false
      roundingType=HalfHeightRadius
      width=0
      color {colorUseRule=UseColor_Specified, specifiedColor=<rgba 0 0 0 0>}
      cornerRadii=<0 0 0 0>
     }
     tintColor {colorUseRule=UseColor_Specified, specifiedColor=<rgba 1 1 1 1>}
     uVRegion {bIsValid=false, max=<0 0>, min=<0 0>}
    }}
   slot {class=HorizontalBoxSlot, properties {verticalAlignment=VAlign_Top, padding=<ltrb 0 8 10 0>}}
  }
  CommonTextBlock {
   widgetClassPath=&CommonTextBlock
   parent=HorizontalBox_Line2
   widget=CommonTextBlock
   properties {
    autoWrapText=true
    bIsScrollingEnabled=false
    text=You may be able to get a refund through your purchase platform.
    font {
     bForceMonospaced=false
     bMaterialIsStencil=false
     fontMaterial=null
     fontObject=&BurbankSmall
     letterSpacing=0
     monospacedWidth=1
     size=16
     skewAmount=0
     typefaceFontName=Bold
     outlineSettings {
      bApplyOutlineToDropShadows=false
      bMiteredCorners=false
      bSeparateFillAlpha=false
      outlineMaterial=null
      outlineSize=0
      outlineColor=<rgba 0 0 0 1>
     }
    }
    shadowOffset=<0 0>
   }
   slot {class=HorizontalBoxSlot, properties {size {sizeRule=Fill, value=1}}}
  }
  WBP_ContentPurchaseBlocker {
   widgetClassPath=/FortniteCoreUI/UI/ContentControls/WBP_ContentPurchaseBlocker
   bIsVariable=true
   parent=Overlay_0
   widget=WBP_ContentPurchaseBlocker
   properties {max Width=600, visibility=Collapsed}
   slot {class=OverlaySlot, properties {horizontalAlignment=HAlign_Center, verticalAlignment=VAlign_Center, padding=<ltrb 0 0 20 0>}}
  }
  Loading_Overlay {
   widgetClassPath=&Overlay
   bIsVariable=true
   parent=Overlay_4
   widget=Loading_Overlay
   properties {visibility=Collapsed}
   slot {class=OverlaySlot, properties {horizontalAlignment=HAlign_Fill, verticalAlignment=VAlign_Fill}}
  }
  WBP_UIKit_Scrim {
   widgetClassPath=/FortUILibrary/Systems/Dialog/WBP_UIKit_Scrim
   parent=Loading_Overlay
   widget=WBP_UIKit_Scrim
   properties {blur=10}
   slot {class=OverlaySlot, properties {horizontalAlignment=HAlign_Fill, verticalAlignment=VAlign_Fill}}
  }
  Loading_Throbber {
   widgetClassPath=/FortUILibrary/UIKit/Throbber/WBP_UIKit_Throbber
   parent=Loading_Overlay
   widget=Loading_Throbber
   slot {class=OverlaySlot, properties {horizontalAlignment=HAlign_Center, verticalAlignment=VAlign_Center}}
  }
  SafeZone_1 {
   widgetClassPath=&SafeZone
   bIsVariable=true
   parent=Overlay_2
   widget=SafeZone_1
   properties {padBottom=false, padLeft=false}
   slot {class=OverlaySlot, properties {horizontalAlignment=HAlign_Right, padding=<ltrb 0 24 48 0>}}
  }
  VBox_MtxContainer {widgetClassPath=&VerticalBox, bIsVariable=true, parent=SafeZone_1, widget=VBox_MtxContainer, slot {class=SafeZoneSlot}}
  Button_MTX {
   widgetClassPath=/Game/UI/Frontend/TopBar/MTXButton
   bInherited=true
   bIsVariable=true
   parent=VBox_MtxContainer
   widget=Button_MTX
   properties {
    onButtonBaseClicked=(&self:WidgetTree.Button_MTX.HandleOnButtonClicked)
    onButtonBaseDoubleClicked=(&self:WidgetTree.Button_MTX.HandleOnButtonDoubleClicked)
    onButtonBaseHovered=(&self:WidgetTree.Button_MTX.HandleOnButtonHovered)
    onButtonBaseUnhovered=(&self:WidgetTree.Button_MTX.HandleOnButtonUnhovered)
    onSelectedChangedBase=(&self:WidgetTree.Button_MTX.HandleOnSelectedChanged)
   }
   slot {class=VerticalBoxSlot, properties {horizontalAlignment=HAlign_Right}}
  }
  FortOwningPlayerSocialNameTextBlock {
   widgetClassPath=/Game/UI/Social/FortOwningPlayerSocialNameTextBlock
   parent=VBox_MtxContainer
   widget=FortOwningPlayerSocialNameTextBlock
   slot {class=VerticalBoxSlot, properties {horizontalAlignment=HAlign_Right, padding=<ltrb 0 4 0 0>}}
  }
  StoreMain_Offer_VBucks_Modal {
   widgetClassPath=/FortniteCoreUI/Frontend/StoreMain/StoreMain_Offer_VBucks_Modal
   bIsVariable=true
   parent=Overlay_2
   widget=StoreMain_Offer_VBucks_Modal
   properties {visibility=Collapsed}
   slot {class=OverlaySlot, properties {horizontalAlignment=HAlign_Fill, verticalAlignment=VAlign_Fill}}
  }
  Container_SimpleMTXDisplay {widgetClassPath=/Script/UMG.Widget, bInherited=true, widget=null}
 ]
}
