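# Small FortUserWidget wrapping a single narrow cosmetic item card, by its name a mini card for bundle items in the item-received UI.
# SetItemToRepresent passes the stored Fort Item to the card, leaving the view-model input unconnected; GetListItemObject returns an empty object.
# Play Anim - Show plays a short scale-overshoot and wobble animation, bound to ItemCardWidget_C_49 rather than ItemCard.
# A gifting sound node is present but shows no execution link.
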WidgetBlueprint /FortniteCoreUI/UI/ItemReceived/BundleItemMiniCardWidget
animations [Show {
  binding_count=1
  display_label=""
  length_seconds=0.4
  object=&self:Show
  sequence {
   bindings [ItemCardWidget_C_49 {
     id=#1
     possessed_class=ItemCardWidget_C
     tracks [{
       type=MovieScene2DTransformTrack
       sections [{
         type=MovieScene2DTransformSection
         end_frame=400
         start_frame=0
         channels [
          Translation.X {type=MovieSceneScriptingFloatChannel}
          Translation.Y {type=MovieSceneScriptingFloatChannel}
          Angle {type=MovieSceneScriptingFloatChannel, keys [[72.7167,0],[181.8167,6],[327.2667,0]]}
          Scale.X {type=MovieSceneScriptingFloatChannel, keys [[0,0],[109.0833,1.5],[300,0.93],[400,1]]}
          Scale.Y {type=MovieSceneScriptingFloatChannel, keys [[0,0],[109.0833,1.5],[300,0.93],[400,1]]}
          Shear.X {type=MovieSceneScriptingFloatChannel}
          Shear.Y {type=MovieSceneScriptingFloatChannel}
         ]
        }]
      }]
    }]
   display_rate {numerator=1000}
  }
 }]
graphs [
 EventGraph {nodes [
   SetItemToRepresent() {is=title K2Node_CustomEvent}
   > SetItemToRepresent!(
    execute
    self:Fort Cosmetic Item Card Object Reference=Get ItemCard() -> (ItemCard:Fort Cosmetic Item Card Object Reference)
    InItemToRepresent:Fort Item Object Reference=Get Fort Item() -> (Fort Item:Fort Item Object Reference)
    InItemVM:Fort Item VM Object Reference
   ) -> (then) {is=title K2Node_CallFunction}
   Play Anim - Show() {is=title K2Node_CustomEvent}
   > PlayAnimation(InAnimation=Get Show() -> (Show:Widget Animation Object Reference))
   PlaySound2D(Sound=Get SoundOnItemCardShow() -> (SoundOnItemCardShow:Sound Base Object Reference))
  ]}
 GetListItemObject {nodes [
   GetListItemObject() {is=title K2Node_FunctionEntry}
   > Return Node!(execute, ReturnValue:Object Reference)
  ]}
]
variables [
 Fort Item {type=FortItem, default=null}
 Item VM {type=FortItemVM, default=null}
 SoundOnItemCardShow {type=SoundBase, default=/Game/Sounds/Fort_UI_Sounds/FrontEnd/RewardsAndGifts/Gifting_ItemShown_Cue}
]
widgets {
 info {
  inheritedWidgetCount=0
  parentClass=/Script/FortniteUI.FortUserWidget
  rootWidgetClass=/FortniteCoreUI/Athena/UI/Frontend/CosmeticItemCard/FortCosmeticItemCard.FortCosmeticItemCard_C
  widgetCount=1
 }
 widgets [ItemCard {
   widgetClassPath=/FortniteCoreUI/Athena/UI/Frontend/CosmeticItemCard/FortCosmeticItemCard
   bIsVariable=true
   widget=ItemCard
   properties {cardWidth=88}
  }]
}
