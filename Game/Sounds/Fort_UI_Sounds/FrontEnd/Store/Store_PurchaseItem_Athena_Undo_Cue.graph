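# Sound cue for undoing an item purchase in the Athena front-end store, played on the UI sound class.
# A quality-level node at the root picks between two wave players: the standard undo wave fills the first two quality slots, and its "_opt" variant fills the last two.
# The cue overrides concurrency so that new instances are refused once the limit is reached.
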SoundCue /Game/Sounds/Fort_UI_Sounds/FrontEnd/Store/Store_PurchaseItem_Athena_Undo_Cue
root=QualityLevel_0
nodes [
 {id=QualityLevel_0, children [WavePlayer_5,WavePlayer_5,WavePlayer_6,WavePlayer_6]}
 {id=WavePlayer_5, properties {soundWaveAssetPtr=/Game/Sounds/Fort_UI_Sounds/FrontEnd/Store/wav/store_purchaseitem_Undo_athena_01}}
 {id=WavePlayer_6, properties {soundWaveAssetPtr=/Game/Sounds/Fort_UI_Sounds/FrontEnd/Store/wav/store_purchaseitem_Undo_athena_01_opt}}
]
properties {
 bOverrideConcurrency=true
 maxDistance=2097152
 soundClassObject=&Fort_UI
 volumeMultiplier=1
 concurrencyOverrides {resolutionRule=PreventNew}
}
