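# Store UI sound cue, named for the refund element appearing in the item purchase screen, played in the Fort_UI sound class.
# It reuses the Creative create-button wave, passed through a Modulator for slight upward pitch and volume variation.
# A quality-level root routes all four inputs to that same Modulator, so every level sounds alike.
# A concurrency override caps simultaneous plays, stops the oldest and scales down the volume of overlapping instances.
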SoundCue /Game/Sounds/Fort_UI_Sounds/FrontEnd/Store/Store_PurchaseItem_RefundAppear_Cue
root=QualityLevel_0
nodes [
 {id=Modulator_0, children [WavePlayer_3], properties {pitchMax=1.1, pitchMin=1, volumeMin=0.7}}
 {id=QualityLevel_0, children [Modulator_0,Modulator_0,Modulator_0,Modulator_0]}
 {id=WavePlayer_3, properties {soundWaveAssetPtr=/Game/Sounds/Creative/IngameUI/Creative_CreateButton_01}}
]
properties {
 bOverrideConcurrency=true
 maxDistance=2097152
 soundClassObject=&Fort_UI
 volumeMultiplier=1
 concurrencyOverrides {maxCount=3, resolutionRule=StopOldest, voiceStealReleaseTime=0.2, volumeScale=0.4, volumeScaleAttackTime=0.2}
}
