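# Item shop overlay that celebrates a completed purchase, or a refund, with animated item card, rarity flare, name, status text and check.
# Show Next Item updates from the indexed purchased item, selecting single, bundle or refund animations and sounds.
# Bundles keep a looping background blur, vary playback speed by item index, and reveal pre-created hidden mini-cards in a DynamicEntryBox, scrolling the grid by row.
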WidgetBlueprint /ItemShop/Athena/AthenaItemShopOfferPurchasedOverlay
animations [
 "Purchased!" {
  binding_count=7
  display_label=""
  length_seconds=3
  object="&self:Purchased!"
  sequence {
   bindings [
    BorderItemNAme {
     id=#1
     possessed_class=Border
     tracks [
      {
       type=MovieSceneColorTrack
       display_name=Content Color and Opacity
       sections [{
         type=MovieSceneColorSection
         end_frame=199
         start_frame=0
         channels [
          Color.R {type=MovieSceneScriptingFloatChannel}
          Color.G {type=MovieSceneScriptingFloatChannel}
          Color.B {type=MovieSceneScriptingFloatChannel}
          Color.A {type=MovieSceneScriptingFloatChannel, keys [[25.25,0],[49.25,1]]}
         ]
        }]
      }
      {
       type=MovieScene2DTransformTrack
       sections [{
         type=MovieScene2DTransformSection
         end_frame=199
         start_frame=0
         channels [
          Translation.X {type=MovieSceneScriptingFloatChannel, keys [[25.25,768],[49.25,-64.0017],[55.25,16],[73.25,-8],[97.25,4],[127.2483,0]]}
          Translation.Y {type=MovieSceneScriptingFloatChannel}
          Angle {type=MovieSceneScriptingFloatChannel}
          Scale.X {type=MovieSceneScriptingFloatChannel, keys [[25.25,0.25],[55.25,1]]}
          Scale.Y {type=MovieSceneScriptingFloatChannel, keys [[25.25,0.25],[55.25,1]]}
          Shear.X {type=MovieSceneScriptingFloatChannel, keys [[25.25,-88],[61.25,24],[79.25,-4],[103.2483,2],[133.25,0]]}
          Shear.Y {type=MovieSceneScriptingFloatChannel}
         ]
        }]
      }
     ]
    }
    BorderRarityFlare {
     id=#2
     possessed_class=Border
     tracks [
      {
       type=MovieSceneColorTrack
       display_name=Content Color and Opacity
       sections [{
         type=MovieSceneColorSection
         end_frame=199
         start_frame=0
         channels [
          Color.R {type=MovieSceneScriptingFloatChannel}
          Color.G {type=MovieSceneScriptingFloatChannel}
          Color.B {type=MovieSceneScriptingFloatChannel}
          Color.A {type=MovieSceneScriptingFloatChannel, keys [[12.9983,0],[27.5,1]]}
         ]
        }]
      }
      {
       type=MovieScene2DTransformTrack
       sections [{
         type=MovieScene2DTransformSection
         end_frame=199
         start_frame=0
         channels [
          Translation.X {type=MovieSceneScriptingFloatChannel}
          Translation.Y {type=MovieSceneScriptingFloatChannel}
          Angle {type=MovieSceneScriptingFloatChannel}
          Scale.X {type=MovieSceneScriptingFloatChannel, keys [[12.9983,0],[33.5,1.85],[51.5,0.9],[75.5,1.05],[111.5,0.975],[165.5,1]]}
          Scale.Y {type=MovieSceneScriptingFloatChannel, keys [[12.9983,0],[33.5,1.85],[51.5,0.9],[75.5,1.05],[111.5,0.975],[165.5,1]]}
          Shear.X {type=MovieSceneScriptingFloatChannel}
          Shear.Y {type=MovieSceneScriptingFloatChannel}
         ]
        }]
      }
     ]
    }
    BorderItemCard {
     id=#3
     possessed_class=Border
     tracks [
      {
       type=MovieSceneColorTrack
       display_name=Content Color and Opacity
       sections [{
         type=MovieSceneColorSection
         end_frame=199
         start_frame=0
         channels [
          Color.R {type=MovieSceneScriptingFloatChannel}
          Color.G {type=MovieSceneScriptingFloatChannel}
          Color.B {type=MovieSceneScriptingFloatChannel}
          Color.A {type=MovieSceneScriptingFloatChannel, keys [[0,0],[21.5,1]]}
         ]
        }]
      }
      {
       type=MovieScene2DTransformTrack
       sections [{
         type=MovieScene2DTransformSection
         end_frame=300
         start_frame=-1
         channels [
          Translation.X {type=MovieSceneScriptingFloatChannel}
          Translation.Y {type=MovieSceneScriptingFloatChannel}
          Angle {type=MovieSceneScriptingFloatChannel}
          Scale.X {type=MovieSceneScriptingFloatChannel, keys [[-0.0017,3],[21.5,0.75],[32.75,1.125],[49.25,0.95],[69.875,1]]}
          Scale.Y {type=MovieSceneScriptingFloatChannel, keys [[-0.0017,3],[21.5,0.5],[32.75,1.275],[49.25,0.85],[69.875,1]]}
          Shear.X {type=MovieSceneScriptingFloatChannel}
          Shear.Y {type=MovieSceneScriptingFloatChannel, keys [[-0.0017,8],[26,-2],[37.25,4],[53,0]]}
         ]
        }]
      }
     ]
    }
    Blur {
     id=#4
     possessed_class=BackgroundBlur
     tracks [{
       type=MovieSceneFloatTrack
       display_name=Render Opacity
       sections [{
         type=MovieSceneFloatSection
         end_frame=300
         start_frame=0
         channels [None {type=MovieSceneScriptingFloatChannel, keys [[0,1],[275,1],[300,0]]}]
        }]
      }]
    }
    BorderRarityGlow {
     id=#5
     possessed_class=Border
     tracks [{
       type=MovieSceneColorTrack
       display_name=Content Color and Opacity
       sections [{
         type=MovieSceneColorSection
         end_frame=199
         start_frame=0
         channels [
          Color.R {type=MovieSceneScriptingFloatChannel}
          Color.G {type=MovieSceneScriptingFloatChannel}
          Color.B {type=MovieSceneScriptingFloatChannel}
          Color.A {type=MovieSceneScriptingFloatChannel, keys [[37.25,0],[96.875,1]]}
         ]
        }]
      }]
    }
    BorderSuccessText {
     id=#6
     possessed_class=Border
     tracks [
      {
       type=MovieScene2DTransformTrack
       sections [{
         type=MovieScene2DTransformSection
         end_frame=199
         start_frame=0
         channels [
          Translation.X {type=MovieSceneScriptingFloatChannel, keys [[93.5,16],[111.4983,0]]}
          Translation.Y {type=MovieSceneScriptingFloatChannel, keys [[93.5,80],[111.4983,16],[117.4983,-4],[135.5,0]]}
          Angle {type=MovieSceneScriptingFloatChannel}
          Scale.X {type=MovieSceneScriptingFloatChannel, keys [[87.5,0],[105.5,0.2],[117.5,1],[123.4983,1]]}
          Scale.Y {type=MovieSceneScriptingFloatChannel, keys [[87.5,2],[105.5,2],[111.5,0.75],[123.4983,1.85],[135.5,0.95],[141.5,1]]}
          Shear.X {type=MovieSceneScriptingFloatChannel}
          Shear.Y {type=MovieSceneScriptingFloatChannel}
         ]
        }]
      }
      {
       type=MovieSceneColorTrack
       display_name=Content Color and Opacity
       sections [{
         type=MovieSceneColorSection
         end_frame=199
         start_frame=0
         channels [
          Color.R {type=MovieSceneScriptingFloatChannel}
          Color.G {type=MovieSceneScriptingFloatChannel}
          Color.B {type=MovieSceneScriptingFloatChannel}
          Color.A {type=MovieSceneScriptingFloatChannel, keys [[87.5,0],[93.5,1]]}
         ]
        }]
      }
     ]
    }
    BorderCheck {
     id=#7
     possessed_class=Border
     tracks [
      {
       type=MovieSceneColorTrack
       display_name=Content Color and Opacity
       sections [{
         type=MovieSceneColorSection
         end_frame=199
         start_frame=0
         channels [
          Color.R {type=MovieSceneScriptingFloatChannel}
          Color.G {type=MovieSceneScriptingFloatChannel}
          Color.B {type=MovieSceneScriptingFloatChannel}
          Color.A {type=MovieSceneScriptingFloatChannel, keys [[114.9983,0],[126.6117,1]]}
         ]
        }]
      }
      {
       type=MovieScene2DTransformTrack
       sections [{
         type=MovieScene2DTransformSection
         end_frame=199
         start_frame=0
         channels [
          Translation.X {type=MovieSceneScriptingFloatChannel}
          Translation.Y {type=MovieSceneScriptingFloatChannel, keys [[114.9983,64],[126.6117,0]]}
          Angle {type=MovieSceneScriptingFloatChannel, keys [[115,360],[132.4183,0]]}
          Scale.X {type=MovieSceneScriptingFloatChannel, keys [[114.9983,4],[132.4183,0.75],[144.0317,1.25],[161.4517,1]]}
          Scale.Y {type=MovieSceneScriptingFloatChannel, keys [[114.9983,4],[132.4183,0.75],[144,1.5],[144.0317,1.25],[161.4517,1]]}
          Shear.X {type=MovieSceneScriptingFloatChannel}
          Shear.Y {type=MovieSceneScriptingFloatChannel}
         ]
        }]
      }
     ]
    }
   ]
   display_rate {numerator=100}
  }
 }
 "Purchased! (Bundle Item)" {
  binding_count=4
  display_label=""
  length_seconds=2
  object="&self:Purchased! (Bundle Item)"
  sequence {
   bindings [
    BorderItemNAme {
     id=#1
     possessed_class=Border
     tracks [
      {
       type=MovieSceneColorTrack
       display_name=Content Color and Opacity
       sections [{
         type=MovieSceneColorSection
         end_frame=199
         start_frame=0
         channels [
          Color.R {type=MovieSceneScriptingFloatChannel}
          Color.G {type=MovieSceneScriptingFloatChannel}
          Color.B {type=MovieSceneScriptingFloatChannel}
          Color.A {type=MovieSceneScriptingFloatChannel, keys [[25.25,0],[49.25,1]]}
         ]
        }]
      }
      {
       type=MovieScene2DTransformTrack
       sections [{
         type=MovieScene2DTransformSection
         end_frame=199
         start_frame=0
         channels [
          Translation.X {type=MovieSceneScriptingFloatChannel, keys [[25.25,768],[49.25,-64.0017],[55.25,16],[73.25,-8],[97.25,4],[127.2483,0]]}
          Translation.Y {type=MovieSceneScriptingFloatChannel}
          Angle {type=MovieSceneScriptingFloatChannel}
          Scale.X {type=MovieSceneScriptingFloatChannel, keys [[25.25,0.25],[55.25,1]]}
          Scale.Y {type=MovieSceneScriptingFloatChannel, keys [[25.25,0.25],[55.25,1]]}
          Shear.X {type=MovieSceneScriptingFloatChannel, keys [[25.25,-88],[61.25,24],[79.25,-4],[103.2483,2],[133.25,0]]}
          Shear.Y {type=MovieSceneScriptingFloatChannel}
         ]
        }]
      }
     ]
    }
    BorderRarityFlare {
     id=#2
     possessed_class=Border
     tracks [
      {
       type=MovieSceneColorTrack
       display_name=Content Color and Opacity
       sections [{
         type=MovieSceneColorSection
         end_frame=199
         start_frame=0
         channels [
          Color.R {type=MovieSceneScriptingFloatChannel}
          Color.G {type=MovieSceneScriptingFloatChannel}
          Color.B {type=MovieSceneScriptingFloatChannel}
          Color.A {type=MovieSceneScriptingFloatChannel, keys [[12.9983,0],[27.5,1]]}
         ]
        }]
      }
      {
       type=MovieScene2DTransformTrack
       sections [{
         type=MovieScene2DTransformSection
         end_frame=199
         start_frame=0
         channels [
          Translation.X {type=MovieSceneScriptingFloatChannel}
          Translation.Y {type=MovieSceneScriptingFloatChannel}
          Angle {type=MovieSceneScriptingFloatChannel}
          Scale.X {type=MovieSceneScriptingFloatChannel, keys [[12.9983,0],[33.5,1.85],[51.5,0.9],[75.5,1.05],[111.5,0.975],[165.5,1]]}
          Scale.Y {type=MovieSceneScriptingFloatChannel, keys [[12.9983,0],[33.5,1.85],[51.5,0.9],[75.5,1.05],[111.5,0.975],[165.5,1]]}
          Shear.X {type=MovieSceneScriptingFloatChannel}
          Shear.Y {type=MovieSceneScriptingFloatChannel}
         ]
        }]
      }
     ]
    }
    BorderItemCard {
     id=#3
     possessed_class=Border
     tracks [
      {
       type=MovieSceneColorTrack
       display_name=Content Color and Opacity
       sections [{
         type=MovieSceneColorSection
         end_frame=199
         start_frame=0
         channels [
          Color.R {type=MovieSceneScriptingFloatChannel}
          Color.G {type=MovieSceneScriptingFloatChannel}
          Color.B {type=MovieSceneScriptingFloatChannel}
          Color.A {type=MovieSceneScriptingFloatChannel, keys [[0,0],[21.5,1]]}
         ]
        }]
      }
      {
       type=MovieScene2DTransformTrack
       sections [{
         type=MovieScene2DTransformSection
         end_frame=199
         start_frame=-1
         channels [
          Translation.X {type=MovieSceneScriptingFloatChannel}
          Translation.Y {type=MovieSceneScriptingFloatChannel}
          Angle {type=MovieSceneScriptingFloatChannel}
          Scale.X {type=MovieSceneScriptingFloatChannel, keys [[-0.0017,3],[21.5,0.75],[32.75,1.125],[49.25,0.95],[69.875,1]]}
          Scale.Y {type=MovieSceneScriptingFloatChannel, keys [[-0.0017,3],[21.5,0.5],[32.75,1.275],[49.25,0.85],[69.875,1]]}
          Shear.X {type=MovieSceneScriptingFloatChannel}
          Shear.Y {type=MovieSceneScriptingFloatChannel, keys [[-0.0017,8],[26,-2],[37.25,4],[53,0]]}
         ]
        }]
      }
     ]
    }
    BorderRarityGlow {
     id=#5
     possessed_class=Border
     tracks [{
       type=MovieSceneColorTrack
       display_name=Content Color and Opacity
       sections [{
         type=MovieSceneColorSection
         end_frame=199
         start_frame=0
         channels [
          Color.R {type=MovieSceneScriptingFloatChannel}
          Color.G {type=MovieSceneScriptingFloatChannel}
          Color.B {type=MovieSceneScriptingFloatChannel}
          Color.A {type=MovieSceneScriptingFloatChannel, keys [[37.25,0],[96.875,1]]}
         ]
        }]
      }]
    }
   ]
   display_rate {numerator=100}
   tracks [{type=MovieSceneEventTrack, sections [{type=MovieSceneEventSection}]}]
  }
 }
 "Bundle Blur (loop)" {
  binding_count=1
  display_label=""
  length_seconds=3
  object="&self:Bundle Blur (loop)"
  sequence {
   bindings [Blur {
     id=#8
     possessed_class=BackgroundBlur
     tracks [{
       type=MovieSceneFloatTrack
       display_name=Render Opacity
       sections [{type=MovieSceneFloatSection, end_frame=300, start_frame=0, channels [None {type=MovieSceneScriptingFloatChannel, const [1,1]}]}]
      }]
    }]
   display_rate {numerator=100}
  }
 }
 "Bundle Blur (out)" {
  binding_count=1
  display_label=""
  length_seconds=3
  object="&self:Bundle Blur (out)"
  sequence {
   bindings [Blur {
     id=#8
     possessed_class=BackgroundBlur
     tracks [{
       type=MovieSceneFloatTrack
       display_name=Render Opacity
       sections [{type=MovieSceneFloatSection, end_frame=300, start_frame=0, channels [None {type=MovieSceneScriptingFloatChannel, keys [[0,1],[25,0]]}]}]
      }]
    }]
   display_rate {numerator=100}
  }
 }
 "Successfully Purchased! (Bundle Item)" {
  binding_count=2
  display_label=""
  length_seconds=2
  object="&self:Successfully Purchased! (Bundle Item)"
  sequence {
   bindings [
    BorderSuccessText {
     id=#6
     possessed_class=Border
     tracks [
      {
       type=MovieScene2DTransformTrack
       sections [{
         type=MovieScene2DTransformSection
         end_frame=199
         start_frame=0
         channels [
          Translation.X {type=MovieSceneScriptingFloatChannel, keys [[93.5,16],[111.4983,0]]}
          Translation.Y {type=MovieSceneScriptingFloatChannel, keys [[93.5,80],[111.4983,16],[117.4983,-4],[135.5,0]]}
          Angle {type=MovieSceneScriptingFloatChannel}
          Scale.X {type=MovieSceneScriptingFloatChannel, keys [[87.5,0],[105.5,0.2],[117.5,1],[123.4983,1]]}
          Scale.Y {type=MovieSceneScriptingFloatChannel, keys [[87.5,2],[105.5,2],[111.5,0.75],[123.4983,1.85],[135.5,0.95],[141.5,1]]}
          Shear.X {type=MovieSceneScriptingFloatChannel}
          Shear.Y {type=MovieSceneScriptingFloatChannel}
         ]
        }]
      }
      {
       type=MovieSceneColorTrack
       display_name=Content Color and Opacity
       sections [{
         type=MovieSceneColorSection
         end_frame=199
         start_frame=0
         channels [
          Color.R {type=MovieSceneScriptingFloatChannel}
          Color.G {type=MovieSceneScriptingFloatChannel}
          Color.B {type=MovieSceneScriptingFloatChannel}
          Color.A {type=MovieSceneScriptingFloatChannel, keys [[87.5,0],[93.5,1]]}
         ]
        }]
      }
     ]
    }
    BorderCheck {
     id=#7
     possessed_class=Border
     tracks [
      {
       type=MovieSceneColorTrack
       display_name=Content Color and Opacity
       sections [{
         type=MovieSceneColorSection
         end_frame=199
         start_frame=0
         channels [
          Color.R {type=MovieSceneScriptingFloatChannel}
          Color.G {type=MovieSceneScriptingFloatChannel}
          Color.B {type=MovieSceneScriptingFloatChannel}
          Color.A {type=MovieSceneScriptingFloatChannel, keys [[114.9983,0],[126.6117,1]]}
         ]
        }]
      }
      {
       type=MovieScene2DTransformTrack
       sections [{
         type=MovieScene2DTransformSection
         end_frame=199
         start_frame=0
         channels [
          Translation.X {type=MovieSceneScriptingFloatChannel}
          Translation.Y {type=MovieSceneScriptingFloatChannel, keys [[114.9983,64],[126.6117,0]]}
          Angle {type=MovieSceneScriptingFloatChannel, keys [[115,360],[132.4183,0]]}
          Scale.X {type=MovieSceneScriptingFloatChannel, keys [[114.9983,4],[132.4183,0.75],[144.0317,1.25],[161.4517,1]]}
          Scale.Y {type=MovieSceneScriptingFloatChannel, keys [[114.9983,4],[132.4183,0.75],[144,1.5],[144.0317,1.25],[161.4517,1]]}
          Shear.X {type=MovieSceneScriptingFloatChannel}
          Shear.Y {type=MovieSceneScriptingFloatChannel}
         ]
        }]
      }
     ]
    }
   ]
   display_rate {numerator=100}
  }
 }
 "Purchased! (Bundle Item) - FAST" {
  binding_count=4
  display_label=""
  length_seconds=2
  object="&self:Purchased! (Bundle Item) - FAST"
  sequence {
   bindings [
    BorderItemNAme {
     id=#1
     possessed_class=Border
     tracks [
      {
       type=MovieSceneColorTrack
       display_name=Content Color and Opacity
       sections [{
         type=MovieSceneColorSection
         end_frame=199
         start_frame=0
         channels [
          Color.R {type=MovieSceneScriptingFloatChannel}
          Color.G {type=MovieSceneScriptingFloatChannel}
          Color.B {type=MovieSceneScriptingFloatChannel}
          Color.A {type=MovieSceneScriptingFloatChannel, keys [[25.25,0],[49.25,1]]}
         ]
        }]
      }
      {
       type=MovieScene2DTransformTrack
       sections [{
         type=MovieScene2DTransformSection
         end_frame=199
         start_frame=0
         channels [
          Translation.X {type=MovieSceneScriptingFloatChannel, keys [[25.25,768],[49.25,-64.0017],[55.25,16],[73.25,-8],[97.25,4],[127.2483,0]]}
          Translation.Y {type=MovieSceneScriptingFloatChannel}
          Angle {type=MovieSceneScriptingFloatChannel}
          Scale.X {type=MovieSceneScriptingFloatChannel, keys [[25.25,0.25],[55.25,1]]}
          Scale.Y {type=MovieSceneScriptingFloatChannel, keys [[25.25,0.25],[55.25,1]]}
          Shear.X {type=MovieSceneScriptingFloatChannel, keys [[25.25,-88],[61.25,24],[79.25,-4],[103.2483,2],[133.25,0]]}
          Shear.Y {type=MovieSceneScriptingFloatChannel}
         ]
        }]
      }
     ]
    }
    BorderRarityFlare {
     id=#2
     possessed_class=Border
     tracks [
      {
       type=MovieSceneColorTrack
       display_name=Content Color and Opacity
       sections [{
         type=MovieSceneColorSection
         end_frame=199
         start_frame=0
         channels [
          Color.R {type=MovieSceneScriptingFloatChannel}
          Color.G {type=MovieSceneScriptingFloatChannel}
          Color.B {type=MovieSceneScriptingFloatChannel}
          Color.A {type=MovieSceneScriptingFloatChannel, keys [[12.9983,0],[27.5,1]]}
         ]
        }]
      }
      {
       type=MovieScene2DTransformTrack
       sections [{
         type=MovieScene2DTransformSection
         end_frame=199
         start_frame=0
         channels [
          Translation.X {type=MovieSceneScriptingFloatChannel}
          Translation.Y {type=MovieSceneScriptingFloatChannel}
          Angle {type=MovieSceneScriptingFloatChannel}
          Scale.X {type=MovieSceneScriptingFloatChannel, keys [[12.9983,0],[33.5,1.85],[51.5,0.9],[75.5,1.05],[111.5,0.975],[165.5,1]]}
          Scale.Y {type=MovieSceneScriptingFloatChannel, keys [[12.9983,0],[33.5,1.85],[51.5,0.9],[75.5,1.05],[111.5,0.975],[165.5,1]]}
          Shear.X {type=MovieSceneScriptingFloatChannel}
          Shear.Y {type=MovieSceneScriptingFloatChannel}
         ]
        }]
      }
     ]
    }
    BorderItemCard {
     id=#3
     possessed_class=Border
     tracks [
      {
       type=MovieSceneColorTrack
       display_name=Content Color and Opacity
       sections [{
         type=MovieSceneColorSection
         end_frame=199
         start_frame=0
         channels [
          Color.R {type=MovieSceneScriptingFloatChannel}
          Color.G {type=MovieSceneScriptingFloatChannel}
          Color.B {type=MovieSceneScriptingFloatChannel}
          Color.A {type=MovieSceneScriptingFloatChannel, keys [[0,0],[21.5,1]]}
         ]
        }]
      }
      {
       type=MovieScene2DTransformTrack
       sections [{
         type=MovieScene2DTransformSection
         end_frame=199
         start_frame=-1
         channels [
          Translation.X {type=MovieSceneScriptingFloatChannel}
          Translation.Y {type=MovieSceneScriptingFloatChannel}
          Angle {type=MovieSceneScriptingFloatChannel}
          Scale.X {type=MovieSceneScriptingFloatChannel, keys [[-0.0017,1.5],[21.5,0.75],[32.75,1.125],[49.25,0.95],[69.875,1]]}
          Scale.Y {type=MovieSceneScriptingFloatChannel, keys [[-0.0017,1.5],[21.5,0.5],[32.75,1.275],[49.25,0.85],[69.875,1]]}
          Shear.X {type=MovieSceneScriptingFloatChannel}
          Shear.Y {type=MovieSceneScriptingFloatChannel, keys [[-0.0017,8],[26,-2],[37.25,4],[53,0]]}
         ]
        }]
      }
     ]
    }
    BorderRarityGlow {
     id=#5
     possessed_class=Border
     tracks [{
       type=MovieSceneColorTrack
       display_name=Content Color and Opacity
       sections [{
         type=MovieSceneColorSection
         end_frame=199
         start_frame=0
         channels [
          Color.R {type=MovieSceneScriptingFloatChannel}
          Color.G {type=MovieSceneScriptingFloatChannel}
          Color.B {type=MovieSceneScriptingFloatChannel}
          Color.A {type=MovieSceneScriptingFloatChannel, keys [[37.25,0],[96.875,1]]}
         ]
        }]
      }]
    }
   ]
   display_rate {numerator=100}
   tracks [{type=MovieSceneEventTrack, sections [{type=MovieSceneEventSection}]}]
  }
 }
 Refunded {
  binding_count=7
  length_seconds=3
  object=&self:Refunded
  sequence {
   bindings [
    BorderItemNAme {
     id=#1
     possessed_class=Border
     tracks [
      {
       type=MovieSceneColorTrack
       display_name=Content Color and Opacity
       sections [{
         type=MovieSceneColorSection
         end_frame=199
         start_frame=0
         channels [
          Color.R {type=MovieSceneScriptingFloatChannel}
          Color.G {type=MovieSceneScriptingFloatChannel}
          Color.B {type=MovieSceneScriptingFloatChannel}
          Color.A {type=MovieSceneScriptingFloatChannel, keys [[25.25,0],[49.25,1]]}
         ]
        }]
      }
      {
       type=MovieScene2DTransformTrack
       sections [{
         type=MovieScene2DTransformSection
         end_frame=199
         start_frame=0
         channels [
          Translation.X {type=MovieSceneScriptingFloatChannel, keys [[25.25,768],[49.25,-64.0017],[55.25,16],[73.25,-8],[97.25,4],[127.2483,0]]}
          Translation.Y {type=MovieSceneScriptingFloatChannel}
          Angle {type=MovieSceneScriptingFloatChannel}
          Scale.X {type=MovieSceneScriptingFloatChannel, keys [[25.25,0.25],[55.25,1]]}
          Scale.Y {type=MovieSceneScriptingFloatChannel, keys [[25.25,0.25],[55.25,1]]}
          Shear.X {type=MovieSceneScriptingFloatChannel, keys [[25.25,-88],[61.25,24],[79.25,-4],[103.2483,2],[133.25,0]]}
          Shear.Y {type=MovieSceneScriptingFloatChannel}
         ]
        }]
      }
     ]
    }
    BorderRefundCheck {
     id=#9
     possessed_class=Border
     tracks [
      {
       type=MovieSceneColorTrack
       display_name=Content Color and Opacity
       sections [{
         type=MovieSceneColorSection
         end_frame=199
         start_frame=0
         channels [
          Color.R {type=MovieSceneScriptingFloatChannel}
          Color.G {type=MovieSceneScriptingFloatChannel}
          Color.B {type=MovieSceneScriptingFloatChannel}
          Color.A {type=MovieSceneScriptingFloatChannel, keys [[114.9983,0],[126.6117,1]]}
         ]
        }]
      }
      {
       type=MovieScene2DTransformTrack
       sections [{
         type=MovieScene2DTransformSection
         end_frame=199
         start_frame=0
         channels [
          Translation.X {type=MovieSceneScriptingFloatChannel}
          Translation.Y {type=MovieSceneScriptingFloatChannel, keys [[114.9983,64],[126.6117,0]]}
          Angle {type=MovieSceneScriptingFloatChannel, keys [[115,360],[132.4183,0]]}
          Scale.X {type=MovieSceneScriptingFloatChannel, keys [[114.9983,4],[132.4183,0.75],[144.0317,1.25],[161.4517,1]]}
          Scale.Y {type=MovieSceneScriptingFloatChannel, keys [[114.9983,4],[132.4183,0.75],[144,1.5],[144.0317,1.25],[161.4517,1]]}
          Shear.X {type=MovieSceneScriptingFloatChannel}
          Shear.Y {type=MovieSceneScriptingFloatChannel}
         ]
        }]
      }
     ]
    }
    BorderRarityFlare {
     id=#2
     possessed_class=Border
     tracks [
      {
       type=MovieSceneColorTrack
       display_name=Content Color and Opacity
       sections [{
         type=MovieSceneColorSection
         end_frame=199
         start_frame=0
         channels [
          Color.R {type=MovieSceneScriptingFloatChannel}
          Color.G {type=MovieSceneScriptingFloatChannel}
          Color.B {type=MovieSceneScriptingFloatChannel}
          Color.A {type=MovieSceneScriptingFloatChannel, keys [[12.9983,0],[27.5,1]]}
         ]
        }]
      }
      {
       type=MovieScene2DTransformTrack
       sections [{
         type=MovieScene2DTransformSection
         end_frame=199
         start_frame=0
         channels [
          Translation.X {type=MovieSceneScriptingFloatChannel}
          Translation.Y {type=MovieSceneScriptingFloatChannel}
          Angle {type=MovieSceneScriptingFloatChannel}
          Scale.X {type=MovieSceneScriptingFloatChannel, keys [[12.9983,0],[33.5,1.85],[51.5,0.9],[75.5,1.05],[111.5,0.975],[165.5,1]]}
          Scale.Y {type=MovieSceneScriptingFloatChannel, keys [[12.9983,0],[33.5,1.85],[51.5,0.9],[75.5,1.05],[111.5,0.975],[165.5,1]]}
          Shear.X {type=MovieSceneScriptingFloatChannel}
          Shear.Y {type=MovieSceneScriptingFloatChannel}
         ]
        }]
      }
     ]
    }
    BorderItemCard {
     id=#3
     possessed_class=Border
     tracks [
      {
       type=MovieSceneColorTrack
       display_name=Content Color and Opacity
       sections [{
         type=MovieSceneColorSection
         end_frame=199
         start_frame=0
         channels [
          Color.R {type=MovieSceneScriptingFloatChannel}
          Color.G {type=MovieSceneScriptingFloatChannel}
          Color.B {type=MovieSceneScriptingFloatChannel}
          Color.A {type=MovieSceneScriptingFloatChannel, keys [[0,0],[21.5,1]]}
         ]
        }]
      }
      {
       type=MovieScene2DTransformTrack
       sections [{
         type=MovieScene2DTransformSection
         end_frame=199
         start_frame=-1
         channels [
          Translation.X {type=MovieSceneScriptingFloatChannel}
          Translation.Y {type=MovieSceneScriptingFloatChannel}
          Angle {type=MovieSceneScriptingFloatChannel}
          Scale.X {type=MovieSceneScriptingFloatChannel, keys [[-0.0017,3],[21.5,0.75],[32.75,1.125],[49.25,0.95],[69.875,1]]}
          Scale.Y {type=MovieSceneScriptingFloatChannel, keys [[-0.0017,3],[21.5,0.5],[32.75,1.275],[49.25,0.85],[69.875,1]]}
          Shear.X {type=MovieSceneScriptingFloatChannel}
          Shear.Y {type=MovieSceneScriptingFloatChannel, keys [[-0.0017,8],[26,-2],[37.25,4],[53,0]]}
         ]
        }]
      }
     ]
    }
    Blur {
     id=#4
     possessed_class=BackgroundBlur
     tracks [{
       type=MovieSceneFloatTrack
       display_name=Render Opacity
       sections [{
         type=MovieSceneFloatSection
         end_frame=300
         start_frame=0
         channels [None {type=MovieSceneScriptingFloatChannel, keys [[0,1],[275,1],[300,0]]}]
        }]
      }]
    }
    BorderRarityGlow {
     id=#5
     possessed_class=Border
     tracks [{
       type=MovieSceneColorTrack
       display_name=Content Color and Opacity
       sections [{
         type=MovieSceneColorSection
         end_frame=199
         start_frame=0
         channels [
          Color.R {type=MovieSceneScriptingFloatChannel}
          Color.G {type=MovieSceneScriptingFloatChannel}
          Color.B {type=MovieSceneScriptingFloatChannel}
          Color.A {type=MovieSceneScriptingFloatChannel, keys [[37.25,0],[96.875,1]]}
         ]
        }]
      }]
    }
    BorderSuccessRefundText {
     id=#10
     possessed_class=Border
     tracks [
      {
       type=MovieSceneColorTrack
       display_name=Content Color and Opacity
       sections [{
         type=MovieSceneColorSection
         end_frame=199
         start_frame=0
         channels [
          Color.R {type=MovieSceneScriptingFloatChannel}
          Color.G {type=MovieSceneScriptingFloatChannel}
          Color.B {type=MovieSceneScriptingFloatChannel}
          Color.A {type=MovieSceneScriptingFloatChannel, keys [[87.5,0],[93.5,1]]}
         ]
        }]
      }
      {
       type=MovieScene2DTransformTrack
       sections [{
         type=MovieScene2DTransformSection
         end_frame=199
         start_frame=0
         channels [
          Translation.X {type=MovieSceneScriptingFloatChannel, keys [[93.5,16],[111.4983,0]]}
          Translation.Y {type=MovieSceneScriptingFloatChannel, keys [[93.5,80],[111.4983,16],[117.4983,-4],[135.5,0]]}
          Angle {type=MovieSceneScriptingFloatChannel}
          Scale.X {type=MovieSceneScriptingFloatChannel, keys [[87.5,0],[105.5,0.2],[117.5,1],[123.4983,1]]}
          Scale.Y {type=MovieSceneScriptingFloatChannel, keys [[87.5,2],[105.5,2],[111.5,0.75],[123.4983,1.85],[135.5,0.95],[141.5,1]]}
          Shear.X {type=MovieSceneScriptingFloatChannel}
          Shear.Y {type=MovieSceneScriptingFloatChannel}
         ]
        }]
      }
     ]
    }
   ]
   display_rate {numerator=100}
  }
 }
 SuccessfullyRefundedBundleItem {
  binding_count=2
  length_seconds=2
  object=&self:SuccessfullyRefundedBundleItem
  sequence {
   bindings [
    BorderRefundCheck {
     id=#11
     possessed_class=Border
     tracks [
      {
       type=MovieSceneColorTrack
       display_name=Content Color and Opacity
       sections [{
         type=MovieSceneColorSection
         end_frame=199
         start_frame=0
         channels [
          Color.R {type=MovieSceneScriptingFloatChannel}
          Color.G {type=MovieSceneScriptingFloatChannel}
          Color.B {type=MovieSceneScriptingFloatChannel}
          Color.A {type=MovieSceneScriptingFloatChannel, keys [[114.9983,0],[126.6117,1]]}
         ]
        }]
      }
      {
       type=MovieScene2DTransformTrack
       sections [{
         type=MovieScene2DTransformSection
         end_frame=199
         start_frame=0
         channels [
          Translation.X {type=MovieSceneScriptingFloatChannel}
          Translation.Y {type=MovieSceneScriptingFloatChannel, keys [[114.9983,64],[126.6117,0]]}
          Angle {type=MovieSceneScriptingFloatChannel, keys [[115,360],[132.4183,0]]}
          Scale.X {type=MovieSceneScriptingFloatChannel, keys [[114.9983,4],[132.4183,0.75],[144.0317,1.25],[161.4517,1]]}
          Scale.Y {type=MovieSceneScriptingFloatChannel, keys [[114.9983,4],[132.4183,0.75],[144,1.5],[144.0317,1.25],[161.4517,1]]}
          Shear.X {type=MovieSceneScriptingFloatChannel}
          Shear.Y {type=MovieSceneScriptingFloatChannel}
         ]
        }]
      }
     ]
    }
    BorderSuccessRefundText {
     id=#12
     possessed_class=Border
     tracks [
      {
       type=MovieSceneColorTrack
       display_name=Content Color and Opacity
       sections [{
         type=MovieSceneColorSection
         end_frame=199
         start_frame=0
         channels [
          Color.R {type=MovieSceneScriptingFloatChannel}
          Color.G {type=MovieSceneScriptingFloatChannel}
          Color.B {type=MovieSceneScriptingFloatChannel}
          Color.A {type=MovieSceneScriptingFloatChannel, keys [[87.5,0],[93.5,1]]}
         ]
        }]
      }
      {
       type=MovieScene2DTransformTrack
       sections [{
         type=MovieScene2DTransformSection
         end_frame=199
         start_frame=0
         channels [
          Translation.X {type=MovieSceneScriptingFloatChannel, keys [[93.5,16],[111.4983,0]]}
          Translation.Y {type=MovieSceneScriptingFloatChannel, keys [[93.5,80],[111.4983,16],[117.4983,-4],[135.5,0]]}
          Angle {type=MovieSceneScriptingFloatChannel}
          Scale.X {type=MovieSceneScriptingFloatChannel, keys [[87.5,0],[105.5,0.2],[117.5,1],[123.4983,1]]}
          Scale.Y {type=MovieSceneScriptingFloatChannel, keys [[87.5,2],[105.5,2],[111.5,0.75],[123.4983,1.85],[135.5,0.95],[141.5,1]]}
          Shear.X {type=MovieSceneScriptingFloatChannel}
          Shear.Y {type=MovieSceneScriptingFloatChannel}
         ]
        }]
      }
     ]
    }
   ]
   display_rate {numerator=100}
  }
 }
]
graphs [
 Bundle Item Anim Speed {nodes [
   Bundle Item Anim Speed() {is=title K2Node_FunctionEntry}
   n1 = Get Bundle Item Index() -> (Bundle Item Index:int=0)
   n2 = "Equal (Integer)"(
    A=Last Index!(TargetArray:Array of Fort Item VM Object References=Get PurchasedItemsVMs() -> (PurchasedItemsVMs:Array of Fort Item VM Object References)) -> (ReturnValue=0)
    B=@n1
   ) {is=title K2Node_CallFunction}
   > Return Node!(
    execute
    speed:double=SelectFloat(
     A=SelectFloat(A=1, B=5, bPickA=@n2)
     B=1
     bPickA="integer >= integer"!(A:int=@n1, B:int=1) -> (ReturnValue:bool) {is=title K2Node_CallFunction}
    )
   )
  ]}
 EventGraph {nodes [
   Show Next Item() {is=title K2Node_CustomEvent}
   > Show Bundle Mini Item Card()
   n1 = "Get (a copy)"(Array:Array of Fort Item VM Object References=Get PurchasedItemsVMs(), Dimension 1=Get Bundle Item Index() -> (Bundle Item Index=0)) -> (Output:Fort Item VM Object Reference)
   > Is Valid(InputObject=@n1)
   > UpdateFromOffer!(execute, self, ItemVM:Fort Item VM Object Reference=@n1) -> (then)
   > n2 = Branch(Condition=ShouldShowAsRefund!(self:Athena Item Shop Purchase Overlay Modal Object Reference) -> (ReturnValue:bool))
   > n3 = PlaySound2D(Sound=Get RefundAppearSoundTiles() -> (RefundAppearSoundTiles:Sound Base Object Reference))
   n4 = Sequence(execute=[@n3.then, @n2.else])
   n5 = Get BundleBlurPlayed() -> (BundleBlurPlayed:bool)
   > Branch(Condition=AND Boolean(A=HasMultipleItems!(self:Athena Item Shop Purchase Overlay Modal Object Reference) -> (ReturnValue:bool), B=NOT Boolean(A=@n5)))
   > Set BundleBlurPlayed(BundleBlurPlayed:bool=true) -> (Output_Get:bool)
   n6 = "Get Bundle Blur (loop)"() -> ("Bundle Blur (loop)":Widget Animation Object Reference)
   > PlayAnimation(InAnimation=@n6, NumLoopsToPlay=0)
   > PlayAnimation(InAnimation=Select(
     Option 0:Widget Animation Object Reference="Get Successfully Purchased! (Bundle Item)"() -> ("Successfully Purchased! (Bundle Item)":Widget Animation Object Reference)
     Option 1:Widget Animation Object Reference=Get SuccessfullyRefundedBundleItem() -> (SuccessfullyRefundedBundleItem:Widget Animation Object Reference)
     Index=ShouldShowAsRefund()
    ) -> (ReturnValue:Widget Animation Object Reference))
   > Do Once()
   > PlaySound2D(Sound=Get RefundAppearSoundInitial() -> (RefundAppearSoundInitial:Sound Base Object Reference))
   n7 = "float > float"(A=Bundle Item Anim Speed!() -> (speed:double) {is=title K2Node_CallFunction}, B=3) {is=title K2Node_CallFunction}
   n8 = "Get Purchased! (Bundle Item) - FAST"() -> ("Purchased! (Bundle Item) - FAST":Widget Animation Object Reference)
   n9 = "Get Purchased! (Bundle Item)"() -> ("Purchased! (Bundle Item)":Widget Animation Object Reference)
   n10 = ShouldShowAsRefund()
   n11 = Get Refunded() -> (Refunded:Widget Animation Object Reference)
   n12 = "Get Purchased!"() -> ("Purchased!":Widget Animation Object Reference)
   PlayAnimation(
    execute=@n4.then_1
    InAnimation=Select(Option 0=Select(Option 0=@n12, Option 1=@n11, Index=@n10), Option 1=Select(Option 0=@n9, Option 1=@n8, Index=@n7), Index=HasMultipleItems())
    PlaybackSpeed=Bundle Item Anim Speed()
   )
   > SetUserFocus!(execute, self:Widget Object Reference, PlayerController:Player Controller Object Reference=GetOwningPlayer()) -> (then)
   BundleBlurOut() {is=title K2Node_CustomEvent}
   n13 = Get PurchasedItemsVMs()
   > Branch(Condition="Equal (Integer)"(A=Last Index(TargetArray=@n13) -> (ReturnValue=0), B=Get Bundle Item Index() -> (Bundle Item Index=0)) {is=title K2Node_PromotableOperator})
   > StopAnimation(InAnimation=@n6)
   > PlayAnimation(
    InAnimation="Get Bundle Blur (out)"() -> ("Bundle Blur (out)":Widget Animation Object Reference)
    NumLoopsToPlay=0
    PlaybackSpeed=Bundle Item Anim Speed()
   )
   Populate Bundle Mini-Cards() {is=title K2Node_CustomEvent}
   > n14 = For Each Loop(Array:Array of Fort Item VM Object References=Get PurchasedItemsVMs()) -> (Array Element:Fort Item VM Object Reference)
   n15 = CreateTemporaryItemInstance!(
    ItemDefinition:Item Definition Base Object Reference=Get ItemDefinition!(self:Fort Item VM Object Reference=@n14) -> (ItemDefinition:Fort Item Definition Object Reference)
    Count:int=1
    Level:int=-1
   ) -> (ReturnValue:Fort Item Object Reference)
   > Is Valid(InputObject=@n15)
   > n16 = Create Entry!(execute, self:Dynamic Entry Box Object Reference=Get DynaBox() -> (DynaBox:Dynamic Entry Box Object Reference)) -> (then, ReturnValue:User Widget Object Reference)
   > n17 = Cast To BundleItemMiniCardWidget!(execute, Object=@n16) -> (then, CastFailed, AsBundle Item Mini Card Widget:Bundle Item Mini Card Widget Object Reference) {is=title K2Node_DynamicCast}
   > SetItemToRepresent!(
    execute
    self:Fort Cosmetic Item Card Object Reference=Get ItemCard!(self:Bundle Item Mini Card Widget Object Reference=@n17) -> (ItemCard:Fort Cosmetic Item Card Object Reference)
    InItemToRepresent:Fort Item Object Reference=@n15
    InItemVM:Fort Item VM Object Reference=@n14
   ) -> (then)
   > Add(
    TargetArray:Array of Bundle Item Mini Card Widget Object References=Get Bundle Mini Item Cards() -> (Bundle Mini Item Cards:Array of Bundle Item Mini Card Widget Object References)
    NewItem:"Bundle Item Mini Card Widget Object Reference (by ref)"=@n17
   ) {is=title K2Node_CallArrayFunction}
   > SetVisibility(self=@n17, InVisibility=Hidden)
   Show Bundle Mini Item Card() {is=title K2Node_CustomEvent}
   > Branch(Condition=Is Valid Index!(
     TargetArray:Array of Bundle Item Mini Card Widget Object References=Get Bundle Mini Item Cards()
     IndexToTest:int=Get Bundle Item Index() -> (Bundle Item Index=0)
    ) -> (ReturnValue:bool))
   n18 = Get Bundle Item Index() -> (Bundle Item Index=0)
   n19 = "Get (a copy)"(Array:Array of Bundle Item Mini Card Widget Object References=Get Bundle Mini Item Cards(), Dimension 1=@n18) -> (Output:Bundle Item Mini Card Widget Object Reference)
   > SetVisibility(self=@n19, InVisibility=HitTestInvisible)
   > Play Anim - Show(self:Bundle Item Mini Card Widget Object Reference=@n19)
   n20 = float / float(A="To Float (Integer)"(InInt=@n18), B=Get MiniGridMaxItemsInRow() -> (MiniGridMaxItemsInRow:double))
   n21 = "Max (Integer)"(A=int - int(A=Floor(A=@n20) {is=title K2Node_CallFunction}, B=Get RowsToSkipBeforeScrolling() -> (RowsToSkipBeforeScrolling:int=0)), B=0)
   n22 = Get MiniCardHeight() -> (MiniCardHeight:int=0)
   > SetScrollOffset!(
    execute
    self:Scroll Box Object Reference=Get ScrollBox_Dyna() -> (ScrollBox_Dyna:Scroll Box Object Reference)
    NewScrollOffset:float="To Float (Integer)"(InInt=int * int(A=@n21, B=@n22))
   ) -> (then)
   Event InitBundleLayout() {is=title K2Node_Event}
   > n23 = Sequence!(execute) -> (then_0, then_1, then_2:Exec)
   > SetFortItemDetailsVM!(execute, self:Fort Cosmetic Item Card Object Reference=Get ItemCard!() -> (ItemCard), Viewmodel:Fort Item Details VM Object Reference) -> (then)
   > Clear(TargetArray:Array of Bundle Item Mini Card Widget Object References=Get Bundle Mini Item Cards())
   > Reset!(execute, self:Dynamic Entry Box Object Reference=Get DynaBox(), bDeleteWidgets:bool=false) -> (then)
   n24 = Branch(execute=@n23.then_1, Condition=HasMultipleItems())
   n25 = SlotAsUniformGridSlot!(Widget:Widget Object Reference=Get OverlayMainItem() -> (OverlayMainItem:Overlay Object Reference)) -> (ReturnValue:Uniform Grid Slot Object Reference)
   > SetColumn!(execute, self:Uniform Grid Slot Object Reference=@n25, InColumn:int=1) -> (then)
   > Populate Bundle Mini-Cards() {is=title K2Node_CallFunction}
   SetColumn(execute=@n24.else, self=@n25, InColumn=0)
   Event Destruct()
   > StopAllAnimations(self:User Widget Object Reference)
   Event On Offer Set() {is=title K2Node_Event}
   > Set Bundle Item Index(Bundle Item Index:int=0) -> (Output_Get:int=0)
   > Set BundleBlurPlayed(BundleBlurPlayed=false)
   > Show Next Item() {is=title K2Node_CallFunction}
   n26 = "Animation Finished (Purchased!)"!() -> (OutputDelegate:Delegate, then:Exec) {is=title K2Node_WidgetAnimationEvent}
   Dismiss(execute=[@n26.then, @n31.then], self:Athena Item Shop Purchase Overlay Modal Object Reference)
   n27 = "Animation Finished (Purchased! (Bundle Item) - FAST)"!() -> (OutputDelegate:Delegate, then:Exec) {is=title K2Node_WidgetAnimationEvent}
   Increment Int(" "=[@n27.then, @n30.then], Value=Get Bundle Item Index() -> (Bundle Item Index=0))
   n28 = Get PurchasedItemsVMs()
   > n29 = Branch(Condition="integer >= integer"(A=Get Bundle Item Index() -> (Bundle Item Index=0), B=Length(TargetArray:Array of Fort Item VM Object References=@n28)))
   > Dismiss()
   Show Next Item(execute=@n29.else)
   n30 = "Animation Finished (Purchased! (Bundle Item))"!() -> (OutputDelegate:Delegate, then:Exec) {is=title K2Node_WidgetAnimationEvent}
   n31 = "Animation Finished (Refunded)"!() -> (OutputDelegate:Delegate, then:Exec) {is=title K2Node_WidgetAnimationEvent}
  ]}
 Get Bundle Info from Item Def {nodes [
   n1 = Get Bundle Info from Item Def!() -> (then, Item Definition:Fort Account Item Definition Object Reference) {is=title K2Node_FunctionEntry}
   > "Set Item Definition (local)"("Item Definition (local)":Fort Account Item Definition Object Reference=@n1) -> (Output_Get:Fort Account Item Definition Object Reference)
   n2 = GetDynamicBundleItems!(self:Fort Store Front Offer Info Object Reference=GetOfferInfo!(self:Athena Item Shop Offer Details Widget Base Object Reference) -> (ReturnValue:Fort Store Front Offer Info Object Reference)) -> (Items:Array of Bundled Item Info Structures)
   > n3 = Branch(Condition="Equal (Integer)"(A=Length(TargetArray:Array of Bundled Item Info Structures=@n2), B=0) {is=title K2Node_CallFunction})
   > Return Node!(execute, Bundled Item Info:Bundled Item Info Structure, bFoundItem:bool=false)
   n4 = For Each Loop with Break!(Exec:Exec=@n3.else, Array:Array of Bundled Item Info Structures=@n2, Break:Exec=@n6.then) -> (LoopBody:Exec, Array Element:Bundled Item Info Structure, Array Index:int, Completed:Exec)
   n5 = "Get Item Definition (local)"() -> ("Item Definition (local)":Fort Account Item Definition Object Reference)
   > Branch(Condition="Equal Exactly (String)"(
     A=Break Bundled Item Info!(BundledItemInfo:"Bundled Item Info Structure (by ref)"=@n4) -> (TemplateId:String, Quantity:int, bOwned:bool=~) {is=title K2Node_BreakStruct}
     B=GetPersistentName!(self:Fort Item Definition Object Reference=@n5) -> (ReturnValue:String)
    ))
   > "Set Bundled Item Info (local)"("Bundled Item Info (local)":Bundled Item Info Structure=@n4) -> (Output_Get:Bundled Item Info Structure)
   > n6 = "Set bFoundItem (local)"("bFoundItem (local)":bool=true) -> (Output_Get:bool)
   Return Node(
    execute=@n4.Completed
    Bundled Item Info="Get Bundled Item Info (local)"() -> ("Bundled Item Info (local)":Bundled Item Info Structure)
    bFoundItem="Get bFoundItem (local)"() -> ("bFoundItem (local)":bool)
   )
  ]}
 HandleBack {nodes [
   HandleBack() {is=title K2Node_FunctionEntry}
   > Dismiss()
   > Return Node!(execute, PassThrough:bool=false)
  ]}
 UpdateFromOffer {nodes [
   n1 = UpdateFromOffer!() -> (then, ItemVM:Fort Item VM Object Reference) {is=title K2Node_FunctionEntry}
   > n2 = Cast To FortAccountItemDefinition!(execute, Object=Get ItemDefinition(self=@n1)) -> (then, CastFailed, AsFort Account Item Definition:Fort Account Item Definition Object Reference) {is=title K2Node_DynamicCast}
   > "Set Item Definition (local)"("Item Definition (local)"=@n2)
   > n3 = Sequence!(execute) -> (then_0, then_1, then_2, then_3:Exec)
   n4 = Get ItemDetails!(self:Fort Item VM Object Reference=Get ItemVM() -> (ItemVM:Fort Item VM Object Reference)) -> (ItemDetails:Fort Item Details VM Object Reference)
   n5 = Is Valid(exec=@n3.then_1, InputObject=@n4)
   "SetText (Text)"(
    execute=@n5.Is Not Valid
    self=Get TextName() -> (TextName:Common Text Object Reference)
    InText=GetDisplayName!(self:Item Definition Base Object Reference="Get Item Definition (local)"()) -> (ReturnValue:Text)
   )
   "SetText (Text)"(execute=@n5.Is Valid, self=Get TextName(), InText=GetDisplayName(self:Fort Item Details VM Object Reference=@n4))
   n6 = Branch(execute=@n3.then_0, Condition=HasMultipleItems())
   > n7 = Get Bundle Info from Item Def!(execute, self, Item Definition:Fort Account Item Definition Object Reference="Get Item Definition (local)"()) -> (then, Bundled Item Info_TemplateId:String, Bundled Item Info_Quantity:int, Bundled Item Info_bOwned:bool=~, bFoundItem:bool) {is=title K2Node_CallFunction}
   > n8 = "Set Grant Quantity (local)"("Grant Quantity (local)":int=@n7.Bundled Item Info_Quantity) -> (Output_Get:int=0)
   n9 = GetQuantityPurchased!(self:Athena Item Shop Purchase Overlay Modal Object Reference) -> (ReturnValue:int=0)
   n10 = "Get Grant Quantity (local)"() -> ("Grant Quantity (local)":int=0)
   n11 = "Get Item Definition (local)"()
   SetItemToRepresent(
    execute=[@n19.then, @n8.then]
    self=Get ItemCard()
    InItemToRepresent=CreateTemporaryItemInstance!(
     self:Fort Item Definition Object Reference=@n11
     Count=int * int(A:int=@n9, B:int=@n10) -> (ReturnValue:int=0) {is=title K2Node_CommutativeAssociativeBinaryOperator}
     Level=-1
    ) -> (ReturnValue)
    InItemVM=Get ItemVM()
   )
   n12 = Get ItemDetails(self=Get ItemVM())
   > n13 = GetRarityData!(
    execute
    Rarity:EFortRarity Enum=Select(
     Option 0:EFortRarity Enum=Common
     Option 1:EFortRarity Enum=GetRarity!(self:Fort Item Definition Object Reference=@n11) -> (ReturnValue:EFortRarity Enum=Common)
     Index=Get bShouldShowRarity!(self:Fort Item Details VM Object Reference=@n12) -> (bShouldShowRarity:bool)
    ) -> (ReturnValue:EFortRarity Enum)
   ) -> (then, ReturnValue:Fort Rarity Item Data Structure)
   > n14 = GetDynamicMaterial(self=Get ImageRarityFlare() -> (ImageRarityFlare:Image Object Reference))
   n15 = Break Fort Rarity Item Data!(FortRarityItemData:"Fort Rarity Item Data Structure (by ref)"=@n13) -> (
    Name:Text
    Color1:Linear Color Structure
    Color2:Linear Color Structure
    Color3:Linear Color Structure
    Color4:Linear Color Structure
    Color5:Linear Color Structure
    Radius:float=~
    Falloff:float=~
    Brightness:float=~
    Roughness:float=~
    Glow:float=~
   ) {is=title K2Node_BreakStruct}
   > SetVectorParameterValue(self=@n14, ParameterName=Rarity Color 1, Value=@n15.Color2)
   > SetVectorParameterValue(self=@n14, ParameterName=Rarity Color 2, Value=@n15.Color3)
   n16 = Get TextName()
   > SetStyle!(
    execute
    self:Common Text Object Reference=@n16
    InStyle:Common Text Style Class Reference=/FortUILibrary/Styles/Typography/Foundation/TextStyle-Header-XL-Horizon.TextStyle-Header-XL-Horizon_C
   ) -> (then)
   > n17 = GetDynamicFontMaterial!(execute, self:Text Object Reference=@n16) -> (then, ReturnValue:Material Instance Dynamic Object Reference)
   > SetVectorParameterValue(self=@n17, ParameterName=MakeLiteralName(Value=RGB_1), Value=@n15.Color1)
   > SetVectorParameterValue(self=@n17, ParameterName=MakeLiteralName(Value=RGB_2), Value=@n15.Color2)
   > n18 = GetDynamicMaterial(self=Get RarityGlow() -> (RarityGlow:Image Object Reference))
   > SetVectorParameterValue(self=@n18, ParameterName=vRGB, Value=@n15.Color1)
   n19 = "Set Grant Quantity (local)"(
    execute=@n6.else
    "Grant Quantity (local)"=GetGrantedItemQuantity!(self:Fort Store Front Offer Info Object Reference=GetOfferInfo(), GrantedItemIndex:int=0) -> (ReturnValue:int=0)
   ) -> (Output_Get=0)
   n20 = Branch(Condition=HasMultipleItems())
   > "SetText (Text)"(self=Get TextName(), InText=GetDisplayName(self:Item Definition Base Object Reference="Get Item Definition (local)"()))
   n21 = GetOfferInfo()
   n22 = GetName!(self:Fort Store Front Offer Info Object Reference=@n21) -> (ReturnValue:Text)
   n23 = TextIsEmpty(InText=@n22)
   n24 = GetItemTypeText!(self:Fort Store Front Offer Info Object Reference=@n21) -> (ReturnValue:Text)
   "SetText (Text)"(
    execute=@n20.else
    self=Get TextName()
    InText=TextToUpper!(InText:"Text (by ref)"=Select(Option 0:Text=@n22, Option 1:Text=@n24, Index=@n23) -> (ReturnValue:Text)) -> (ReturnValue:Text)
   )
  ]}
]
variables [
 Bundle Item Index {type=integer}
 Bundle Mini Item Cards {type="array<BundleItemMiniCardWidget_C>", default []}
 BundleBlurPlayed {type=boolean}
 BundleMiniGrid - Column {type=integer}
 BundleMiniGrid - Row {type=integer}
 Fort Received Loot Info Struct {
  type=FortReceivedItemLootInfo
  default {generatedItemInstance=null, itemDef=null, itemVM=null, quantity=0, templateId="", headerInformation {}}
 }
 MiniCardHeight {type=integer, default=118}
 MiniGridMaxItemsInRow {type=number, default=4}
 RefundAppearSoundInitial {type=SoundBase, default=/Game/Sounds/Fort_UI_Sounds/FrontEnd/Store/Store_PurchaseItem_Athena_Undo_Cue}
 RefundAppearSoundTiles {type=SoundBase, default=/Game/Sounds/Fort_UI_Sounds/FrontEnd/Store/Store_PurchaseItem_RefundAppear_Cue}
 RowsToSkipBeforeScrolling {type=integer, default=2}
]
widgets {
 info {
  inheritedWidgetCount=6
  parentClass=/Script/FortniteUI.AthenaItemShopPurchaseOverlayModal
  rootWidgetClass=&BackgroundBlur
  widgetCount=36
 }
 widgets [
  Blur {
   widgetClassPath=&BackgroundBlur
   widget=Blur
   properties {
    blurStrength=32
    visibility=Visible
    lowQualityFallbackBrush {
     bIsDynamicallyLoaded=false
     drawAs=Image
     imageType=NoImage
     mirroring=NoMirror
     resourceName=None
     resourceObject=null
     tiling=NoTile
     imageSize=<0 0>
     margin=<ltrb 0 0 0 0>
     outlineSettings {
      bUseBrushTransparency=false
      roundingType=HalfHeightRadius
      width=0
      color {colorUseRule=UseColor_Specified, specifiedColor=<rgba 0 0 0 0>}
      cornerRadii=<0 0 0 0>
     }
     tintColor {colorUseRule=UseColor_Specified, specifiedColor=<rgba 0 0 0 0.33>}
     uVRegion {bIsValid=false, max=<0 0>, min=<0 0>}
    }
   }
  }
  MainSafeZone {
   widgetClassPath=&SafeZone
   parent=Blur
   widget=MainSafeZone
   properties {padBottom=false, padTop=false}
   slot {class=BackgroundBlurSlot, properties {padding=<ltrb 0 0 0 0>}}
  }
  Overlay_1 {widgetClassPath=&Overlay, parent=MainSafeZone, widget=Overlay_1, slot {class=SafeZoneSlot}}
  HorizontalBox_2 {
   widgetClassPath=&HorizontalBox
   parent=Overlay_1
   widget=HorizontalBox_2
   slot {class=OverlaySlot, properties {horizontalAlignment=HAlign_Fill, verticalAlignment=VAlign_Fill}}
  }
  VerticalBox_8 {
   widgetClassPath=&VerticalBox
   parent=HorizontalBox_2
   widget=VerticalBox_8
   slot {
    class=HorizontalBoxSlot
    properties {horizontalAlignment=HAlign_Center, verticalAlignment=VAlign_Center, padding=<ltrb 0 0 0 64>, size {sizeRule=Fill, value=1}}
   }
  }
  HBox_SuccessfullyPurchased {
   widgetClassPath=&HorizontalBox
   bInherited=true
   parent=VerticalBox_8
   widget=HBox_SuccessfullyPurchased
   slot {class=VerticalBoxSlot, properties {horizontalAlignment=HAlign_Center}}
  }
  Spacer_42 {
   widgetClassPath=&Spacer
   parent=HBox_SuccessfullyPurchased
   widget=Spacer_42
   properties {size=<24 1>}
   slot {class=HorizontalBoxSlot}
  }
  BorderSuccessText {
   widgetClassPath=&Border
   parent=HBox_SuccessfullyPurchased
   widget=BorderSuccessText
   properties {
    background {
     bIsDynamicallyLoaded=false
     drawAs=NoDrawType
     imageType=NoImage
     mirroring=NoMirror
     resourceName=None
     resourceObject=null
     tiling=NoTile
     imageSize=<32 32>
     margin=<ltrb 0 0 0 0>
     outlineSettings {
      bUseBrushTransparency=false
      roundingType=HalfHeightRadius
      width=0
      color {colorUseRule=UseColor_Specified, specifiedColor=<rgba 0 0 0 0>}
      cornerRadii=<0 0 0 0>
     }
     tintColor {colorUseRule=UseColor_Specified, specifiedColor=<rgba 1 1 1 1>}
     uVRegion {bIsValid=false, max=<0 0>, min=<0 0>}
    }
    padding=<ltrb 0 0 0 0>
   }
   slot {class=HorizontalBoxSlot, properties {verticalAlignment=VAlign_Center}}
  }
  CommonTextBlock_4 {
   widgetClassPath=&CommonTextBlock
   parent=BorderSuccessText
   widget=CommonTextBlock_4
   properties {
    justification=Center
    style=/FortUILibrary/Styles/Typography/Foundation/TextStyle-Header-M.TextStyle-Header-M_C
    text=SUCCESSFULLY PURCHASED
    font {
     bForceMonospaced=false
     bMaterialIsStencil=false
     fontMaterial=null
     fontObject=&Burbank
     letterSpacing=0
     monospacedWidth=1
     size=32
     skewAmount=0
     typefaceFontName=Burbank-Black
     outlineSettings {
      bApplyOutlineToDropShadows=false
      bMiteredCorners=false
      bSeparateFillAlpha=false
      outlineMaterial=null
      outlineSize=0
      outlineColor=<rgba 0 0 0 1>
     }
    }
    shadowOffset=<0 0>
   }
   slot {class=BorderSlot, properties {padding=<ltrb 0 0 0 0>}}
  }
  BorderCheck {
   widgetClassPath=&Border
   parent=HBox_SuccessfullyPurchased
   widget=BorderCheck
   properties {
    background {
     bIsDynamicallyLoaded=false
     drawAs=NoDrawType
     imageType=NoImage
     mirroring=NoMirror
     resourceName=None
     resourceObject=null
     tiling=NoTile
     imageSize=<32 32>
     margin=<ltrb 0 0 0 0>
     outlineSettings {
      bUseBrushTransparency=false
      roundingType=HalfHeightRadius
      width=0
      color {colorUseRule=UseColor_Specified, specifiedColor=<rgba 0 0 0 0>}
      cornerRadii=<0 0 0 0>
     }
     tintColor {colorUseRule=UseColor_Specified, specifiedColor=<rgba 1 1 1 1>}
     uVRegion {bIsValid=false, max=<0 0>, min=<0 0>}
    }
    padding=<ltrb 0 0 0 0>
   }
   slot {class=HorizontalBoxSlot, properties {horizontalAlignment=HAlign_Center, verticalAlignment=VAlign_Center, padding=<ltrb 4 0 0 0>}}
  }
  Check {
   widgetClassPath=&Image
   parent=BorderCheck
   widget=Check
   properties {
    brush {
     bIsDynamicallyLoaded=false
     drawAs=Image
     imageType=NoImage
     mirroring=NoMirror
     resourceName=None
     resourceObject=/Game/UI/Foundation/Textures/Icons/Checks/T-Icon-Check-64
     tiling=NoTile
     imageSize=<48 48>
     margin=<ltrb 0 0 0 0>
     outlineSettings {
      bUseBrushTransparency=false
      roundingType=HalfHeightRadius
      width=0
      color {colorUseRule=UseColor_Specified, specifiedColor=<rgba 0 0 0 0>}
      cornerRadii=<0 0 0 0>
     }
     tintColor {colorUseRule=UseColor_Specified, specifiedColor=<rgba 1 1 1 1>}
     uVRegion {bIsValid=false, max=<0 0>, min=<0 0>}
    }
    colorAndOpacity=<rgba 0.07 0.7 0.175 1>
   }
   slot {class=BorderSlot, properties {padding=<ltrb 0 0 0 0>}}
  }
  BorderItemNAme {
   widgetClassPath=&Border
   bIsVariable=true
   parent=VerticalBox_8
   widget=BorderItemNAme
   properties {
    background {
     bIsDynamicallyLoaded=false
     drawAs=NoDrawType
     imageType=NoImage
     mirroring=NoMirror
     resourceName=None
     resourceObject=null
     tiling=NoTile
     imageSize=<32 32>
     margin=<ltrb 0 0 0 0>
     outlineSettings {
      bUseBrushTransparency=false
      roundingType=HalfHeightRadius
      width=0
      color {colorUseRule=UseColor_Specified, specifiedColor=<rgba 0 0 0 0>}
      cornerRadii=<0 0 0 0>
     }
     tintColor {colorUseRule=UseColor_Specified, specifiedColor=<rgba 1 1 1 1>}
     uVRegion {bIsValid=false, max=<0 0>, min=<0 0>}
    }
    padding=<ltrb 0 0 0 0>
   }
   slot {class=VerticalBoxSlot, properties {horizontalAlignment=HAlign_Center, verticalAlignment=VAlign_Center}}
  }
  Overlay_9 {widgetClassPath=&Overlay, parent=BorderItemNAme, widget=Overlay_9, slot {class=BorderSlot, properties {padding=<ltrb 0 0 0 0>}}}
  BorderRarityGlow {
   widgetClassPath=&Border
   parent=Overlay_9
   widget=BorderRarityGlow
   properties {
    background {
     bIsDynamicallyLoaded=false
     drawAs=NoDrawType
     imageType=NoImage
     mirroring=NoMirror
     resourceName=None
     resourceObject=null
     tiling=NoTile
     imageSize=<32 32>
     margin=<ltrb 0 0 0 0>
     outlineSettings {
      bUseBrushTransparency=false
      roundingType=HalfHeightRadius
      width=0
      color {colorUseRule=UseColor_Specified, specifiedColor=<rgba 0 0 0 0>}
      cornerRadii=<0 0 0 0>
     }
     tintColor {colorUseRule=UseColor_Specified, specifiedColor=<rgba 1 1 1 1>}
     uVRegion {bIsValid=false, max=<0 0>, min=<0 0>}
    }
    padding=<ltrb 0 0 0 0>
   }
   slot {class=OverlaySlot, properties {horizontalAlignment=HAlign_Fill, verticalAlignment=VAlign_Fill, padding=<ltrb -64 0 -64 0>}}
  }
  RarityGlow {
   widgetClassPath=&Image
   bIsVariable=true
   parent=BorderRarityGlow
   widget=RarityGlow
   properties {
    brush {
     bIsDynamicallyLoaded=false
     drawAs=Box
     imageType=NoImage
     mirroring=NoMirror
     resourceName=None
     resourceObject=/FortUILibrary/Styles/Border/Materials/M_Border_RarityGlow
     tiling=NoTile
     imageSize=<256 128>
     margin=<ltrb 0.5 0.5 0.5 0.5>
     outlineSettings {
      bUseBrushTransparency=false
      roundingType=HalfHeightRadius
      width=0
      color {colorUseRule=UseColor_Specified, specifiedColor=<rgba 0 0 0 0>}
      cornerRadii=<0 0 0 0>
     }
     tintColor {colorUseRule=UseColor_Specified, specifiedColor=<rgba 1 1 1 1>}
     uVRegion {bIsValid=false, max=<0 0>, min=<0 0>}
    }
    colorAndOpacity=<rgba 1 1 1 0.5>
   }
   slot {class=BorderSlot, properties {padding=<ltrb 0 0 0 0>}}
  }
  TextName {
   widgetClassPath=&CommonTextBlock
   bIsVariable=true
   parent=Overlay_9
   widget=TextName
   properties {
    justification=Center
    scrollStyle=&TextScrollStyle-Base_C
    style=&TextStyle-Header-XL-Horizon_C
    text=VERY LONG ITEM NAME EVEN
    font {
     bForceMonospaced=false
     bMaterialIsStencil=false
     fontMaterial=/FortUILibrary/Materials/Foundation/M_Font_Horizon_Dynamic
     fontObject=&Burbank
     letterSpacing=0
     monospacedWidth=1
     size=72
     skewAmount=0
     typefaceFontName=Burbank-Black
     outlineSettings {
      bApplyOutlineToDropShadows=true
      bMiteredCorners=false
      bSeparateFillAlpha=false
      outlineMaterial=null
      outlineSize=3
      outlineColor=<rgba 0 0 0 0.1>
     }
    }
    shadowOffset=<0 0>
   }
   slot {class=OverlaySlot}
  }
  HBox_SuccessfullyRefunded {
   widgetClassPath=&HorizontalBox
   bInherited=true
   parent=VerticalBox_8
   widget=HBox_SuccessfullyRefunded
   slot {class=VerticalBoxSlot, properties {horizontalAlignment=HAlign_Center}}
  }
  Spacer {
   widgetClassPath=&Spacer
   parent=HBox_SuccessfullyRefunded
   widget=Spacer
   properties {size=<24 1>}
   slot {class=HorizontalBoxSlot}
  }
  BorderSuccessRefundText {
   widgetClassPath=&Border
   parent=HBox_SuccessfullyRefunded
   widget=BorderSuccessRefundText
   properties {
    background {
     bIsDynamicallyLoaded=false
     drawAs=NoDrawType
     imageType=NoImage
     mirroring=NoMirror
     resourceName=None
     resourceObject=null
     tiling=NoTile
     imageSize=<32 32>
     margin=<ltrb 0 0 0 0>
     outlineSettings {
      bUseBrushTransparency=false
      roundingType=HalfHeightRadius
      width=0
      color {colorUseRule=UseColor_Specified, specifiedColor=<rgba 0 0 0 0>}
      cornerRadii=<0 0 0 0>
     }
     tintColor {colorUseRule=UseColor_Specified, specifiedColor=<rgba 1 1 1 1>}
     uVRegion {bIsValid=false, max=<0 0>, min=<0 0>}
    }
    padding=<ltrb 0 0 0 0>
   }
   slot {class=HorizontalBoxSlot, properties {verticalAlignment=VAlign_Center}}
  }
  CommonTextBlock {
   widgetClassPath=&CommonTextBlock
   parent=BorderSuccessRefundText
   widget=CommonTextBlock
   properties {
    justification=Center
    style=&TextStyle-Header-M_C
    text=PURCHASE HAS BEEN CANCELED AND YOUR V-BUCKS HAVE BEEN RETURNED
    font {
     bForceMonospaced=false
     bMaterialIsStencil=false
     fontMaterial=null
     fontObject=&Burbank
     letterSpacing=0
     monospacedWidth=1
     size=32
     skewAmount=0
     typefaceFontName=Burbank-Black
     outlineSettings {
      bApplyOutlineToDropShadows=false
      bMiteredCorners=false
      bSeparateFillAlpha=false
      outlineMaterial=null
      outlineSize=0
      outlineColor=<rgba 0 0 0 1>
     }
    }
    shadowOffset=<0 0>
   }
   slot {class=BorderSlot, properties {padding=<ltrb 0 0 0 0>}}
  }
  BorderRefundCheck {
   widgetClassPath=&Border
   parent=HBox_SuccessfullyRefunded
   widget=BorderRefundCheck
   properties {
    background {
     bIsDynamicallyLoaded=false
     drawAs=NoDrawType
     imageType=NoImage
     mirroring=NoMirror
     resourceName=None
     resourceObject=null
     tiling=NoTile
     imageSize=<32 32>
     margin=<ltrb 0 0 0 0>
     outlineSettings {
      bUseBrushTransparency=false
      roundingType=HalfHeightRadius
      width=0
      color {colorUseRule=UseColor_Specified, specifiedColor=<rgba 0 0 0 0>}
      cornerRadii=<0 0 0 0>
     }
     tintColor {colorUseRule=UseColor_Specified, specifiedColor=<rgba 1 1 1 1>}
     uVRegion {bIsValid=false, max=<0 0>, min=<0 0>}
    }
    padding=<ltrb 0 0 0 0>
   }
   slot {class=HorizontalBoxSlot, properties {horizontalAlignment=HAlign_Center, verticalAlignment=VAlign_Center, padding=<ltrb 4 0 0 0>}}
  }
  CheckRefund {
   widgetClassPath=&Image
   parent=BorderRefundCheck
   widget=CheckRefund
   properties {
    brush {
     bIsDynamicallyLoaded=false
     drawAs=Image
     imageType=NoImage
     mirroring=NoMirror
     resourceName=None
     resourceObject=&T-Icon-Check-64
     tiling=NoTile
     imageSize=<48 48>
     margin=<ltrb 0 0 0 0>
     outlineSettings {
      bUseBrushTransparency=false
      roundingType=HalfHeightRadius
      width=0
      color {colorUseRule=UseColor_Specified, specifiedColor=<rgba 0 0 0 0>}
      cornerRadii=<0 0 0 0>
     }
     tintColor {colorUseRule=UseColor_Specified, specifiedColor=<rgba 1 1 1 1>}
     uVRegion {bIsValid=false, max=<0 0>, min=<0 0>}
    }
    colorAndOpacity=<rgba 0.07 0.7 0.175 1>
   }
   slot {class=BorderSlot, properties {padding=<ltrb 0 0 0 0>}}
  }
  UniformGridPanelBundles {
   widgetClassPath=/Script/UMG.UniformGridPanel
   parent=VerticalBox_8
   widget=UniformGridPanelBundles
   properties {slotPadding=<ltrb 20 8 20 0>}
   slot {
    class=VerticalBoxSlot
    properties {horizontalAlignment=HAlign_Center, verticalAlignment=VAlign_Center, padding=<ltrb 0 18 0 0>, size {sizeRule=Fill, value=1}}
   }
  }
  SizeBox_0 {
   widgetClassPath=&SizeBox
   parent=UniformGridPanelBundles
   widget=SizeBox_0
   properties {bOverride_HeightOverride=true, bOverride_WidthOverride=true, heightOverride=354, widthOverride=380}
   slot {class=UniformGridSlot, properties {verticalAlignment=VAlign_Center}}
  }
  ScrollBox_Dyna {
   widgetClassPath=&ScrollBox
   bIsVariable=true
   parent=SizeBox_0
   widget=ScrollBox_Dyna
   properties {
    scrollBarVisibility=Collapsed
    widgetStyle {
     barThickness=9
     bottomShadowBrush {
      bIsDynamicallyLoaded=false
      drawAs=NoDrawType
      imageType=FullColor
      mirroring=NoMirror
      resourceName=../../../Engine/Content/Slate/Common/ScrollBoxShadowBottom.png
      resourceObject=null
      tiling=NoTile
      imageSize=<16 8>
      margin=<ltrb 0.5 0 0.5 1>
      outlineSettings {
       bUseBrushTransparency=false
       roundingType=HalfHeightRadius
       width=0
       color {colorUseRule=UseColor_Specified, specifiedColor=<rgba 0 0 0 0>}
       cornerRadii=<0 0 0 0>
      }
      tintColor {colorUseRule=UseColor_Specified, specifiedColor=<rgba 1 1 1 1>}
      uVRegion {bIsValid=false, max=<0 0>, min=<0 0>}
     }
     horizontalScrolledContentPadding=<ltrb 0 0 1 0>
     leftShadowBrush {
      bIsDynamicallyLoaded=false
      drawAs=NoDrawType
      imageType=FullColor
      mirroring=NoMirror
      resourceName=../../../Engine/Content/Slate/Common/ScrollBoxShadowLeft.png
      resourceObject=null
      tiling=NoTile
      imageSize=<8 16>
      margin=<ltrb 1 0.5 0 0.5>
      outlineSettings {
       bUseBrushTransparency=false
       roundingType=HalfHeightRadius
       width=0
       color {colorUseRule=UseColor_Specified, specifiedColor=<rgba 0 0 0 0>}
       cornerRadii=<0 0 0 0>
      }
      tintColor {colorUseRule=UseColor_Specified, specifiedColor=<rgba 1 1 1 1>}
      uVRegion {bIsValid=false, max=<0 0>, min=<0 0>}
     }
     rightShadowBrush {
      bIsDynamicallyLoaded=false
      drawAs=NoDrawType
      imageType=FullColor
      mirroring=NoMirror
      resourceName=../../../Engine/Content/Slate/Common/ScrollBoxShadowRight.png
      resourceObject=null
      tiling=NoTile
      imageSize=<8 16>
      margin=<ltrb 0 0.5 1 0.5>
      outlineSettings {
       bUseBrushTransparency=false
       roundingType=HalfHeightRadius
       width=0
       color {colorUseRule=UseColor_Specified, specifiedColor=<rgba 0 0 0 0>}
       cornerRadii=<0 0 0 0>
      }
      tintColor {colorUseRule=UseColor_Specified, specifiedColor=<rgba 1 1 1 1>}
      uVRegion {bIsValid=false, max=<0 0>, min=<0 0>}
     }
     topShadowBrush {
      bIsDynamicallyLoaded=false
      drawAs=NoDrawType
      imageType=FullColor
      mirroring=NoMirror
      resourceName=../../../Engine/Content/Slate/Common/ScrollBoxShadowTop.png
      resourceObject=null
      tiling=NoTile
      imageSize=<16 8>
      margin=<ltrb 0.5 1 0.5 0>
      outlineSettings {
       bUseBrushTransparency=false
       roundingType=HalfHeightRadius
       width=0
       color {colorUseRule=UseColor_Specified, specifiedColor=<rgba 0 0 0 0>}
       cornerRadii=<0 0 0 0>
      }
      tintColor {colorUseRule=UseColor_Specified, specifiedColor=<rgba 1 1 1 1>}
      uVRegion {bIsValid=false, max=<0 0>, min=<0 0>}
     }
     verticalScrolledContentPadding=<ltrb 0 0 0 1>
    }
   }
   slot {class=SizeBoxSlot}
  }
  DynaBox {
   widgetClassPath=&DynamicEntryBox
   bIsVariable=true
   parent=ScrollBox_Dyna
   widget=DynaBox
   properties {
    entryBoxType=Wrap
    entryWidgetClass=/FortniteCoreUI/UI/ItemReceived/BundleItemMiniCardWidget.BundleItemMiniCardWidget_C
    maxElementSize=100
    numDesignerPreviewEntries=20
    entrySpacing=<8 8>
   }
   slot {class=ScrollBoxSlot}
  }
  OverlayMainItem {
   widgetClassPath=&Overlay
   bIsVariable=true
   parent=UniformGridPanelBundles
   widget=OverlayMainItem
   slot {class=UniformGridSlot, properties {horizontalAlignment=HAlign_Fill, verticalAlignment=VAlign_Fill}}
  }
  BorderRarityFlare {
   widgetClassPath=&Border
   parent=OverlayMainItem
   widget=BorderRarityFlare
   properties {
    background {
     bIsDynamicallyLoaded=false
     drawAs=NoDrawType
     imageType=NoImage
     mirroring=NoMirror
     resourceName=None
     resourceObject=null
     tiling=NoTile
     imageSize=<32 32>
     margin=<ltrb 0 0 0 0>
     outlineSettings {
      bUseBrushTransparency=false
      roundingType=HalfHeightRadius
      width=0
      color {colorUseRule=UseColor_Specified, specifiedColor=<rgba 0 0 0 0>}
      cornerRadii=<0 0 0 0>
     }
     tintColor {colorUseRule=UseColor_Specified, specifiedColor=<rgba 1 1 1 1>}
     uVRegion {bIsValid=false, max=<0 0>, min=<0 0>}
    }
    padding=<ltrb 0 0 0 0>
   }
   slot {class=OverlaySlot, properties {horizontalAlignment=HAlign_Fill, verticalAlignment=VAlign_Fill}}
  }
  ImageRarityFlare {
   widgetClassPath=&Image
   bIsVariable=true
   parent=BorderRarityFlare
   widget=ImageRarityFlare
   properties {
    visibility=SelfHitTestInvisible
    brush {
     bIsDynamicallyLoaded=false
     drawAs=Image
     imageType=NoImage
     mirroring=NoMirror
     resourceName=None
     resourceObject=/Game/UI/Frontend/Rewards/materials/M_UI_DailyReward_Rarity-Flares
     tiling=NoTile
     imageSize=<256 256>
     margin=<ltrb 0 0 0 0>
     outlineSettings {
      bUseBrushTransparency=false
      roundingType=HalfHeightRadius
      width=0
      color {colorUseRule=UseColor_Specified, specifiedColor=<rgba 0 0 0 0>}
      cornerRadii=<0 0 0 0>
     }
     tintColor {colorUseRule=UseColor_Specified, specifiedColor=<rgba 1 1 1 1>}
     uVRegion {bIsValid=false, max=<0 0>, min=<0 0>}
    }
    renderTransform {angle=0, scale=<2.75 1.875>, shear=<0 0>, translation=<0 0>}
   }
   slot {class=BorderSlot, properties {padding=<ltrb 0 0 0 0>}}
  }
  BorderItemCard {
   widgetClassPath=&Border
   parent=OverlayMainItem
   widget=BorderItemCard
   properties {
    background {
     bIsDynamicallyLoaded=false
     drawAs=NoDrawType
     imageType=NoImage
     mirroring=NoMirror
     resourceName=None
     resourceObject=null
     tiling=NoTile
     imageSize=<32 32>
     margin=<ltrb 0 0 0 0>
     outlineSettings {
      bUseBrushTransparency=false
      roundingType=HalfHeightRadius
      width=0
      color {colorUseRule=UseColor_Specified, specifiedColor=<rgba 0 0 0 0>}
      cornerRadii=<0 0 0 0>
     }
     tintColor {colorUseRule=UseColor_Specified, specifiedColor=<rgba 1 1 1 1>}
     uVRegion {bIsValid=false, max=<0 0>, min=<0 0>}
    }
    padding=<ltrb 0 0 0 0>
   }
   slot {class=OverlaySlot, properties {horizontalAlignment=HAlign_Center, verticalAlignment=VAlign_Center}}
  }
  SizeBox_7 {
   widgetClassPath=&SizeBox
   parent=BorderItemCard
   widget=SizeBox_7
   properties {bOverride_MinDesiredHeight=true, bOverride_MinDesiredWidth=true, minDesiredHeight=414, minDesiredWidth=256}
   slot {class=BorderSlot, properties {padding=<ltrb 0 0 0 0>}}
  }
  ItemCard {
   widgetClassPath=/FortniteCoreUI/Athena/UI/Frontend/CosmeticItemCard/FortCosmeticItemCard
   bIsVariable=true
   parent=SizeBox_7
   widget=ItemCard
   properties {cardWidth=256}
   slot {class=SizeBoxSlot, properties {horizontalAlignment=HAlign_Center, verticalAlignment=VAlign_Center}}
  }
  Text_CosmeticWarning {widgetClassPath=&CommonTextBlock, bInherited=true, widget=null}
  TextUndoAvailability {widgetClassPath=&CommonTextBlock, bInherited=true, widget=null}
  Text_UndoBottomBarWarning {widgetClassPath=&CommonTextBlock, bInherited=true, widget=null}
  Button_Undo {widgetClassPath=&CommonButtonBase, bInherited=true, widget=null}
 ]
}
